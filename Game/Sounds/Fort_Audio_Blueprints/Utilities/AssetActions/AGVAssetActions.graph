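# Editor utility blueprint that batch-edits ambient audio settings on the selected actors, offering AdjustPriority and AdjustCrossfadeTime functions.
# Each function loops over the selected actors and every AmbientAudioComponent they own.
# An operation enum picks whether the entered value replaces, is added to, or is subtracted from each component's current priority or crossfade time.
# Both functions share that one enum, and the event graph is empty.
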EditorUtilityBlueprint /Game/Sounds/Fort_Audio_Blueprints/Utilities/AssetActions/AGVAssetActions
graphs [
 AdjustCrossfadeTime {nodes [
   AdjustCrossfadeTime!() -> (then, Value:double=~, Operation:EAGVAssetAction_PriorityOperationType Enum=NewEnumerator0) {is=title K2Node_FunctionEntry}
   > n1 = GetSelectedActors()
   > n2 = For Each Loop(Array:Array of Actor Object References=@n1) -> (Array Element:Actor Object Reference)
   > n3 = For Each Loop(Array:Array of Ambient Audio Component Object References=Get Components By Class!(self:Actor Object Reference=@n2, ComponentClass:Actor Component Class Reference=/Script/AmbientAudio.AmbientAudioComponent) -> (ReturnValue:Array of Ambient Audio Component Object References)) -> (Array Element:Ambient Audio Component Object Reference)
   n4 = Get Value() -> (Value:double)
   n5 = Get CrossfadeTime!(self:Ambient Audio Component Object Reference=@n3) -> (CrossfadeTime:float)
   > SetCrossfadeTime!(
    execute
    self:Ambient Audio Component Object Reference=@n3
    InCrossfadeTime:float=Select!(
     NewEnumerator0:double=Get Value()
     NewEnumerator1:double=float + float(A=@n5, B=@n4)
     NewEnumerator2:double=float - float(A=@n5, B=@n4)
     Index:EAGVAssetAction_PriorityOperationType Enum=Get Operation() -> (Operation:EAGVAssetAction_PriorityOperationType Enum=NewEnumerator0)
    ) -> (ReturnValue:double)
   ) -> (then)
  ]}
 AdjustPriority {nodes [
   AdjustPriority!() -> (then, Value:int, Operation:EAGVAssetAction_PriorityOperationType Enum=NewEnumerator0) {is=title K2Node_FunctionEntry}
   > n1 = GetSelectedActors()
   > n2 = For Each Loop(Array:Array of Actor Object References=@n1) -> (Array Element:Actor Object Reference)
   > n3 = For Each Loop(Array:Array of Ambient Audio Component Object References=Get Components By Class(self=@n2, ComponentClass=&AmbientAudioComponent)) -> (Array Element:Ambient Audio Component Object Reference)
   n4 = Get Priority!(self:Ambient Audio Component Object Reference=@n3) -> (Priority:int=0)
   n5 = Get Value() -> (Value:int=0)
   > SetPriority!(
    execute
    self:Ambient Audio Component Object Reference=@n3
    InPriority:int=Select(
     NewEnumerator0:int=Get Value() -> (Value=0)
     NewEnumerator1:int=int + int(A=@n4, B=@n5)
     NewEnumerator2:int=int - int(A=@n4, B=@n5)
     Index=Get Operation() -> (Operation=NewEnumerator0)
    ) -> (ReturnValue:int)
   ) -> (then)
  ]}
 EventGraph {links [], nodes []}
]
variables []
