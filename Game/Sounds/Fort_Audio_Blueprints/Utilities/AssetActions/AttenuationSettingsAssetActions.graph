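# Editor utility with asset actions that batch-edit the Attenuation struct on selected Sound Attenuation assets via Get/SetEditorProperty.
# AdjustReverbSendByDb dB-scales first submix send min/max levels, dropping other sends; its reverb wet-level scaling is unwired.
# ApplyDistantReverb replaces sends with a curve-driven distant reverb template; ResetHPFToDefault zeroes HPF frequencies.
# ReverbSendAudit collects assets with reverb or submix sends enabled into a local collection; Tool prints dB-adjusted linear values.
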EditorUtilityBlueprint /Game/Sounds/Fort_Audio_Blueprints/Utilities/AssetActions/AttenuationSettingsAssetActions
graphs [
 AdjustReverbSendByDb {nodes [
   n1 = AdjustReverbSendByDb!() -> (then, db Adjustment:double=~) {is=title K2Node_FunctionEntry}
   > Set DbAdjustment(DbAdjustment:double=@n1) -> (Output_Get:double)
   > n2 = Iterate Over Selected Assets()
   > n3 = GetEditorProperty!(execute:Exec, Object:Object Reference=@n2, PropertyName:Name=Attenuation) -> (then:Exec, PropertyValue:"Sound Attenuation Settings Structure (by ref)", ReturnValue:bool)
   n4 = Break Sound Attenuation Settings!(SoundAttenuationSettings:"Sound Attenuation Settings Structure (by ref)"=@n3) -> (
    bEnableSourceDataOverride:bool=~
    bEnableSendToAudioLink:bool=~
    AudioLinkSettingsOverride:Audio Link Settings Abstract Object Reference
    NonSpatializedRadiusStart:float=~
    NonSpatializedRadiusEnd:float=~
    NonSpatializedRadiusMode:ENonSpatializedRadiusSpeakerMapMode Enum
    ReverbWetLevelMin:float=~
    ReverbWetLevelMax:float=~
    SubmixSendSettings:Array of Attenuation Submix Send Settings Structures
   ) {is=title K2Node_BreakStruct}
   > Branch(Condition=Is Not Empty!(TargetArray:Array of Attenuation Submix Send Settings Structures=@n4.SubmixSendSettings) -> (ReturnValue:bool))
   n5 = "Get (a copy)"(Array:Array of Attenuation Submix Send Settings Structures=@n4.SubmixSendSettings) -> (Output:Attenuation Submix Send Settings Structure)
   n6 = Break Attenuation Submix Send Settings!(AttenuationSubmixSendSettings:"Attenuation Submix Send Settings Structure (by ref)"=@n5) -> (
    SendLevelControlMethod:ESendLevelControlMethod Enum
    SoundSubmix:Sound Submix Base Object Reference
    SendLevel:float=~
    DisableManualSendClamp:bool=~
    MinSendLevel:float=~
    MaxSendLevel:float=~
    MinSendDistance:float=~
    MaxSendDistance:float=~
    CustomSendLevelCurve:Runtime Float Curve Structure
    bEnableLPFCutoff:bool=~
    LPFCutoff:float=~
    bEnableHPFCutoff:bool=~
    HPFCutoff:float=~
   ) {is=title K2Node_BreakStruct}
   n7 = Get DbAdjustment() -> (DbAdjustment:double)
   > n8 = Set members in Attenuation Submix Send Settings!(
    execute:Exec
    StructRef:"Attenuation Submix Send Settings Structure (by ref)"=@n5
    MinSendLevel:float=Adjust Linear by Db!(Linear:double=@n6.MinSendLevel, Db Adjustment:double=@n7) -> (Scaled Linear:double=~) {is=title K2Node_MacroInstance}
    MaxSendLevel:float=Adjust Linear by Db(Linear=@n6.MaxSendLevel, Db Adjustment=@n7) -> (Scaled Linear=~)
   ) -> (then:Exec, StructOut:"Attenuation Submix Send Settings Structure (by ref)") {is=title K2Node_SetFieldsInStruct}
   > n9 = Set members in Sound Attenuation Settings!(
    execute:Exec
    StructRef:"Sound Attenuation Settings Structure (by ref)"=@n3
    SubmixSendSettings:Array of Attenuation Submix Send Settings Structures=Make Array("[0]":Attenuation Submix Send Settings Structure=@n8) -> (Array:Array of Attenuation Submix Send Settings Structures)
   ) -> (then:Exec, StructOut:"Sound Attenuation Settings Structure (by ref)") {is=title K2Node_SetFieldsInStruct}
   > SetEditorProperty(Object=@n2, PropertyName=Attenuation, PropertyValue:"Sound Attenuation Settings Structure (by ref)"=@n9)
   Adjust Linear by Db(Linear=@n4.ReverbWetLevelMin, Db Adjustment=@n7) -> (Scaled Linear=~)
   Adjust Linear by Db(Linear=@n4.ReverbWetLevelMax, Db Adjustment=@n7) -> (Scaled Linear=~)
  ]}
 ApplyDistantReverb {nodes [
   ApplyDistantReverb() {is=title K2Node_FunctionEntry}
   > n1 = Sequence()
   > n2 = LoadAsset_Blocking(Asset=Get DistanceReverbSubmixSoftRef() -> (DistanceReverbSubmixSoftRef:Sound Submix Base Soft Object Reference))
   > n3 = Cast To SoundSubmixBase!(execute, Object=@n2) -> (then, CastFailed, AsSound Submix Base:Sound Submix Base Object Reference) {is=title K2Node_DynamicCast}
   > Set members in Attenuation Submix Send Settings!(
    execute
    StructRef=Get SubmixSendSetting() -> (SubmixSendSetting:Attenuation Submix Send Settings Structure)
    SoundSubmix:Sound Submix Base Object Reference=@n3
   ) -> (then, StructOut)
   n4 = Iterate Over Selected Assets(Trigger=@n1.then_1)
   > n5 = GetEditorProperty(Object=@n4, PropertyName=Attenuation)
   > n6 = Set members in Sound Attenuation Settings(StructRef=@n5, SubmixSendSettings=Make Array("[0]"=Get SubmixSendSetting()))
   > SetEditorProperty(Object=@n4, PropertyName=Attenuation, PropertyValue=@n6)
  ]}
 EventGraph {links [], nodes []}
 ResetHPFToDefault {nodes [
   ResetHPFToDefault() {is=title K2Node_FunctionEntry}
   > n1 = Iterate Over Selected Assets()
   > n2 = GetEditorProperty(Object=@n1, PropertyName=Attenuation)
   > n3 = Set members in Sound Attenuation Settings!(execute, StructRef=@n2, HPFFrequencyAtMin:float=0, HPFFrequencyAtMax:float=0) -> (then, StructOut)
   > SetEditorProperty(Object=@n1, PropertyName=Attenuation, PropertyValue=@n3)
  ]}
 ReverbSendAudit {nodes [
   n1 = ReverbSendAudit!() -> (then, Name:Name) {is=title K2Node_FunctionEntry}
   > n2 = CreateCollection!(
    execute
    self:Collection Manager Subsystem Object Reference=Get CollectionManagerScriptingSubsystem()
    Container:Collection Container Source Structure
    Collection:Name=@n1
    ShareType:ECollectionScriptingShareType Enum=Local
   ) -> (then, OutNewCollection:Collection Structure, ReturnValue:bool)
   > n3 = Iterate Over Selected Assets()
   AddAssetsToCollection!(
    execute=@n3.Completed
    self:Collection Manager Subsystem Object Reference=Get CollectionManagerScriptingSubsystem()
    Collection:"Collection Structure (by ref)"=@n2
    AssetPaths:Array of Soft Object Path Structures=Get SoftObjPaths() -> (SoftObjPaths:Array of Soft Object Path Structures)
   ) -> (then, ReturnValue:bool)
   n4 = GetEditorProperty(execute=@n3.Iteration, Object=@n3, PropertyName=Attenuation)
   n5 = Break Sound Attenuation Settings!(SoundAttenuationSettings=@n4) -> (
    bEnableReverbSend:bool=~
    bEnableSubmixSends:bool=~
    bEnableSourceDataOverride=~
    bEnableSendToAudioLink=~
    AudioLinkSettingsOverride
    NonSpatializedRadiusStart=~
    NonSpatializedRadiusEnd=~
    NonSpatializedRadiusMode
   )
   > Branch(Condition=OR Boolean(A=@n5, B=@n5.bEnableSubmixSends))
   > Add(
    TargetArray:Array of Soft Object Path Structures=Get SoftObjPaths()
    NewItem:"Soft Object Path Structure (by ref)"=GetSoftObjectPath(Object=@n3)
   ) {is=title K2Node_CallArrayFunction}
  ]}
 Tool {nodes [
   n1 = Tool!() -> (then, Linear:double=~, Db:double=~) {is=title K2Node_FunctionEntry}
   n2 = float + float(A=Linear to Db!(Linear:double=@n1) -> (Db:double=~) {is=title K2Node_MacroInstance}, B=@n1.Db)
   > PrintString(InString="To String (Float)"(InDouble=Db to Linear!(Db:double=@n2) -> (Linear:double=~) {is=title K2Node_MacroInstance}))
  ]}
]
variables [
 DistanceReverbSubmixSoftRef {type=SoundSubmixBase, default=&Fortnite_Reverb_Distant}
 SubmixSendSetting {
  type=AttenuationSubmixSendSettings
  default {
   bEnableHPFCutoff=false
   bEnableLPFCutoff=false
   disableManualSendClamp=false
   hPFCutoff=20
   lPFCutoff=20000
   maxSendDistance=7500
   maxSendLevel=1
   minSendDistance=0
   minSendLevel=0
   sendLevel=0.2
   sendLevelControlMethod=CustomCurve
   soundSubmix=null
   customSendLevelCurve {
    externalCurve=/Game/Sounds/SubMixes/Reverb/Fortnite_Reverb_Distant_Send
    editorCurveData {defaultValue=3.4028234663852886e+38, postInfinityExtrap=RCCE_Constant, preInfinityExtrap=RCCE_Constant, keys []}
   }
  }
 }
]
