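# AudioAssetActionBase is an Editor Utility Blueprint in Fortnite's audio blueprint utilities, named as a base for audio asset actions.
# Its event graph is empty and it declares no variables, so it carries no logic of its own in this file.
# The parent class is not shown, so any inherited behavior is not visible here.
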EditorUtilityBlueprint /Game/Sounds/Fort_Audio_Blueprints/Utilities/AssetActions/AudioAssetActionBase
graphs [EventGraph {links [], nodes []}]
variables []
