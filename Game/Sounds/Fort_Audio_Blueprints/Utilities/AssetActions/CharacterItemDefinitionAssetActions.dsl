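; Editor asset-action utility for character item definitions that builds a CSV string of their foley sound tags.
; ExportFoleyTagsToCSVString casts each object to FortItemDefinition, matches owned gameplay tags against StepType, ClothingType and HandType, and keeps each leaf name.
; It appends one "Tagged" row per item with display name, those values and the CIDTagging metadata note.
; The tag values are not reset between items, and the Test event is empty.
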(event Custom|Test)

(fn ExportFoleyTagsToCSVString
  (params (Objects :type Array%20of%20Object%20References))
  (LocalVariables|SetCSV "")
  (for _array_element Objects
    (bind _asfort_item_definition (Utilities|Casting|CastToFortItemDefinition _array_element))
    (bind _returnvalue (EditorScripting|Metadata|GetMetadataTag _asfort_item_definition "CIDTagging"))
    (for _array_element_2 (GameplayTags|BreakGameplayTagContainer (GameplayTags|GetOwnedGameplayTags _asfort_item_definition :!class /Script/GameplayTags.GameplayTagAssetInterface))
      (if (GameplayTags|MatchesTag _array_element_2 "(TagName=\"SoundLibrary.ID.StepType\")")
        (bind _output_get (LocalVariables|SetFeet (.RightS (Utilities|String|Split (GameplayTags|GetDebugStringfromGameplayTag _array_element_2) "." "IgnoreCase" "FromEnd"))))
        (elif (GameplayTags|MatchesTag _array_element_2 "(TagName=\"SoundLibrary.ID.ClothingType\")")
          (bind _output_get_2 (LocalVariables|SetClothes (.RightS (Utilities|String|Split (GameplayTags|GetDebugStringfromGameplayTag _array_element_2) "." "IgnoreCase" "FromEnd"))))
          (elif (GameplayTags|MatchesTag _array_element_2 "(TagName=\"SoundLibrary.ID.HandType\")")
            (bind _output_get_3 (LocalVariables|SetHands (.RightS (Utilities|String|Split (GameplayTags|GetDebugStringfromGameplayTag _array_element_2) "." "IgnoreCase" "FromEnd"))))))))
    (LocalVariables|SetCSV (Utilities|Text|FormatText "NSLOCTEXT(\"[5816C31C48B421A31A539ABB662EBBF0]\", \"F49899EB4D9251396131B7B317DF0A90\", \"{Previous}{Name}, Tagged, {Feet}, {Clothes}, {Hands}, {Notes}{NewLine}\")" (LocalVariables|GetCSV) (Utilities|Text|ToText(String) (Utilities|GetDisplayName _asfort_item_definition)) (stale _output_get) (stale _output_get_2) (stale _output_get_3) _returnvalue (Utilities|Text|MakeLiteralText "NSLOCTEXT(\"[5816C31C48B421A31A539ABB662EBBF0]\", \"2D28825740BB9088EF1918ACD091A8F1\", \"\\r\\n\")"))))
  (return (LocalVariables|GetCSV)))
