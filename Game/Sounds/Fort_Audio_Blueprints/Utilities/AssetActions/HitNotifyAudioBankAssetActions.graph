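# Editor utility asset action that converts selected WeaponHitNotifyAudioBank assets into Sound Library assets.
# Embedded Python creates each renamed SL_ asset, which gets a parent and gameplay-tagged Instigate and Receive sounds; Receive.DeathCrit reuses SoundDeathCrit.
# Transient adds are persisted by copying FullLibrary into Sounds through editor properties.
# ReplaceHitNotifyReferencers also converts, then appends the new library to the SoundLibraryComponent of referencing Blueprints.
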EditorUtilityBlueprint /Game/Sounds/Fort_Audio_Blueprints/Utilities/AssetActions/HitNotifyAudioBankAssetActions
graphs [
 ConvertToSoundLibrary {nodes [
   ConvertToSoundLibrary() {is=title K2Node_FunctionEntry}
   > n1 = Iterate Over Selected Assets()
   > n2 = Cast To WeaponHitNotifyAudioBank!(execute, Object=@n1) -> (then, CastFailed, AsWeapon Hit Notify Audio Bank:Weapon Hit Notify Audio Bank Object Reference) {is=title K2Node_DynamicCast}
   > ConvertToSoundLibraryInternal!(execute, self, HitNotifyAudioBank:Weapon Hit Notify Audio Bank Object Reference=@n2) -> (then, Sound Library:Sound Library Object Reference)
  ]}
 ConvertToSoundLibraryInternal {nodes [
   n1 = ConvertToSoundLibraryInternal!() -> (then, HitNotifyAudioBank:Weapon Hit Notify Audio Bank Object Reference) {is=title K2Node_FunctionEntry}
   > n2 = Sequence()
   n3 = Split!(InPath:String=Get Object Path String(Object=@n1)) -> (PathPart:String, FilenamePart:String, ExtensionPart:String)
   n4 = Replace(SourceString=@n3.FilenamePart, From=HitNotifyAudioBank_, To=HitNotify_)
   > n5 = Execute Python Script!(
    execute
    PythonScript="import unreal\r\n\r\nfactory = unreal.SoundLibraryFactory()\r\n\r\nlibrary = unreal.AssetToolsHelpers.get_asset_tools().create_asset(name, path, unreal.SoundLibrary, factory)"
    path:String=@n3
    name:String=Append(A=SL_, B=Replace(SourceString=@n4, From=_HitNotifyAudioBank, To=_HitNotify))
   ) -> (then, ReturnValue, library:Sound Library Object Reference) {inputs [path,name], outputs [library]}
   > Set SoundLibrary(SoundLibrary:Sound Library Object Reference=@n5.library) -> (Output_Get:Sound Library Object Reference)
   SetEditorProperty(
    execute=@n2.then_1
    Object=Get SoundLibrary() -> (SoundLibrary:Sound Library Object Reference)
    PropertyName=Parent
    PropertyValue:Sound Library Object Reference=Get SoundLibraryParent() -> (SoundLibraryParent:Sound Library Object Reference)
   )
   n6 = Get SoundLibrary()
   > "Add Sound (Transient)"!(
    execute
    self:Sound Library Object Reference=@n6
    InTag:Gameplay Tag Structure=(TagName="SoundLibrary.HitNotify.Instigate.Body")
    InSound:Sound Base Object Reference=Get SoundBody!(self:Weapon Hit Notify Audio Bank Object Reference=@n1) -> (SoundBody:Sound Base Object Reference)
   ) -> (then)
   > "Add Sound (Transient)"(
    self=@n6
    InTag=(TagName="SoundLibrary.HitNotify.Instigate.BodyCrit")
    InSound=Get SoundCrit!(self:Weapon Hit Notify Audio Bank Object Reference=@n1) -> (SoundCrit:Sound Base Object Reference)
   )
   > "Add Sound (Transient)"(
    self=@n6
    InTag=(TagName="SoundLibrary.HitNotify.Instigate.Shield")
    InSound=Get SoundShield!(self:Weapon Hit Notify Audio Bank Object Reference=@n1) -> (SoundShield:Sound Base Object Reference)
   )
   > "Add Sound (Transient)"(
    self=@n6
    InTag=(TagName="SoundLibrary.HitNotify.Instigate.Death")
    InSound=Get SoundDeath!(self:Weapon Hit Notify Audio Bank Object Reference=@n1) -> (SoundDeath:Sound Base Object Reference)
   )
   > "Add Sound (Transient)"(
    self=@n6
    InTag=(TagName="SoundLibrary.HitNotify.Instigate.DeathCrit")
    InSound=Get SoundDeathCrit!(self:Weapon Hit Notify Audio Bank Object Reference=@n1) -> (SoundDeathCrit:Sound Base Object Reference)
   )
   > "Add Sound (Transient)"(
    self=@n6
    InTag=(TagName="SoundLibrary.HitNotify.Receive.Body")
    InSound=Get SoundBodyReceive!(self:Weapon Hit Notify Audio Bank Object Reference=@n1) -> (SoundBodyReceive:Sound Base Object Reference)
   )
   > "Add Sound (Transient)"(
    self=@n6
    InTag=(TagName="SoundLibrary.HitNotify.Receive.Shield")
    InSound=Get SoundShieldReceive!(self:Weapon Hit Notify Audio Bank Object Reference=@n1) -> (SoundShieldReceive:Sound Base Object Reference)
   )
   > "Add Sound (Transient)"(
    self=@n6
    InTag=(TagName="SoundLibrary.HitNotify.Receive.BodyCrit")
    InSound=Get SoundCritReceive!(self:Weapon Hit Notify Audio Bank Object Reference=@n1) -> (SoundCritReceive:Sound Base Object Reference)
   )
   > "Add Sound (Transient)"(
    self=@n6
    InTag=(TagName="SoundLibrary.HitNotify.Receive.Death")
    InSound=Get SoundDeathReceive!(self:Weapon Hit Notify Audio Bank Object Reference=@n1) -> (SoundDeathReceive:Sound Base Object Reference)
   )
   > "Add Sound (Transient)"(self=@n6, InTag=(TagName="SoundLibrary.HitNotify.Receive.DeathCrit"), InSound=Get SoundDeathCrit(self=@n1))
   > "Add Sound (Transient)"(
    self=@n6
    InTag=(TagName="SoundLibrary.HitNotify.Receive.Fall")
    InSound=Get SoundFallReceive!(self:Weapon Hit Notify Audio Bank Object Reference=@n1) -> (SoundFallReceive:Sound Base Object Reference)
   )
   > "Add Sound (Transient)"(
    self=@n6
    InTag=(TagName="SoundLibrary.HitNotify.Receive.FallDeath")
    InSound=Get SoundFallDeathReceive!(self:Weapon Hit Notify Audio Bank Object Reference=@n1) -> (SoundFallDeathReceive:Sound Base Object Reference)
   )
   n7 = Get SoundLibrary()
   > n8 = GetEditorProperty!(execute:Exec, Object:Object Reference=@n7, PropertyName:Name=FullLibrary) -> (then:Exec, PropertyValue:Map of Gameplay Tag Structures to Sound Base Soft Object References, ReturnValue:bool)
   > SetEditorProperty(Object=@n7, PropertyName=Sounds, PropertyValue:Map of Gameplay Tag Structures to Sound Base Soft Object References=@n8)
   > Return Node!(execute, Sound Library:Sound Library Object Reference=Get SoundLibrary())
   Get Sounds!(self:Sound Library Object Reference=@n7) -> (Sounds:Map of Gameplay Tag Structures to Sound Base Soft Object References)
   Set NewLocalVar(NewLocalVar:Map of Gameplay Tag Structures to Sound Base Soft Object References=@n8) -> (Output_Get:Map of Gameplay Tag Structures to Sound Base Soft Object References)
  ]}
 EventGraph {links [], nodes []}
 ReplaceHitNotifyReferencers {nodes [
   ReplaceHitNotifyReferencers() {is=title K2Node_FunctionEntry}
   > n1 = Iterate Over Selected Assets()
   > n2 = Cast To WeaponHitNotifyAudioBank(Object=@n1)
   > n3 = ConvertToSoundLibraryInternal(HitNotifyAudioBank=@n2) {is=title K2Node_CallFunction}
   > n4 = Get Asset Referencers!(execute:Exec, Object:Object Reference=@n2) -> (Iteration:Exec, Referencer:Object Reference, Completed:Exec) {is=title K2Node_MacroInstance}
   > n5 = GetBlueprintAsset(Object=@n4)
   > n6 = Gather Subobject Data For Blueprint(self=Get SubobjectDataSubsystem(), Context=@n5)
   > n7 = For Each Loop(Array:Array of Subobject Data Handle Structures=@n6) -> (Array Element:Subobject Data Handle Structure)
   n8 = GetData(DataHandle=@n7)
   > Branch(Condition="Equal (Name)"(A=GetVariableName!(Data:"Subobject Data Structure (by ref)"=@n8) -> (ReturnValue:Name=None), B=SoundLibraryComponent))
   > TransactObject(Object=@n5)
   > n9 = Cast To SoundLibraryComponent!(execute, Object=GetObject!(Data:"Subobject Data Structure (by ref)"=@n8, bEvenIfPendingKill:bool=false) -> (ReturnValue:Object Reference)) -> (then, CastFailed, AsSoundLibraryComponent:Sound Library Component Object Reference) {is=title K2Node_DynamicCast}
   > n10 = GetEditorProperty(Object=@n9, PropertyName=Libraries) -> (PropertyValue:Array of Sound Library Object References)
   > Add(TargetArray:Array of Sound Library Object References=@n10, NewItem:"Sound Library Object Reference (by ref)"=@n3) {is=title K2Node_CallArrayFunction}
   > SetEditorProperty(Object=@n9, PropertyName=Libraries, PropertyValue:Array of Sound Library Object References=@n10)
  ]}
]
variables []
