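# Editor utility with asset actions for MetaSounds: batch-updating input defaults, creating a source from a reference, and fixing duplicate references.
# UpdateMetaSoundInputs edits selected Sources or Patches through the builder, choosing a literal by input data type, then saves and logs failures.
# Unmapped types fall back to object or object array.
# CreateFromPreset runs embedded Python with MetaSoundSourceFactory; Fix Up Duplicate References processes selected assets against supplied reference directories.
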EditorUtilityBlueprint /Game/Sounds/Fort_Audio_Blueprints/Utilities/AssetActions/MetaSoundAssetActions
graphs [
 CreateFromPreset {nodes [
   n1 = CreateFromPreset!() -> (then, Preset:Meta Sound Source Object Reference) {is=title K2Node_FunctionEntry}
   > n2 = Set MSPreset(MSPreset:Meta Sound Source Object Reference=@n1) -> (Output_Get:Meta Sound Source Object Reference)
   n3 = Split!(InPath:String=Get Object Path String(Object=@n2)) -> (PathPart:String, FilenamePart:String, ExtensionPart:String)
   > n4 = Execute Python Script!(
    execute
    PythonScript="import unreal\r\n\r\nfactory = unreal.MetaSoundSourceFactory()\r\nfactory.set_editor_property(\"ReferencedMetaSoundObject\", preset)\r\n\r\nmetasound = unreal.AssetToolsHelpers.get_asset_tools().create_asset(name, path, unreal.MetaSoundSource, factory)"
    path:String=@n3
    name:String=@n3.FilenamePart
    preset:Meta Sound Source Object Reference=@n2
   ) -> (then, ReturnValue, metasound:Object Reference) {inputs [path,name,preset], outputs [metasound]}
   > Cast To MetaSoundSource!(execute, Object=@n4.metasound) -> (then, CastFailed, AsMetaSound Source:Meta Sound Source Object Reference) {is=title K2Node_DynamicCast}
  ]}
 EventGraph {links [], nodes []}
 Fix Up Duplicate References {nodes [
   Fix Up Duplicate References!() -> (then, Reference Directories:Array of Meta Sound Asset Directory Structures) {is=title K2Node_FunctionEntry}
   > n1 = GetSelectedAssets()
   > n2 = For Each Loop(Array:Array of Object References=@n1) -> (Array Element:Object Reference)
   > n3 = GetEditorWorld(self=Get UnrealEditorSubsystem())
   > Fix Up Duplicate References!(
    execute
    MetaSound Asset:Object Reference=@n2
    Reference Directories:Array of Meta Sound Asset Directory Structures=Get Reference Directories() -> (Reference Directories:Array of Meta Sound Asset Directory Structures)
    __WorldContext:Object Reference=@n3
   ) -> (then) {is=title K2Node_CallFunction}
  ]}
 UpdateMetaSoundInputs {nodes [
   n1 = UpdateMetaSoundInputs!() -> (
    then
    ParamName:Array of Names
    ValueIfBool:bool=~
    ValueIfFloat:float=~
    ValueIfInt:int
    ValueIfObj:Object Reference
    ValueIfObjArray:Array of Object References
   ) {is=title K2Node_FunctionEntry}
   > n2 = Iterate Over Selected Assets()
   > n3 = For Each Loop(Array:Array of Names=@n1) -> (Array Element:Name)
   > Set CurrentParamName(CurrentParamName:Name=@n3) -> (Output_Get:Name=None)
   > n4 = Cast To MetaSoundSource(Object=@n2)
   > n5 = Set RefMetaSoundInterface(RefMetaSoundInterface:Meta Sound Document Interface Interface=@n4) -> (Output_Get:Meta Sound Document Interface Interface)
   n6 = Get CurrentParamName() -> (CurrentParamName:Name=None)
   n7 = Find Or Begin Building MetaSound Asset!(
    self:Meta Sound Editor Subsystem Object Reference=Get MetaSoundEditorSubsystem() -> (ReturnValue:Meta Sound Editor Subsystem Object Reference) {is=title K2Node_GetEditorSubsystem}
    MetaSound:Meta Sound Document Interface Interface=Get RefMetaSoundInterface() -> (RefMetaSoundInterface:Meta Sound Document Interface Interface)
   ) -> (ReturnValue:Meta Sound Builder Base Object Reference)
   n8 = FindGraphInputNode!(execute=[@n5.then, @n26.then], self:Meta Sound Builder Base Object Reference=@n7, InputName:Name=@n6) -> (
    Succeeded:Exec
    Failed:Exec
    OutDataType:Name=None
    NodeOutputHandle:Meta Sound Node Output Handle Structure
    ReturnValue:Meta Sound Node Handle Structure
   )
   > n9 = FindNodeInputByName!(
    execute
    self:Meta Sound Builder Base Object Reference=@n7
    NodeHandle:"Meta Sound Node Handle Structure (by ref)"=@n8.ReturnValue
    InputName:Name=@n6
   ) -> (Succeeded:Exec, Failed:Exec, ReturnValue:Meta Sound Node Input Handle Structure)
   > n10 = GetNodeInputData!(execute, self:Meta Sound Builder Base Object Reference=@n7, InputHandle:"Meta Sound Node Input Handle Structure (by ref)"=@n9) -> (Succeeded:Exec, Failed:Exec, Name:Name=None, DataType:Name=None)
   n11 = Find(
    TargetMap:Map of Names to Enum_MetaSoundParameterTypes Enums=Get TMapDataType() -> (TMapDataType:Map of Names to Enum_MetaSoundParameterTypes Enums)
    Key:"Name (by ref)"=@n10.DataType
   ) -> (Value:Enum_MetaSoundParameterTypes Enum=NewEnumerator0)
   n12 = "To String (Name)"(InName=@n10.DataType)
   n13 = Select(
    Option 0:Enum_MetaSoundParameterTypes Enum=Select(Option 0=NewEnumerator3, Option 1=NewEnumerator4, Index=Contains(SearchIn=@n12, Substring=Array, bSearchFromEnd=true))
    Option 1:Enum_MetaSoundParameterTypes Enum=@n11
    Index=@n11.ReturnValue
   ) -> (ReturnValue:Enum_MetaSoundParameterTypes Enum)
   > n14 = Switch on Enum_MetaSoundParameterTypes!(execute, Selection:Enum_MetaSoundParameterTypes Enum=@n13) -> (NewEnumerator0:Exec, NewEnumerator1:Exec, NewEnumerator2:Exec, NewEnumerator3:Exec, NewEnumerator4:Exec) {is=title K2Node_SwitchEnum}
   n15 = Get MetaSoundBuilderSubsystem() -> (ReturnValue:Meta Sound Builder Subsystem Object Reference) {is=title K2Node_GetEngineSubsystem}
   > n16 = Create MetaSound Bool Literal!(execute, self:Meta Sound Builder Subsystem Object Reference=@n15, Value:bool=Get ValueIfBool() -> (ValueIfBool:bool)) -> (then, DataType:Name=None, ReturnValue:Meta Sound Literal Structure)
   n17 = Create MetaSound Int Literal!(execute=@n14.NewEnumerator2, self:Meta Sound Builder Subsystem Object Reference=@n15, Value:int=Get ValueIfInt() -> (ValueIfInt:int=0)) -> (then, DataType:Name=None, ReturnValue:Meta Sound Literal Structure)
   n18 = Create MetaSound Object Literal!(
    execute=@n14.NewEnumerator3
    self:Meta Sound Builder Subsystem Object Reference=@n15
    Value:Object Reference=Get ValueIfObj() -> (ValueIfObj:Object Reference)
   ) -> (then, ReturnValue:Meta Sound Literal Structure)
   n19 = Create MetaSound Object Array Literal!(
    execute=@n14.NewEnumerator4
    self:Meta Sound Builder Subsystem Object Reference=@n15
    Value:Array of Object References=Get ValueIfObjArray() -> (ValueIfObjArray:Array of Object References)
   ) -> (then, ReturnValue:Meta Sound Literal Structure)
   n20 = Create MetaSound Float Literal!(execute=@n14.NewEnumerator1, self:Meta Sound Builder Subsystem Object Reference=@n15, Value:float=Get ValueIfFloat() -> (ValueIfFloat:float)) -> (then, DataType:Name=None, ReturnValue:Meta Sound Literal Structure)
   n21 = SetGraphInputDefault!(
    execute=[@n16.then, @n17.then, @n18.then, @n19.then, @n20.then]
    self:Meta Sound Builder Base Object Reference=@n7
    InputName:Name=Get CurrentParamName() -> (CurrentParamName=None)
    Literal:"Meta Sound Literal Structure (by ref)"=Select!(
     NewEnumerator0:Meta Sound Literal Structure=@n16.ReturnValue
     NewEnumerator1:Meta Sound Literal Structure=@n20.ReturnValue
     NewEnumerator2:Meta Sound Literal Structure=@n17.ReturnValue
     NewEnumerator3:Meta Sound Literal Structure=@n18
     NewEnumerator4:Meta Sound Literal Structure=@n19
     Index:Enum_MetaSoundParameterTypes Enum=@n13
    ) -> (ReturnValue:Meta Sound Literal Structure)
    PageName:Name
   ) -> (Succeeded:Exec, Failed:Exec)
   n22 = GetDisplayName(Object=@n2)
   PrintString(execute=@n21.Failed, InString=Append(A=@n22, B=" : Couldn't update. Invalid Input for asset"), bPrintToScreen=false)
   n23 = SaveLoadedAsset(execute=@n21.Succeeded, AssetToSave=@n2, bOnlyIfIsDirty=false)
   > n24 = Branch(Condition=@n23)
   PrintString(execute=@n24.else, InString=Append(A=@n22, B=" : Failed to checkout and save asset"), bPrintToScreen=false)
   PrintString(execute=[@n10.Failed, @n9.Failed, @n8.Failed], InString=Append(A=@n22, B=" : Invalid Input or asset data type"), bPrintToScreen=false)
   n25 = Cast To MetaSoundPatch!(execute=@n4.CastFailed, Object=@n2) -> (then, CastFailed, AsMetaSound Patch:Meta Sound Patch Object Reference) {is=title K2Node_DynamicCast}
   > n26 = Set RefMetaSoundInterface(RefMetaSoundInterface=@n25)
   PrintString(execute=@n25.CastFailed, InString=Append(A=@n22, B=" :is not a MetaSound Source or MetaSound Patch"), bPrintToScreen=false)
  ]}
]
variables []
