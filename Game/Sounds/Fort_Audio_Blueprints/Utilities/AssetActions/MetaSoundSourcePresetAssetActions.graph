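# Editor asset action that migrates selected MetaSound Source presets to a chosen parent MetaSound Source, creating new suffixed presets.
# For each preset, it looks up saved old-to-new parent input links in the SavedMetaSoundLinks data asset.
# It builds the new preset from original literals matching linked input keys, then copies source properties.
# Editor notifications report outcomes, and failed property migration prompts whether to keep or delete the new asset.
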EditorUtilityBlueprint /Game/Sounds/Fort_Audio_Blueprints/Utilities/AssetActions/MetaSoundSourcePresetAssetActions
graphs [
 EventGraph {links [], nodes []}
 Migrate MetaSound Source Preset {nodes [
   Migrate MetaSound Source Preset!() -> (then, Parent MetaSound Source:Meta Sound Source Object Reference, Suffix:String=_Migrated) {is=title K2Node_FunctionEntry}
   > n1 = Get!(execute:Exec) -> (then:Exec, else:Exec, Saved MetaSound Input Links:Meta Sound Preset Migrator Saved Meta Sound Links Object Reference)
   > n2 = Get!(execute) -> (then, else, Parent MetaSound Source:Meta Sound Source Object Reference)
   > n3 = GetEditorWorld(self=Get UnrealEditorSubsystem())
   > n4 = For Each Loop(Array:Array of Meta Sound Source Object References=Get Selected MetaSound Source Presets!(__WorldContext:Object Reference=@n3) -> (Selected MetaSound Source Presets:Array of Meta Sound Source Object References) {is=title K2Node_CallFunction}) -> (Array Element:Meta Sound Source Object Reference)
   K2Node_Composite_5 = Notification Migrated MetaSound Source presets(execute=@n4.Completed) {is=title K2Node_Composite}
   > n5 = Set Original MetaSound Source Preset(Original MetaSound Source Preset:Meta Sound Source Object Reference=@n4) -> (Output_Get:Meta Sound Source Object Reference)
   > K2Node_Composite_6 = Validate Assets!(execute, MetaSound Source:Meta Sound Source Object Reference=@n5) -> (Valid:Exec, Invalid:Exec) {is=title K2Node_Composite}
   > n6 = GetEditorWorld(self=Get UnrealEditorSubsystem())
   n7 = Get Saved Input Links!(
    self:Meta Sound Preset Migrator Saved Meta Sound Links Object Reference=Get Saved MetaSound Input Links() -> (Saved MetaSound Input Links:Meta Sound Preset Migrator Saved Meta Sound Links Object Reference)
    Origin Parent MetaSound Source:Meta Sound Source Object Reference=Get Preset Parent!(
     MetaSound Source Preset:Meta Sound Source Object Reference=Get Original MetaSound Source Preset() -> (Original MetaSound Source Preset:Meta Sound Source Object Reference)
     __WorldContext:Object Reference=@n6
    ) -> (Parent MetaSound Source:Meta Sound Source Object Reference) {is=title K2Node_CallFunction}
    Destination MetaSound Source:Meta Sound Source Object Reference=Get Parent MetaSound Source() -> (Parent MetaSound Source:Meta Sound Source Object Reference)
   ) -> (Success:bool, Linked Inputs:Map of Names to Names) {is=title K2Node_CallFunction}
   > n8 = Branch(Condition=@n7)
   > Set Linked Inputs(Linked Inputs:Map of Names to Names=@n7.Linked Inputs) -> (Output_Get:Map of Names to Names)
   > n9 = GetEditorWorld(self=Get UnrealEditorSubsystem())
   > n10 = Keys!(execute, TargetMap:Map of Names to Names=Get Linked Inputs() -> (Linked Inputs:Map of Names to Names)) -> (then, Keys:Array of Names)
   > n11 = For Each Loop(Array:Array of Names=@n10) -> (Array Element:Name)
   n12 = Split!(InPath:String=Get Object Path String(Object=Get Original MetaSound Source Preset())) -> (PathPart:String, FilenamePart:String, ExtensionPart:String)
   n13 = Create MetaSound Preset with Literals!(
    execute=@n11.Completed
    Parent MetaSound Source:Meta Sound Source Object Reference=Get Parent MetaSound Source()
    MetaSound Input Literals:Map of Names to Meta Sound Literal Structures=Get MetaSound Input Literals() -> (MetaSound Input Literals:Map of Names to Meta Sound Literal Structures)
    Package Path:String=@n12
    Asset Name:String=@n12.FilenamePart
    Suffix:String=Get Suffix() -> (Suffix:String)
    __WorldContext:Object Reference=@n9
   ) -> (then, MetaSound Source Preset:Meta Sound Source Object Reference)
   > n14 = Set New MetaSound Preset(New MetaSound Preset:Meta Sound Source Object Reference=@n13) -> (Output_Get:Meta Sound Source Object Reference)
   > n15 = Is Valid(InputObject=@n14)
   K2Node_Composite_4 = Notification Create preset failed(execute=@n15.Is Not Valid) {is=title K2Node_Composite}
   > n16 = Migrate Source Properties!(
    execute
    Asset to Migrate From:Sound Base Object Reference=Get Original MetaSound Source Preset()
    Asset to Migrate To:Sound Base Object Reference=Get New MetaSound Preset() -> (New MetaSound Preset:Meta Sound Source Object Reference)
    __WorldContext:Object Reference=@n9
   ) -> (then, Success:bool, Results:Array of Fort Audio Tools Result Structures)
   > n17 = Branch(Condition=@n16)
   > Increment Int(Value=Get Number of Migrated Assets() -> (Number of Migrated Assets:int=0))
   Print Results!(
    execute=@n17.else
    Results:Array of Fort Audio Tools Result Structures=@n16.Results
    Exclude Successful Results:bool=true
    __WorldContext:Object Reference=@n9
   ) -> (then)
   > n18 = Get!(execute) -> (then, else, New MetaSound Preset:Meta Sound Source Object Reference)
   > K2Node_Composite_2 = Would you like to keep the newly created asset?!(execute) -> (No:Exec, Yes:Exec) {is=title K2Node_Composite}
   > DeleteLoadedAsset!(execute, AssetToDelete:Object Reference=@n18) -> (then, ReturnValue:bool)
   n19 = Find(
    TargetMap:Map of Names to Meta Sound Literal Structures=Get All MetaSound Source Input Literals!(MetaSound Source:Meta Sound Source Object Reference=Get Original MetaSound Source Preset(), __WorldContext:Object Reference=@n9) -> (Input Default Literals:Map of Names to Meta Sound Literal Structures) {is=title K2Node_CallFunction}
    Key:"Name (by ref)"=@n11
   ) -> (Value:Meta Sound Literal Structure)
   Branch(execute=@n11.LoopBody, Condition=@n19.ReturnValue)
   > Add!(
    execute
    TargetMap:Map of Names to Meta Sound Literal Structures=Get MetaSound Input Literals()
    Key:"Name (by ref)"=@n11
    Value:"Meta Sound Literal Structure (by ref)"=@n19
   ) -> (then) {is=title K2Node_CallFunction}
   K2Node_Composite_1 = Notification No Saved Links(execute=@n8.else) {is=title K2Node_Composite}
   K2Node_Composite_0 = Notification Null Parent MetaSound Source(execute=@n2.else) {is=title K2Node_Composite}
   K2Node_Composite_3 = Notification Invalid Save Data(execute=@n1.else) {is=title K2Node_Composite}
  ]}
 Notification Null Parent MetaSound Source {
  path=Migrate MetaSound Source Preset.K2Node_Composite_0.Notification Null Parent MetaSound Source
  nodes [
   Inputs()
   > n1 = AddEditorNotification!(
    execute
    EditorNotificationSettings:"Editor Notification Settings Structure (by ref)"=MakeEditorNotificationSettings!(
     Text:Text="LOCTEXT(\"Did not migrate any MetaSound Source assets.\")"
     SubText:Text=Format Text(Format="LOCTEXT(\"Parent MetaSound Source was null. Be sure to select a new parent MetaSound Source asset to migrate to.\")")
     FadeInDuration:float=0.5
     FadeOutDuration:float=2
     ExpireDuration:float=15
     bUseThrobber:bool=true
     bUseSuccessFailIcons:bool=true
     bFireAndForget:bool=true
     bAllowThrottleWhenFrameRateIsLow:bool=true
    ) -> (ReturnValue:Editor Notification Settings Structure)
   ) -> (then, ReturnValue:Editor Notification Handle Structure)
   > SetCompletionState!(execute, NotificationHandle:"Editor Notification Handle Structure (by ref)"=@n1, State:EEditorNotificationCompletionState Enum=Fail) -> (then)
   Outputs()
  ]
 }
 Notification No Saved Links {
  path=Migrate MetaSound Source Preset.K2Node_Composite_1.Notification No Saved Links
  nodes [
   Inputs()
   > n1 = GetEditorWorld(self=Get UnrealEditorSubsystem())
   n2 = GetDisplayName(Object=Get Parent MetaSound Source())
   n3 = GetDisplayName(Object=Get Saved MetaSound Input Links())
   n4 = GetDisplayName(Object=Get Preset Parent(MetaSound Source Preset=Get Original MetaSound Source Preset(), __WorldContext=@n1))
   > n5 = AddEditorNotification(EditorNotificationSettings=MakeEditorNotificationSettings(
     Text="LOCTEXT(\"Failed to migrate MetaSound Source preset.\")"
     SubText=Format Text!(
      Format="LOCTEXT(\"Use the MetaSound Preset Migrator to create and save links before using this asset action for these assets.\\r\\n\\r\\nCould not find link between assets \\'{origin}\\' and \\'{destination}\\' in data asset \\'{dataasset}\\'. \")"
      origin:String=@n4
      destination:String=@n2
      dataasset:String=@n3
     ) -> (Result)
     FadeInDuration=0.5
     FadeOutDuration=2
     ExpireDuration=15
     bUseThrobber=true
     bUseSuccessFailIcons=true
     bFireAndForget=true
     bAllowThrottleWhenFrameRateIsLow=true
    ))
   > SetCompletionState(NotificationHandle=@n5, State=Fail)
   Outputs()
  ]
 }
 Would you like to keep the newly created asset? {
  path=Migrate MetaSound Source Preset.K2Node_Composite_2.Would you like to keep the newly created asset?
  nodes [
   Inputs()
   > n1 = Show Message Dialog(
    Title=Format Text!(Format="LOCTEXT(\"Failed to migrate source properties!\")") -> (Result)
    Message=Format Text(Format="LOCTEXT(\"Successfully migrated inputs to the new preset, but failed to migrate source properties. \\r\\n\\r\\nWould you like to keep the newly created asset? Properties that failed to migrate will need to be fixed manually.\\r\\n\\r\\n\")")
    MessageType=YesNo
    DefaultValue=Yes
    MessageCategory=Warning
   )
   > n2 = Switch on EAppReturnType(Selection=@n1)
   > Outputs!(No:Exec, Yes:Exec=@n2.Yes)
  ]
 }
 Notification Invalid Save Data {
  path=Migrate MetaSound Source Preset.K2Node_Composite_3.Notification Invalid Save Data
  nodes [
   Inputs()
   > n1 = AddEditorNotification(EditorNotificationSettings=MakeEditorNotificationSettings(
     Text="LOCTEXT(\"Linked MetaSound inputs save data not found!\")"
     SubText=Format Text(Format="LOCTEXT(\"Could not locate data asset containing saved MetaSound input links!\\r\\nCheck the asset action blueprint to make sure a data asset is set.\")")
     FadeInDuration=0.5
     FadeOutDuration=2
     ExpireDuration=15
     bUseThrobber=true
     bUseSuccessFailIcons=true
     bFireAndForget=true
     bAllowThrottleWhenFrameRateIsLow=true
    ))
   > SetCompletionState(NotificationHandle=@n1, State=Fail)
   Outputs!()
  ]
 }
 Notification Create preset failed {
  path=Migrate MetaSound Source Preset.K2Node_Composite_4.Notification Create preset failed
  nodes [
   Inputs()
   n1 = Get Object Path String(Object=Get Original MetaSound Source Preset())
   > n2 = AddEditorNotification(EditorNotificationSettings=MakeEditorNotificationSettings(
     Text=Format Text!(Format="LOCTEXT(\"Failed to migrate MetaSound Source preset \\'{asset}\\'\")", asset:String=@n1) -> (Result)
     FadeInDuration=0.5
     FadeOutDuration=2
     ExpireDuration=15
     bUseThrobber=true
     bUseSuccessFailIcons=true
     bFireAndForget=true
     bAllowThrottleWhenFrameRateIsLow=true
    ))
   > SetCompletionState(NotificationHandle=@n2, State=Fail)
   Outputs()
  ]
 }
 Notification Migrated MetaSound Source presets {
  path=Migrate MetaSound Source Preset.K2Node_Composite_5.Notification Migrated MetaSound Source presets
  nodes [
   Inputs()
   n1 = Get Number of Migrated Assets() -> (Number of Migrated Assets=0)
   n2 = Select(Option 0:String, Option 1:String=s, Index="integer > integer"(A=@n1, B=1)) -> (ReturnValue:String)
   > n3 = AddEditorNotification(EditorNotificationSettings=MakeEditorNotificationSettings(
     Text=Format Text!(Format="LOCTEXT(\"Successfully migrated {num} MetaSound Source preset{s}.\")", num:int=@n1, s:String=@n2) -> (Result)
     FadeInDuration=0.5
     FadeOutDuration=2
     ExpireDuration=10
     bUseThrobber=true
     bUseSuccessFailIcons=true
     bFireAndForget=true
     bAllowThrottleWhenFrameRateIsLow=true
    ))
   > SetCompletionState(NotificationHandle=@n3, State=Success)
   Outputs()
  ]
 }
 Validate Assets {
  path=Migrate MetaSound Source Preset.K2Node_Composite_6.Validate Assets
  nodes [
   n1 = Inputs!() -> (execute, MetaSound Source:Meta Sound Source Object Reference)
   > n2 = GetEditorWorld(self=Get UnrealEditorSubsystem())
   n3 = Get Parent MetaSound Source()
   > n4 = Branch(Condition=NOT Boolean(A=Is Preset!(MetaSound Source:Meta Sound Source Object Reference=@n3, __WorldContext:Object Reference=@n2) -> (ReturnValue:bool)))
   > Outputs!(Valid:Exec, Invalid:Exec=@n7.then)
   n5 = Get Parent MetaSound Source()
   n6 = AddEditorNotification(
    execute=@n4.else
    EditorNotificationSettings=MakeEditorNotificationSettings(
     Text=Format Text!(Format="LOCTEXT(\"Asset \\'{origin}\\' was not migrated.\")", origin:Meta Sound Source Object Reference=@n1) -> (Result)
     SubText=Format Text!(Format="LOCTEXT(\"Destination asset \\'{parent}\\' should not be a preset. \")", parent:Meta Sound Source Object Reference=@n5) -> (Result)
     FadeInDuration=0.5
     FadeOutDuration=2
     ExpireDuration=15
     bUseThrobber=true
     bUseSuccessFailIcons=true
     bFireAndForget=true
     bAllowThrottleWhenFrameRateIsLow=true
    )
   )
   > n7 = SetCompletionState(NotificationHandle=@n6, State=Fail)
  ]
 }
]
variables [Saved MetaSound Input Links {
  type=MetaSoundPresetMigrator_SavedMetaSoundLinks_C
  default=/Game/Sounds/Fort_Audio_Blueprints/Utilities/EditorUtilityWidgets/Metasounds/Tools/MetaSoundPresetMigratorData/SavedMetaSoundLinks
 }]
