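# Editor asset-action utility that batch-edits audio notifies on selected animation assets.
# It generates footstep notifies from a DataTable by sampling bone poses for lowest height and detecting lift and movement.
# Other actions remove or replace notifies, add Sound Library notifies at existing or reload events, convert emote sounds, and fix or report attach bones.
# Most actions collect changed assets in ModifiedAssets for source-control checkout.
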EditorUtilityBlueprint /Game/Sounds/Fort_Audio_Blueprints/Utilities/AssetActions/SoundAnimationAssetActions
graphs [
 AddFootstepNotifies {nodes [
   n1 = AddFootstepNotifies!() -> (then, FootstepDataTable:Data Table Object Reference, bClearExistingNotifies:bool=~, NotifyFrameOffset:int) {is=title K2Node_FunctionEntry}
   > Clear(TargetArray:Array of Object References=Get ModifiedAssets() -> (ModifiedAssets:Array of Object References))
   > n2 = GetSelectedAssets()
   > n3 = For Each Loop(Array:Array of Object References=@n2) -> (Array Element:Object Reference)
   CheckoutLoadedAssets(execute=@n3.Completed, AssetsToCheckout=Get ModifiedAssets())
   > Clear(TargetArray=Get ModifiedAssets())
   n4 = Cast To AnimSequenceBase!(execute=@n3.LoopBody, Object=@n3) -> (then, CastFailed, AsAnim Sequence Base:Anim Sequence Base Object Reference) {is=title K2Node_DynamicCast}
   > n5 = GetDataTableRowNames(Table=@n1)
   > n6 = For Each Loop(Array:Array of Names=@n5) -> (Array Element:Name)
   > n7 = Get Data Table Row!(execute:Exec, DataTable:Data Table Object Reference=@n1, RowName:Name=@n6) -> (then:Exec, RowNotFound:Exec, ReturnValue:Footstep Audio Notify Struct Structure)
   > n8 = Branch(Condition=@n1.bClearExistingNotifies)
   n9 = Break Footstep Audio Notify Struct!(FootstepAudioNotify_Struct:"Footstep Audio Notify Struct Structure (by ref)"=@n7) -> (
    HeelBone_15_#1:Name
    ToeBone_16_#2:Name
    NotifyClass_21_#3:Anim Notify Class Reference
    GameplayTag_22_#4:Gameplay Tag Structure
    AnimNotifyTrackName_18_#5:Name
    FootBone_25_#6:int
   ) {is=title K2Node_BreakStruct}
   > n10 = Remove Existing Notifies!(
    execute:Exec
    AnimationSequenceBase:Anim Sequence Base Object Reference=@n4
    ParentClass:Object Class Reference=@n9.NotifyClass_21_#3
    AudioFootstepNotify_Struct:"Footstep Audio Notify Struct Structure (by ref)"=@n7
   ) -> (then:Exec) {is=title K2Node_MacroInstance}
   SetFootNotifies!(
    execute=[@n10.then, @n8.else]
    self
    Anim:Anim Sequence Base Object Reference=@n4
    HeelBone:Name=@n9
    ToeBone:Name=@n9.ToeBone_16_#2
    NewParam:Anim Notify Class Reference=@n9.NotifyClass_21_#3
    GameplayCue:Gameplay Tag Structure=@n9.GameplayTag_22_#4
    TrackName:Name=@n9.AnimNotifyTrackName_18_#5
    FrameOffset:int=@n1.NotifyFrameOffset
   ) -> (then)
  ]}
 EventGraph {nodes [
   EMote() {is=title K2Node_CustomEvent}
   > n1 = GetSelectedAssets()
   > n2 = For Each Loop(Array:Array of Object References=@n1) -> (Array Element:Object Reference)
   > n3 = Cast To AnimSequence!(execute, Object=@n2) -> (then, CastFailed, AsAnim Sequence:Anim Sequence Object Reference)
   > AddAnimationNotifyTrack!(
    execute
    AnimationSequenceBase:Anim Sequence Base Object Reference=@n3
    NotifyTrackName:Name=Emote Sound
    TrackColor:Linear Color Structure=(R=0.3543,G=0.89,B=1,A=1)
   ) -> (then)
   > n4 = For Each Loop(Array:Array of Anim Notify Event Structures=GetAnimationNotifyEvents!(AnimationSequenceBase:Anim Sequence Base Object Reference=@n3) -> (NotifyEvents:Array of Anim Notify Event Structures)) -> (Array Element:Anim Notify Event Structure)
   n5 = For Each Loop(Exec=@n4.Completed, Array:Array of Names=Get Notify Name() -> (Notify Name:Array of Names)) -> (Array Element:Name)
   > RemoveAnimationNotifyEventsByName!(execute, AnimationSequenceBase:Anim Sequence Base Object Reference=@n3, NotifyName:Name=@n5) -> (then, ReturnValue:int=0)
   n6 = Break Anim Notify Event!(AnimNotifyEvent:"Anim Notify Event Structure (by ref)"=@n4) -> (
    TriggerWeightThreshold:float=~
    NotifyName:Name
    Notify:Anim Notify Object Reference
    NotifyStateClass:Anim Notify State Object Reference
    MontageTickType:EMontageNotifyTickType Enum
    NotifyTriggerChance:float=~
    NotifyFilterType:ENotifyFilterType Enum
    NotifyFilterLOD:int
    bCanBeFilteredViaRequest:bool=~
    bTriggerOnDedicatedServer:bool=~
    bTriggerOnFollower:bool=~
   ) {is=title K2Node_BreakStruct}
   n7 = Cast To AnimNotify_PlaySound!(execute=@n4.LoopBody, Object=@n6.Notify) -> (then, CastFailed, AsPlay Sound:Play Sound Object Reference) {is=title K2Node_DynamicCast}
   > Add Unique(TargetArray:Array of Names=Get Notify Name(), NewItem:"Name (by ref)"=@n6.NotifyName)
   n8 = Get Sound!(self:Play Sound Object Reference=@n7) -> (Sound:Sound Base Object Reference)
   > n9 = AddAnimationNotifyStateEvent!(
    execute
    AnimationSequenceBase:Anim Sequence Base Object Reference=@n3
    NotifyTrackName:Name=Emote Sound
    StartTime:float=GetAnimNotifyEventTriggerTime!(NotifyEvent:"Anim Notify Event Structure (by ref)"=@n4) -> (ReturnValue:float)
    Duration:float=Get Duration!(self:Sound Base Object Reference=@n8) -> (Duration:float)
    NotifyStateClass:Anim Notify State Class Reference=/Script/FortniteGame.FortAnimNotifyState_EmoteSound
   ) -> (then, ReturnValue:Anim Notify State Object Reference)
   > n10 = Cast To FortAnimNotifyState_EmoteSound!(execute, Object=@n9) -> (then, CastFailed, AsFort Anim Notify State Emote Sound:Fort Anim Notify State Emote Sound Object Reference) {is=title K2Node_DynamicCast}
   > n11 = SetEditorProperty(Object=@n10, PropertyName=EmoteSound1P, PropertyValue:Sound Base Object Reference=@n8)
   > SetEditorProperty(Object=@n10, PropertyName=EmoteSound3P, PropertyValue=@n8)
   PrintString(execute=@n3.CastFailed, InString=Fail)
   PrintString(InString="To String (Boolean)"(InBool=@n11))
  ]}
 FindMissingSoundLibAttachBones {nodes [
   FindMissingSoundLibAttachBones() {is=title K2Node_FunctionEntry}
   > n1 = GetSelectedAssets()
   > n2 = For Each Loop(Array:Array of Object References=@n1) -> (Array Element:Object Reference)
   > n3 = Cast To AnimSequenceBase(Object=@n2)
   > PrintMissingAttachNames!(execute, self, AnimationSequence:Anim Sequence Base Object Reference=@n3) -> (then)
  ]}
 GetIsFootBoneInAir {nodes [
   n1 = Inputs!() -> (CurrentPosition:Vector=~, LastPosition:Vector=~, BoneMap:"Map of Names to Float (double-precision)s", TrackingBone:Name)
   n2 = Find(TargetMap:"Map of Names to Float (double-precision)s"=@n1.BoneMap, Key:"Name (by ref)"=@n1.TrackingBone) -> (Value:double)
   n3 = BreakVector(InVec=@n1)
   n4 = float - float(A=@n3.Z, B=@n2)
   "Nearly Equal (Float)"(A=@n4, B=0, ErrorTolerance=1)
   "float <= float"(A=@n4) -> (ReturnValue=~)
   "float >= float"(A=@n4)
   n5 = float / float(A=float + float(A=@n3.Z, B=BreakVector(InVec=@n1.LastPosition).Z), B=2)
   Outputs!(ReturnValue:bool=NOR Boolean!(
     A:bool="Nearly Equal (Float)"(A=@n5, B=@n2, ErrorTolerance=1)
     B:bool=Vector_IsZero!(A:"Vector (by ref)"=@n1.LastPosition) -> (ReturnValue:bool)
    ) -> (ReturnValue:bool))
  ]}
 GetIsFootBoneMoving {nodes [
   n1 = Inputs!() -> (CurrentPosition:"Vector (by ref)", LastPosition=~)
   n2 = Vector_IsZero(A=@n1.LastPosition)
   n3 = Vector Length(A=vector - vector(A=@n1, B=@n1.LastPosition))
   Outputs(ReturnValue=AND Boolean(A="float > float"(A=@n3, B=1), B=NOT Boolean(A=@n2)))
  ]}
 GetReloadFXNotifyCount {nodes [
   n1 = GetReloadFXNotifyCount!() -> (then, TargetArray:Array of Anim Notify Event Structures) {is=title K2Node_FunctionEntry}
   > n2 = For Each Loop(Array:Array of Anim Notify Event Structures=@n1) -> (Array Element:Anim Notify Event Structure)
   Return Node!(execute=@n2.Completed, Count:int=Get TotalCount() -> (TotalCount:int=0))
   n3 = Break Anim Notify Event!(AnimNotifyEvent=@n2) -> (Notify, bCanBeFilteredViaRequest=~)
   > Branch(Condition=OR Boolean(
     A=Cast To AnimNotify_PlayReloadFX!(Object=@n3) -> (AsAnim Notify Play Reload FX:Anim Notify Play Reload FX Object Reference, bSuccess:bool=~) {is=title K2Node_DynamicCast}.bSuccess
     B=Cast To FortAnimNotify_PlayReloadFX!(Object=@n3) -> (AsPlay Reload FX:Play Reload FX Object Reference, bSuccess:bool=~) {is=title K2Node_DynamicCast}.bSuccess
    ))
   > Increment Int(Value=Get TotalCount() -> (TotalCount=0))
  ]}
 PrintMissingAttachNames {nodes [
   n1 = PrintMissingAttachNames!() -> (then, AnimationSequence:Anim Sequence Base Object Reference) {is=title K2Node_FunctionEntry}
   > n2 = For Each Loop(Array=GetAnimationNotifyEvents(AnimationSequenceBase=@n1))
   Branch(execute=@n2.Completed, Condition=Get bIsMissingBones() -> (bIsMissingBones:bool))
   > PrintString(InString=Append!(A="\r\n", B=GetDisplayName(Object=@n1), C:String=" is missing bones: ", D:String=Get MissingSocketNames() -> (MissingSocketNames:String)) -> (ReturnValue))
   n3 = Cast To AnimNotify_SoundLibrary!(
    execute=@n2.LoopBody
    Object=Break Anim Notify Event!(AnimNotifyEvent=@n2) -> (
     TriggerWeightThreshold:float=~
     NotifyName:Name
     Notify
     NotifyStateClass:Anim Notify State Object Reference
     MontageTickType:EMontageNotifyTickType Enum
     NotifyTriggerChance:float=~
     NotifyFilterType:ENotifyFilterType Enum
     NotifyFilterLOD:int
     bCanBeFilteredViaRequest=~
     bTriggerOnDedicatedServer:bool=~
     bTriggerOnFollower:bool=~
    ).Notify
   ) -> (then, CastFailed, AsAnim Notify Sound Library:Anim Notify Sound Library Object Reference) {is=title K2Node_DynamicCast}
   > n4 = GetEditorProperty!(execute:Exec, Object:Object Reference=@n3, PropertyName:Name=SoundContext) -> (then:Exec, PropertyValue:Animation Notify Context Object Reference, ReturnValue:bool)
   > Branch(Condition=@n4.ReturnValue)
   n5 = GetReferencePose!(Skeleton:Skeleton Object Reference=GetSkeleton!(InAsset:Animation Asset Object Reference=@n1) -> (ReturnValue:Skeleton Object Reference)) -> (OutPose:Anim Pose Structure)
   n6 = Get!(self:Animation Notify Context Object Reference=@n4) -> (
    AnimSettings_AnimRateThreshold:Float Interval Structure
    AnimSettings_EventName:Gameplay Tag Structure
    AnimSettings_bOnThisActor:bool=~
    AnimSettings_bAttachToActor:bool=~
    AnimSettings_bStopWhenAttachedDestroyed:bool=~
    AnimSettings_Offset:Vector=~
    AnimSettings_AttachName:Name
    AnimSettings_FadeTime:float=~
    AnimSettings_NotifyTriggerChance:float=~
    AnimSettings_VolumeMultiplier:float=~
   )
   > n7 = Branch(Condition=Contains Item(TargetArray=GetBoneNames!(Pose:"Anim Pose Structure (by ref)"=@n5) -> (Bones:Array of Names), ItemToFind=@n6.AnimSettings_AttachName))
   Set bIsMissingBones(execute=@n7.else, bIsMissingBones:bool=true) -> (Output_Get:bool)
   > Set MissingSocketNames(MissingSocketNames:String=Append!(A=Get MissingSocketNames(), B="\r\n", C="To String (Name)"(InName=@n6.AnimSettings_AttachName)) -> (ReturnValue)) -> (Output_Get:String)
   ConfigureContext!(
    execute
    self:Animation Notify Context Object Reference=@n4
    OwningActor:Actor Object Reference
    InAnimation:Anim Sequence Base Object Reference=@n1
    InSettings_AnimRateThreshold:Float Interval Structure=(Min=0,Max=0)
    InSettings_EventName:Gameplay Tag Structure=(TagName="")
    InSettings_bOnThisActor:bool=False
    InSettings_bAttachToActor:bool=False
    InSettings_bStopWhenAttachedDestroyed:bool=False
    InSettings_Offset:Vector="0,0,0"
    InSettings_AttachName:Name=None
    InSettings_FadeTime:float=0.5
    InSettings_NotifyTriggerChance:float=1
    InSettings_VolumeMultiplier:float=1
   ) -> (then)
  ]}
 RemoveAnimNotifiesOfType {nodes [
   n1 = RemoveAnimNotifiesOfType!() -> (then, NotifyClass:Anim Notify Class Reference) {is=title K2Node_FunctionEntry}
   > Clear(TargetArray=Get ModifiedAssets())
   > n2 = GetSelectedAssets()
   > n3 = For Each Loop(Array:Array of Object References=@n2) -> (Array Element:Object Reference)
   CheckoutLoadedAssets(execute=@n3.Completed, AssetsToCheckout=Get ModifiedAssets())
   > Clear(TargetArray=Get ModifiedAssets())
   n4 = Cast To AnimSequenceBase(execute=@n3.LoopBody, Object=@n3)
   n5 = GetAnimationNotifyEvents(AnimationSequenceBase=@n4)
   > Branch(Condition=Is Not Empty!(TargetArray:Array of Anim Notify Event Structures=@n5) -> (ReturnValue:bool))
   > n6 = For Each Loop with Break!(Exec:Exec, Array:Array of Anim Notify Event Structures=@n5, Break:Exec=@n8.then) -> (LoopBody:Exec, Array Element:Anim Notify Event Structure, Array Index:int, Completed:Exec)
   n7 = Break Anim Notify Event(AnimNotifyEvent=@n6) -> (TriggerWeightThreshold=~, NotifyTriggerChance=~, bCanBeFilteredViaRequest=~, bTriggerOnDedicatedServer=~, bTriggerOnFollower=~)
   > Branch(Condition=ClassIsChildOf(TestClass=Get Class(Object=@n7.Notify) -> (ReturnValue:Anim Notify Class Reference), ParentClass=@n1))
   > RemoveAnimationNotifyEventsByName(AnimationSequenceBase=@n4, NotifyName=@n7.NotifyName) -> (ReturnValue=0)
   > n8 = Add(TargetArray:Array of Object References=Get ModifiedAssets(), NewItem:"Object Reference (by ref)"=@n4) {is=title K2Node_CallArrayFunction}
  ]}
 RemoveExistingNotifies {nodes [
   n1 = Inputs!() -> (
    execute:Exec
    AnimationSequenceBase:Anim Sequence Base Object Reference
    ParentClass:Object Class Reference
    AudioFootstepNotify_Struct:"Footstep Audio Notify Struct Structure (by ref)"
   )
   > Set bTrackContainsOutsideNotify(bTrackContainsOutsideNotify:bool=false) -> (Output_Get:bool)
   n2 = Break Footstep Audio Notify Struct(FootstepAudioNotify_Struct=@n1.AudioFootstepNotify_Struct)
   n3 = GetAnimationNotifyEventsForTrack!(
    AnimationSequenceBase:Anim Sequence Base Object Reference=@n1
    NotifyTrackName:Name=@n2.AnimNotifyTrackName_18_#5
   ) -> (Events:Array of Anim Notify Event Structures)
   > n4 = Branch(Condition=Is Empty!(TargetArray:Array of Anim Notify Event Structures=@n3) -> (ReturnValue:bool))
   Outputs!(then:Exec=[@n4.then, @n7.then])
   n5 = For Each Loop(Exec=@n4.else, Array:Array of Anim Notify Event Structures=@n3) -> (Array Element:Anim Notify Event Structure)
   n6 = Branch(execute=@n5.Completed, Condition=Get bTrackContainsOutsideNotify() -> (bTrackContainsOutsideNotify:bool))
   > PrintString(InString=Append!(
     A="Track: "
     B="To String (Name)"(InName=@n2.AnimNotifyTrackName_18_#5)
     C=" contains notifes other than notify class "
     D:String=Get Class Display Name(Class=@n1.ParentClass)
     E:String=. Cannot remove previous notifies from track. Manually remove them instead.
    ) -> (ReturnValue))
   n7 = RemoveAnimationNotifyEventsByTrack!(
    execute=@n6.else
    AnimationSequenceBase:Anim Sequence Base Object Reference=@n1
    NotifyTrackName:Name=@n2.AnimNotifyTrackName_18_#5
   ) -> (then, ReturnValue:int=0)
   n8 = Break Anim Notify Event(AnimNotifyEvent=@n5) -> (TriggerWeightThreshold=~, NotifyTriggerChance=~, bCanBeFilteredViaRequest=~, bTriggerOnDedicatedServer=~, bTriggerOnFollower=~)
   n9 = Branch(execute=@n5.LoopBody, Condition=ClassIsChildOf(TestClass=Get Class(Object=@n8.Notify), ParentClass=@n1.ParentClass))
   Set bTrackContainsOutsideNotify(execute=@n9.else, bTrackContainsOutsideNotify=true)
  ]}
 ReplaceAnimNotifiesOfType {nodes [
   n1 = ReplaceAnimNotifiesOfType!() -> (then, NotifyClass:Anim Notify Class Reference, NewNotify:Anim Notify Class Reference) {is=title K2Node_FunctionEntry}
   > Clear(TargetArray=Get ModifiedAssets())
   > n2 = GetSelectedAssets()
   > n3 = For Each Loop(Array:Array of Object References=@n2) -> (Array Element:Object Reference)
   CheckoutLoadedAssets(execute=@n3.Completed, AssetsToCheckout=Get ModifiedAssets())
   > Clear(TargetArray=Get ModifiedAssets())
   n4 = Cast To AnimSequenceBase(execute=@n3.LoopBody, Object=@n3)
   n5 = GetAnimationNotifyEvents(AnimationSequenceBase=@n4)
   > Branch(Condition=Is Not Empty(TargetArray=@n5))
   > n6 = For Each Loop with Break(Array=@n5, Break=@n8.then)
   Branch(execute=@n6.Completed, Condition=Get bNotifyFound() -> (bNotifyFound:bool))
   > Add(TargetArray=Get ModifiedAssets(), NewItem=@n4)
   n7 = Break Anim Notify Event(AnimNotifyEvent=@n6) -> (TriggerWeightThreshold=~, NotifyTriggerChance=~, bCanBeFilteredViaRequest=~, bTriggerOnDedicatedServer=~, bTriggerOnFollower=~)
   Branch(execute=@n6.LoopBody, Condition="Equal (Class)"(A=Get Class(Object=@n7.Notify), B=@n1))
   > Set bNotifyFound(bNotifyFound:bool=true) -> (Output_Get:bool)
   > n8 = ReplaceAnimNotifies!(
    execute
    AnimationSequenceBase:Anim Sequence Base Object Reference=@n4
    OldNotifyClass:Anim Notify Class Reference=@n1
    NewNotifyClass:Anim Notify Class Reference=@n1.NewNotify
    OnNotifyReplaced:Delegate
   ) -> (then)
  ]}
 ReplaceAttachNames {nodes [
   n1 = ReplaceAttachNames!() -> (then, AnimationSequence:Anim Sequence Base Object Reference, OldName:Name, NewName:Name) {is=title K2Node_FunctionEntry}
   > n2 = For Each Loop(Array:Array of Anim Notify Event Structures=GetAnimationNotifyEvents(AnimationSequenceBase=@n1)) -> (Array Element:Anim Notify Event Structure)
   > n3 = Cast To AnimNotify_SoundLibrary(Object=Break Anim Notify Event(AnimNotifyEvent=@n2) -> (TriggerWeightThreshold=~, NotifyTriggerChance=~, bCanBeFilteredViaRequest=~, bTriggerOnDedicatedServer=~, bTriggerOnFollower=~).Notify)
   > n4 = GetEditorProperty(Object=@n3, PropertyName=SoundContext)
   n5 = Get AnimSettings!(self:Animation Notify Context Object Reference=@n4) -> (AnimSettings:Animation Notify Settings Structure)
   > Branch(Condition="Equal (Name)"(
     A=Break Animation Notify Settings!(SoundLibraryAnimContextSettings:"Animation Notify Settings Structure (by ref)"=@n5) -> (
      AnimRateThreshold:Float Interval Structure
      EventName:Gameplay Tag Structure
      bOnThisActor:bool=~
      bAttachToActor:bool=~
      bStopWhenAttachedDestroyed:bool=~
      Offset:Vector=~
      AttachName:Name
      FadeTime:float=~
      NotifyTriggerChance:float=~
      VolumeMultiplier:float=~
     ) {is=title K2Node_BreakStruct}.AttachName
     B=@n1.OldName
    ))
   > n6 = Set members in Animation Notify Settings!(execute:Exec, StructRef:"Animation Notify Settings Structure (by ref)"=@n5, AttachName:Name=@n1.NewName) -> (then:Exec, StructOut:"Animation Notify Settings Structure (by ref)") {is=title K2Node_SetFieldsInStruct}
   > Set AnimSettings!(execute, AnimSettings:Animation Notify Settings Structure=@n6, self:Animation Notify Context Object Reference=@n4) -> (then, Output_Get:Animation Notify Settings Structure)
   > Add(TargetArray=Get ModifiedAssets(), NewItem=@n1)
  ]}
 ReplaceBoneNames {nodes [
   ReplaceBoneNames!() -> (then, OldBone:Name, NewBone:Name) {is=title K2Node_FunctionEntry}
   > n1 = GetSelectedAssets()
   > n2 = For Each Loop(Array:Array of Object References=@n1) -> (Array Element:Object Reference)
   CheckoutLoadedAssets(execute=@n2.Completed, AssetsToCheckout=Get ModifiedAssets())
   > Clear(TargetArray=Get ModifiedAssets())
   n3 = Cast To AnimSequenceBase(execute=@n2.LoopBody, Object=@n2)
   > ReplaceAttachNames!(
    execute
    self
    AnimationSequence:Anim Sequence Base Object Reference=@n3
    OldName:Name=Reroute Node(InputPin:Name) -> (OutputPin:Name)
    NewName:Name=Reroute Node()
   ) -> (then) {is=title K2Node_CallFunction}
  ]}
 SetAttachSoundLibraryNotifies {nodes [
   n1 = SetAttachSoundLibraryNotifies!() -> (then, TagsToAffect:Gameplay Tag Container Structure, bAttach:bool=~, bExactMatches:bool=~) {is=title K2Node_FunctionEntry}
   > Clear(TargetArray=Get ModifiedAssets())
   > n2 = GetSelectedAssets()
   > n3 = For Each Loop(Array=@n2)
   CheckoutLoadedAssets(execute=@n3.Completed, AssetsToCheckout=Get ModifiedAssets())
   n4 = Cast To AnimSequenceBase(execute=@n3.LoopBody, Object=@n3)
   > n5 = For Each Loop(Array:Array of Anim Notify Event Structures=GetAnimationNotifyEvents(AnimationSequenceBase=@n4)) -> (Array Element:Anim Notify Event Structure)
   > n6 = Cast To AnimNotify_SoundLibrary(Object=Break Anim Notify Event(AnimNotifyEvent=@n5) -> (TriggerWeightThreshold=~, NotifyTriggerChance=~, bCanBeFilteredViaRequest=~, bTriggerOnDedicatedServer=~, bTriggerOnFollower=~).Notify)
   > n7 = GetEditorProperty(Object=@n6, PropertyName=SoundContext)
   > n8 = For Each Loop(Array:Array of Gameplay Tag Structures=BreakGameplayTagContainer!(GameplayTagContainer:"Gameplay Tag Container Structure (by ref)"=@n1) -> (GameplayTags:Array of Gameplay Tag Structures)) -> (Array Element:Gameplay Tag Structure)
   n9 = "To String (Name)"(InName=GetTagName!(GameplayTag:"Gameplay Tag Structure (by ref)"=@n8) -> (ReturnValue:Name=None))
   n10 = "To String (Name)"(InName=GetTagName(GameplayTag=@n8) -> (ReturnValue=None))
   n11 = Contains(SearchIn=@n9, Substring=@n10)
   n12 = "Equal Exactly (String)"(A=@n9, B=@n10)
   n13 = Get AnimSettings(self=@n7)
   n14 = Break Animation Notify Settings(SoundLibraryAnimContextSettings=@n13) -> (bOnThisActor=~, bAttachToActor=~, bStopWhenAttachedDestroyed=~, Offset=~, FadeTime=~, NotifyTriggerChance=~, VolumeMultiplier=~)
   > Branch(Condition=AND Boolean(
     A=Select(Option 0:bool=@n11, Option 1:bool=@n12, Index=@n1.bExactMatches) -> (ReturnValue:bool)
     B="Not Equal (Boolean)"(A=@n14.bAttachToActor, B=@n1.bAttach) -> (ReturnValue=~)
    ))
   > n15 = Set members in Animation Notify Settings!(execute, StructRef=@n13, bAttachToActor:bool=@n1.bAttach) -> (then, StructOut)
   > Set AnimSettings(AnimSettings=@n15, self=@n7)
   > Add(TargetArray=Get ModifiedAssets(), NewItem=@n4)
  ]}
 SetFootNotifies {nodes [
   n1 = SetFootNotifies!() -> (
    then
    Anim:Anim Sequence Base Object Reference
    HeelBone:Name
    ToeBone:Name
    NewParam:Anim Notify Class Reference
    GameplayCue:Gameplay Tag Structure
    TrackName:Name
    FrameOffset:int
   ) {is=title K2Node_FunctionEntry}
   > Set AnimRef(AnimRef:Anim Sequence Base Object Reference=@n1) -> (Output_Get:Anim Sequence Base Object Reference)
   > Add(TargetArray:Array of Names=Get BoneNames() -> (BoneNames:Array of Names), NewItem:"Name (by ref)"=@n1.HeelBone)
   > n2 = Branch(Condition="Not Equal (Name)"(A=@n1.ToeBone) -> (ReturnValue=~))
   > Set bUseSupportBone(bUseSupportBone:bool=true) -> (Output_Get:bool)
   > n3 = Add(TargetArray=Get BoneNames(), NewItem=@n1.ToeBone)
   Set NotifyClass(execute=[@n3.then, @n2.else], NotifyClass:Anim Notify Class Reference=@n1.NewParam) -> (Output_Get:Anim Notify Class Reference)
   > Set GameplayTag_Local(GameplayTag_Local:Gameplay Tag Structure=@n1.GameplayCue) -> (Output_Get:Gameplay Tag Structure)
   > Set TrackName_Local(TrackName_Local:Name=@n1.TrackName) -> (Output_Get:Name=None)
   > Set FrameOffset_Local(FrameOffset_Local:int=@n1.FrameOffset) -> (Output_Get:int=0)
   > n4 = For Each Loop(Array:Array of Names=Get BoneNames()) -> (Array Element:Name)
   n5 = For Loop(
    execute=@n4.Completed
    FirstIndex=1
    LastIndex=GetNumFrames!(AnimationSequenceBase:Anim Sequence Base Object Reference=Get AnimRef() -> (AnimRef:Anim Sequence Base Object Reference)) -> (NumFrames:int=0)
   )
   > Set CurrentFrame(CurrentFrame:int=@n5) -> (Output_Get:int=0)
   > n6 = GetAnimPoseAtFrame!(
    execute
    AnimationSequenceBase:Anim Sequence Base Object Reference=Get AnimRef()
    FrameIndex:int=Get CurrentFrame() -> (CurrentFrame:int=0)
    EvaluationOptions:Anim Pose Evaluation Options Structure=Make Anim Pose Evaluation Options!(
     EvaluationType:EAnimDataEvalType Enum=Raw
     bShouldRetarget:bool=True
     bExtractRootMotion:bool=False
     bIncorporateRootMotionIntoPose:bool=True
     OptionalSkeletalMesh:Skeletal Mesh Object Reference
     bRetrieveAdditiveAsFullPose:bool=True
     bEvaluateCurves:bool=True
    ) -> (AnimPoseEvaluationOptions:Anim Pose Evaluation Options Structure) {is=title K2Node_MakeStruct}
   ) -> (then, Pose:Anim Pose Structure)
   > Set TrackingPose(TrackingPose:Anim Pose Structure=@n6) -> (Output_Get:Anim Pose Structure)
   > n7 = Keys!(execute, TargetMap:"Map of Names to Float (double-precision)s"=Get BoneMap() -> (BoneMap:"Map of Names to Float (double-precision)s")) -> (then, Keys:Array of Names)
   > n8 = For Each Loop(Array=@n7)
   n9 = Get bHeelHasMoved() -> (bHeelHasMoved:bool)
   n10 = AND Boolean(A=@n9, B=Get bToeHasMoved() -> (bToeHasMoved:bool))
   n11 = Get bUseSupportBone() -> (bUseSupportBone:bool)
   n12 = Get bHeelIsLifted() -> (bHeelIsLifted:bool)
   n13 = AND Boolean(A=@n12, B=Get bToeIsLifted() -> (bToeIsLifted:bool))
   n14 = Get bUseSupportBone()
   n15 = Branch(
    execute=@n8.Completed
    Condition=AND Boolean(A=Select(Option 0=@n12, Option 1=@n13, Index=@n14), B=Select(Option 0=@n9, Option 1=@n10, Index=@n11))
   )
   > n16 = Branch(Condition=Get bFootMoving() -> (bFootMoving:bool))
   Set bFootMoving(execute=@n16.else, bFootMoving:bool=true) -> (Output_Get:bool)
   Branch(execute=@n15.else, Condition=Get bFootMoving())
   > Set bFootMoving(bFootMoving=false)
   n17 = Get AnimRef()
   n18 = Get TrackName_Local() -> (TrackName_Local:Name=None)
   > n19 = Branch(Condition=Contains Item(
     TargetArray=GetAnimationNotifyTrackNames!(AnimationSequenceBase:Anim Sequence Base Object Reference=@n17) -> (TrackNames:Array of Names)
     ItemToFind=@n18
    ))
   n20 = Cast To AnimNotify_GameplayCue Class!(execute:Exec=[@n32.then, @n19.then], Class:Object Class Reference=Get NotifyClass() -> (NotifyClass:Anim Notify Class Reference)) -> (then:Exec, CastFailed:Exec, "AsGameplayCue (Burst)":"Gameplay Cue (Burst) Class Reference") {is=title K2Node_ClassDynamicCast}
   > PrintString(InString=Error trying to create ANIMNOTIFY_GAMEPLAYCUE notify without a tag. Creating FORTANIMNOTIFY_GAMEPLAYCUE instead)
   > n21 = Set NotifyClass(NotifyClass=/Script/FortniteGame.FortAnimNotify_GameplayCue)
   n22 = int - int(A=Get CurrentFrame() -> (CurrentFrame=0))
   n23 = Get FrameOffset_Local() -> (FrameOffset_Local:int=0)
   n24 = AddAnimationNotifyEvent!(
    execute=[@n21.then, @n20.CastFailed]
    AnimationSequenceBase:Anim Sequence Base Object Reference=@n17
    NotifyTrackName:Name=@n18
    StartTime:float=GetTimeAtFrame!(AnimationSequenceBase:Anim Sequence Base Object Reference=Get AnimRef(), Frame:int=int + int(A=@n22, B=@n23)) -> (Time:float)
    NotifyClass:Anim Notify Class Reference=@n21
   ) -> (then, ReturnValue:Anim Notify Object Reference)
   n25 = Get ModifiedAssets()
   > n26 = Branch(Condition=Contains Item(TargetArray:Array of Object References=@n25, ItemToFind:"Object Reference (by ref)"=Get AnimRef()))
   n27 = Cast To FortAnimNotify_GameplayCue!(execute=[@n31.then, @n26.then], Object=@n24) -> (then, CastFailed, "AsGameplay Cue (Burst)":"Gameplay Cue (Burst) Object Reference") {is=title K2Node_DynamicCast}
   n28 = Get GameplayTag_Local() -> (GameplayTag_Local:Gameplay Tag Structure)
   > SetEditorProperty(
    Object=@n27
    PropertyName=Cue
    PropertyValue:Gameplay Cue Tag Structure=Make Gameplay Cue Tag!(GameplayCueTag:Gameplay Tag Structure=@n28) -> (GameplayCueTag:Gameplay Cue Tag Structure) {is=title K2Node_MakeStruct}
   )
   n29 = Cast To AnimNotify_SoundLibrary(execute=@n27.CastFailed, Object=@n24)
   > n30 = Construct Object from Class!(
    execute
    Class=/Script/SoundLibrary.SoundLibrarySimpleContext
    self=@n29
    Settings:Simple Sound Settings Structure=Make Simple Sound Settings!(
     EventName:Gameplay Tag Structure=@n28
     bAttachedSound:bool=False
     bStopWhenAttachedDestroyed:bool=False
     bAutoDestroy:bool=True
     bForceComponentCreation:bool=False
     VolumeMultiplier:float=1
     PitchMultiplier:float=1
     StartTime:float=0
     ConcurrencySettings:Sound Concurrency Object Reference
     AttenuationSettings:Sound Attenuation Object Reference
     AttachName:Name=None
     AttachToComponent:Scene Component Weak object
     LocationType:EAttachLocation Enum=KeepRelativeOffset
     Location:Vector="0,0,0"
    ) -> (SoundLibrarySimpleContextSettings:Simple Sound Settings Structure) {is=title K2Node_MakeStruct}
    Actor:Actor Weak object
    BaseEventName:Gameplay Tag Structure
    bForwardToActorOwner:bool=True
    bForwardToAttachedActors:bool=True
    bAutoResetContext:bool=True
   ) -> (then, ReturnValue:Simple Sound Context Object Reference)
   > SetEditorProperty(Object=@n29, PropertyName=SoundContext, PropertyValue:Simple Sound Context Object Reference=@n30)
   n31 = Add(execute=@n26.else, TargetArray:Array of Object References=@n25, NewItem:"Object Reference (by ref)"=Get AnimRef())
   n32 = AddAnimationNotifyTrack(execute=@n19.else, AnimationSequenceBase=@n17, NotifyTrackName=@n18, TrackColor=(R=1,G=1,B=1,A=1))
   Set TrackingBone(execute=@n8.LoopBody, TrackingBone:Name="Get (a ref)"(Array:Array of Names=@n7, Dimension 1=@n8.Array Index) -> (Output:"Name (by ref)")) -> (Output_Get:Name=None)
   > n33 = Switch on Int!(execute:Exec, Selection:int=@n8.Array Index) -> (Default:Exec, 0:Exec, 1:Exec)
   n34 = Get TrackingBone() -> (TrackingBone:Name=None)
   n35 = Get TrackingPose() -> (TrackingPose:Anim Pose Structure)
   Set CurrentHeelPosition(
    execute=@n33.0
    CurrentHeelPosition:Vector=BreakTransform(InTransform=GetBonePose!(Pose:"Anim Pose Structure (by ref)"=@n35, BoneName:Name=@n34, Space:EAnimPoseSpaces Enum=World) -> (ReturnValue:"Transform (by ref)"))
   ) -> (Output_Get:Vector)
   > n36 = Branch(Condition="Equal (Vector)"!(
     A:Vector=Get CurrentHeelPosition() -> (CurrentHeelPosition:Vector)
     B:Vector=Get LastHeelPosition() -> (LastHeelPosition:Vector)
     ErrorTolerance:float
    ) -> (ReturnValue:bool=~) {is=title K2Node_PromotableOperator})
   Set bHeelIsLifted(
    execute=@n36.else
    bHeelIsLifted:bool=Get Is Foot Bone in Air!(
     CurrentPosition:Vector=Get CurrentHeelPosition()
     LastPosition:Vector=Get LastHeelPosition()
     BoneMap:"Map of Names to Float (double-precision)s"=Get BoneMap()
     TrackingBone:Name=Get TrackingBone() -> (TrackingBone=None)
    ) -> (ReturnValue:bool=~) {is=title K2Node_MacroInstance}
   ) -> (Output_Get:bool)
   > Set bHeelHasMoved(bHeelHasMoved:bool=Get Is Foot Bone Moving!(CurrentPosition:"Vector (by ref)"=Get CurrentHeelPosition(), LastPosition:Vector=Get LastHeelPosition()) -> (ReturnValue:bool=~) {is=title K2Node_MacroInstance}) -> (Output_Get:bool)
   > Set LastHeelPosition(LastHeelPosition:Vector=Get CurrentHeelPosition()) -> (Output_Get:Vector)
   n37 = Get TrackingBone() -> (TrackingBone=None)
   n38 = Get TrackingPose()
   Set CurrentToePosition(execute=@n33.1, CurrentToePosition:Vector=BreakTransform(InTransform=GetBonePose(Pose=@n38, BoneName=@n37, Space=World))) -> (Output_Get:Vector)
   > Branch(Condition="Equal (Vector)"(A=Get CurrentToePosition() -> (CurrentToePosition:Vector), B=Get LastToePosition() -> (LastToePosition:Vector)) -> (ReturnValue=~))
   > Set bToeIsLifted(bToeIsLifted:bool=Get Is Foot Bone in Air(
     CurrentPosition=Get CurrentToePosition()
     LastPosition=Get LastToePosition()
     BoneMap=Get BoneMap()
     TrackingBone=Get TrackingBone() -> (TrackingBone=None)
    ) -> (ReturnValue=~)) -> (Output_Get:bool)
   > Set bToeHasMoved(bToeHasMoved:bool=Get Is Foot Bone Moving(CurrentPosition=Get CurrentToePosition(), LastPosition=Get LastToePosition()) -> (ReturnValue=~)) -> (Output_Get:bool)
   > Set LastToePosition(LastToePosition:Vector=Get CurrentToePosition()) -> (Output_Get:Vector)
   Set LowestBoneHeight(execute=@n4.LoopBody, LowestBoneHeight:double=0) -> (Output_Get:double)
   > n39 = For Loop(FirstIndex=1, LastIndex=GetNumFrames(AnimationSequenceBase=Get AnimRef()) -> (NumFrames=0))
   Add!(
    execute=@n39.Completed
    TargetMap:"Map of Names to Float (double-precision)s"=Get BoneMap()
    Key:"Name (by ref)"=@n4
    Value:"Float (double-precision) (by ref)"=Get LowestBoneHeight() -> (LowestBoneHeight:double)
   ) -> (then) {is=title K2Node_CallFunction}
   n40 = GetAnimPoseAtFrame(
    execute=@n39.LoopBody
    AnimationSequenceBase=Get AnimRef()
    FrameIndex=@n39
    EvaluationOptions=Make Anim Pose Evaluation Options(
     EvaluationType=Raw
     bShouldRetarget=True
     bExtractRootMotion=False
     bIncorporateRootMotionIntoPose=True
     bRetrieveAdditiveAsFullPose=True
     bEvaluateCurves=True
    )
   )
   n41 = Get LowestBoneHeight()
   > n42 = Branch(Condition="float <= float"(A=@n41) -> (ReturnValue=~))
   > Set LowestBoneHeight(LowestBoneHeight=0)
   n43 = BreakVector(InVec=BreakTransform(InTransform=GetBonePose(Pose=@n40, BoneName=@n4, Space=World)))
   Branch(execute=@n42.else, Condition="float < float"(A=@n43.Z, B=@n41))
   > Set LowestBoneHeight(LowestBoneHeight=@n43.Z)
  ]}
 SoundLib_AddNotifyAtEvent {nodes [
   n1 = SoundLib_AddNotifyAtEvent!() -> (then, TargetNotify:Anim Notify Class Reference, SoundLibTag:Gameplay Tag Structure, TrackName:Name) {is=title K2Node_FunctionEntry}
   > Clear(TargetArray=Get ModifiedAssets())
   > n2 = GetSelectedAssets()
   > n3 = For Each Loop(Array:Array of Object References=@n2) -> (Array Element:Object Reference)
   CheckoutLoadedAssets(execute=@n3.Completed, AssetsToCheckout=Get ModifiedAssets())
   > Clear(TargetArray=Get ModifiedAssets())
   n4 = Cast To AnimSequenceBase(execute=@n3.LoopBody, Object=@n3)
   n5 = GetAnimationNotifyEvents(AnimationSequenceBase=@n4)
   > Branch(Condition=Is Not Empty(TargetArray=@n5))
   > n6 = Branch(Condition=IsValidAnimNotifyTrackName!(AnimationSequenceBase:Anim Sequence Base Object Reference=@n4, NotifyTrackName:Name=@n1.TrackName) -> (ReturnValue:bool))
   n7 = For Each Loop(Exec=[@n14.then, @n6.then], Array:Array of Anim Notify Event Structures=@n5) -> (Array Element:Anim Notify Event Structure)
   n8 = Break Anim Notify Event(AnimNotifyEvent=@n7) -> (TriggerWeightThreshold=~, NotifyTriggerChance=~, bCanBeFilteredViaRequest=~, bTriggerOnDedicatedServer=~, bTriggerOnFollower=~)
   > Branch(Condition=ClassIsChildOf(TestClass=Get Class(Object=@n8.Notify), ParentClass=@n1))
   > n9 = AddAnimationNotifyEvent(
    AnimationSequenceBase=@n4
    NotifyTrackName=@n1.TrackName
    StartTime=GetAnimNotifyEventTriggerTime(NotifyEvent=@n7)
    NotifyClass=/Script/SoundLibrary.AnimNotify_SoundLibrary
   )
   > n10 = Cast To AnimNotify_SoundLibrary(Object=@n9)
   > n11 = Construct Object from Class(
    Class=&SoundLibrarySimpleContext
    self=@n10
    Settings=Make Simple Sound Settings(
     EventName=@n1.SoundLibTag
     bAttachedSound=False
     bStopWhenAttachedDestroyed=False
     bAutoDestroy=True
     bForceComponentCreation=False
     VolumeMultiplier=1
     PitchMultiplier=1
     StartTime=0
     AttachName=None
     LocationType=KeepRelativeOffset
     Location="0,0,0"
    )
    bForwardToActorOwner=True
    bForwardToAttachedActors=True
    bAutoResetContext=True
   )
   > SetEditorProperty(Object=@n10, PropertyName=SoundContext, PropertyValue=@n11)
   n12 = Get ModifiedAssets()
   > n13 = Branch(Condition=Contains Item(TargetArray=@n12, ItemToFind=@n4))
   Add(execute=@n13.else, TargetArray=@n12, NewItem=@n4) {is=title K2Node_CallArrayFunction}
   n14 = AddAnimationNotifyTrack(execute=@n6.else, AnimationSequenceBase=@n4, NotifyTrackName=@n1.TrackName, TrackColor=(R=1,G=1,B=1,A=1))
  ]}
 SoundLib_AutoAddNotifiesAtReloadEvents {nodes [
   n1 = SoundLib_AutoAddNotifiesAtReloadEvents!() -> (then, TrackName:Name, Checkout:bool=~) {is=title K2Node_FunctionEntry}
   > Clear(TargetArray=Get ModifiedAssets())
   > n2 = Iterate Over Selected Assets()
   n3 = Branch(execute=@n2.Completed, Condition=@n1.Checkout)
   > n4 = CheckoutLoadedAssets(AssetsToCheckout=Get ModifiedAssets())
   Clear(execute=[@n4.then, @n3.else], TargetArray=Get ModifiedAssets())
   n5 = Cast To AnimSequenceBase(execute=@n2.Iteration, Object=@n2)
   > Set AnimSequenceBase(AnimSequenceBase:Anim Sequence Base Object Reference=@n5) -> (Output_Get:Anim Sequence Base Object Reference)
   n6 = GetAnimationNotifyEvents(AnimationSequenceBase=Get AnimSequenceBase() -> (AnimSequenceBase:Anim Sequence Base Object Reference))
   n7 = GetReloadFXNotifyCount!(TargetArray:Array of Anim Notify Event Structures=@n6) -> (Count:int=0) {is=title K2Node_CallFunction}
   > Branch(Condition=AND Boolean(A=Is Not Empty(TargetArray=@n6), B="integer > integer"(A=@n7, B=1)))
   > PrintString(InString=Hello)
   > n8 = Branch(Condition=IsValidAnimNotifyTrackName(AnimationSequenceBase=Get AnimSequenceBase(), NotifyTrackName=Get TrackName() -> (TrackName:Name=None)))
   n9 = For Each Loop(Exec=[@n21.then, @n8.then], Array=GetAnimationNotifyEvents(AnimationSequenceBase=Get AnimSequenceBase()))
   n10 = Break Anim Notify Event(AnimNotifyEvent=@n9) -> (TriggerWeightThreshold=~, NotifyTriggerChance=~, bCanBeFilteredViaRequest=~, bTriggerOnDedicatedServer=~, bTriggerOnFollower=~)
   n11 = Cast To AnimNotify_PlayReloadFX(Object=@n10.Notify) -> (bSuccess=~)
   n12 = Cast To FortAnimNotify_PlayReloadFX(Object=@n10.Notify) -> (bSuccess=~)
   > Branch(Condition=OR Boolean(A=@n11.bSuccess, B=@n12.bSuccess))
   > TransactObject(Object=Get AnimSequenceBase())
   > n13 = AddAnimationNotifyEvent(
    AnimationSequenceBase=Get AnimSequenceBase()
    NotifyTrackName=Get TrackName() -> (TrackName=None)
    StartTime=GetAnimNotifyEventTriggerTime(NotifyEvent=@n9)
    NotifyClass=&AnimNotify_SoundLibrary
   )
   > n14 = Cast To AnimNotify_SoundLibrary(Object=@n13)
   n15 = Select(
    Option 0:"EFortReloadFXState Enum (by ref)"=Get ReloadStage!(self:Play Reload FX Object Reference=@n12) -> (ReloadStage:EFortReloadFXState Enum=ReloadStart)
    Option 1:"EFortReloadFXState Enum (by ref)"=Get ReloadStage(self:Anim Notify Play Reload FX Object Reference=@n11) -> (ReloadStage=ReloadStart)
    Index=@n11.bSuccess
   ) -> (ReturnValue:"EFortReloadFXState Enum (by ref)")
   n16 = Get ReloadStateTagMap() -> (ReloadStateTagMap:Map of EFortReloadFXState Enums to Gameplay Tag Structures=ReloadStart)
   > n17 = Construct Object from Class(
    Class=&SoundLibrarySimpleContext
    self=@n14
    Settings=Make Simple Sound Settings(
     EventName=Find(TargetMap:Map of EFortReloadFXState Enums to Gameplay Tag Structures=@n16, Key:"EFortReloadFXState Enum (by ref)"=@n15) -> (Value:Gameplay Tag Structure)
     bAttachedSound=true
     bStopWhenAttachedDestroyed=False
     bAutoDestroy=True
     bForceComponentCreation=False
     VolumeMultiplier=1
     PitchMultiplier=1
     StartTime=0
     AttachName=None
     LocationType=KeepRelativeOffset
     Location="0,0,0"
    )
    bForwardToActorOwner=True
    bForwardToAttachedActors=True
    bAutoResetContext=True
   )
   > SetEditorProperty(Object=@n14, PropertyName=SoundContext, PropertyValue=@n17)
   n18 = Get ModifiedAssets()
   n19 = Get AnimSequenceBase()
   > n20 = Branch(Condition=Contains Item(TargetArray=@n18, ItemToFind=@n19))
   Add(execute=@n20.else, TargetArray=@n18, NewItem=@n19)
   n21 = AddAnimationNotifyTrack(
    execute=@n8.else
    AnimationSequenceBase=Get AnimSequenceBase()
    NotifyTrackName=Get TrackName() -> (TrackName=None)
    TrackColor=(R=1,G=1,B=1,A=1)
   )
  ]}
 Swap Sound Lib Notify Context {nodes [
   n1 = Swap Sound Lib Notify Context!() -> (then, NewSoundLibContextClass:Animation Notify Context Soft Class Reference, EventTagsFilter:Gameplay Tag Container Structure) {is=title K2Node_FunctionEntry}
   > Set NotifyContext(NotifyContext:Animation Notify Context Soft Class Reference=@n1) -> (Output_Get:Animation Notify Context Soft Class Reference)
   > Clear(TargetArray=Get ModifiedAssets())
   > n2 = GetSelectedAssets()
   > n3 = For Each Loop(Array:Array of Object References=@n2) -> (Array Element:Object Reference)
   CheckoutLoadedAssets(execute=@n3.Completed, AssetsToCheckout=Get ModifiedAssets())
   n4 = Cast To AnimSequenceBase(execute=@n3.LoopBody, Object=@n3)
   > n5 = For Each Loop(Array:Array of Anim Notify Event Structures=GetAnimationNotifyEvents(AnimationSequenceBase=@n4)) -> (Array Element:Anim Notify Event Structure)
   n6 = Break Anim Notify Event(AnimNotifyEvent=@n5) -> (TriggerWeightThreshold=~, NotifyTriggerChance=~, bCanBeFilteredViaRequest=~, bTriggerOnDedicatedServer=~, bTriggerOnFollower=~)
   > n7 = Cast To AnimNotify_SoundLibrary(Object=@n6.Notify)
   > n8 = Set SoundLibraryNotify(SoundLibraryNotify:Anim Notify Object Reference=@n7) -> (Output_Get:Anim Notify Object Reference)
   > n9 = GetEditorProperty(Object=@n8, PropertyName=SoundContext) -> (PropertyValue:Object Reference)
   > Set SettingsObject(SettingsObject:Object Reference=@n9) -> (Output_Get:Object Reference)
   > n10 = Set bIsStateNotify(bIsStateNotify:bool=false) -> (Output_Get:bool)
   n11 = Get SettingsObject() -> (SettingsObject:Object Reference)
   n12 = Cast To SoundLibraryAnimContext!(execute=[@n22.then, @n10.then], Object=@n11) -> (then, CastFailed, AsAnimation Notify Context:Animation Notify Context Object Reference) {is=title K2Node_DynamicCast}
   > Is Valid(InputObject=@n11)
   > n13 = Branch(Condition=Gameplay Tag Container Not Empty!(Incoming Tag Container:Gameplay Tag Container Structure=Get EventTagsFilter() -> (EventTagsFilter:Gameplay Tag Container Structure)) -> (Is Not Empty:bool=~) {is=title K2Node_MacroInstance})
   n14 = Break Animation Notify Settings(SoundLibraryAnimContextSettings=Get AnimSettings(self=@n12)) -> (bOnThisActor=~, bAttachToActor=~, bStopWhenAttachedDestroyed=~, Offset=~, FadeTime=~, NotifyTriggerChance=~, VolumeMultiplier=~)
   > n15 = Branch(Condition=MatchesAnyTags!(
     TagOne:Gameplay Tag Structure=@n14.EventName
     OtherContainer:"Gameplay Tag Container Structure (by ref)"=Get EventTagsFilter()
     bExactMatch:bool=false
    ) -> (ReturnValue:bool))
   LoadClassAsset_Blocking(execute=[@n15.then, @n13.else], AssetClass=Get NotifyContext() -> (NotifyContext:Animation Notify Context Soft Class Reference))
   > n16 = Construct Object from Class!(
    execute
    Class=Resolve Soft Reference!(Input:Animation Notify Context Soft Class Reference=Get NotifyContext()) -> (Output:Animation Notify Context Class Reference)
    self=Select(
     Option 0:Object Reference=Get SoundLibraryNotify() -> (SoundLibraryNotify:Anim Notify Object Reference)
     Option 1:Object Reference=Get SoundLibraryNotifyState() -> (SoundLibraryNotifyState:Anim Notify State Object Reference)
     Index=Get bIsStateNotify() -> (bIsStateNotify:bool)
    ) -> (ReturnValue:Object Reference)
    AnimSettings:Animation Notify Settings Structure=Make Animation Notify Settings!(
     AnimRateThreshold:Float Interval Structure=@n14
     EventName:Gameplay Tag Structure=@n14.EventName
     bOnThisActor:bool=False
     bAttachToActor:bool=@n14.bAttachToActor
     bStopWhenAttachedDestroyed:bool=False
     Offset:Vector="0,0,0"
     AttachName:Name=@n14.AttachName
     FadeTime:float=@n14.FadeTime
     NotifyTriggerChance:float=@n14.NotifyTriggerChance
     VolumeMultiplier:float=1
    ) -> (SoundLibraryAnimContextSettings:Animation Notify Settings Structure) {is=title K2Node_MakeStruct}
    Animation:Anim Sequence Base Weak object=@n4
    Actor
    BaseEventName
    bForwardToActorOwner=True
    bForwardToAttachedActors=True
    bAutoResetContext=True
   ) -> (then, ReturnValue:Animation Notify Context Object Reference)
   > SetEditorProperty(
    Object=Select(Option 0=Get SoundLibraryNotify(), Option 1=Get SoundLibraryNotifyState(), Index=Get bIsStateNotify())
    PropertyName=SoundContext
    PropertyValue:Animation Notify Context Object Reference=@n16
   )
   n17 = Get ModifiedAssets()
   > n18 = Branch(Condition=Contains Item(TargetArray=@n17, ItemToFind=@n4))
   Add(execute=@n18.else, TargetArray=@n17, NewItem=@n4)
   n19 = Cast To AnimNotifyState_SoundLibrary!(execute=@n7.CastFailed, Object=@n6.NotifyStateClass) -> (then, CastFailed, AsAnim Notify State Sound Library:Anim Notify State Sound Library Object Reference) {is=title K2Node_DynamicCast}
   > n20 = Set SoundLibraryNotifyState(SoundLibraryNotifyState:Anim Notify State Object Reference=@n19) -> (Output_Get:Anim Notify State Object Reference)
   > n21 = GetEditorProperty(Object=@n20, PropertyName=SoundContext)
   > Set bIsStateNotify(bIsStateNotify=true)
   > n22 = Set SettingsObject(SettingsObject=@n21)
  ]}
 SwapSoundLibNotifyEventName {nodes [
   SwapSoundLibNotifyEventName!() -> (then, OldEventTag:Gameplay Tag Structure, NewEventTag:Gameplay Tag Structure) {is=title K2Node_FunctionEntry}
   > Clear(TargetArray=Get ModifiedAssets())
   > n1 = GetSelectedAssets()
   > n2 = For Each Loop(Array:Array of Object References=@n1) -> (Array Element:Object Reference)
   CheckoutLoadedAssets(execute=@n2.Completed, AssetsToCheckout=Get ModifiedAssets())
   n3 = Cast To AnimSequenceBase(execute=@n2.LoopBody, Object=@n2)
   > n4 = For Each Loop(Array:Array of Anim Notify Event Structures=GetAnimationNotifyEvents(AnimationSequenceBase=@n3)) -> (Array Element:Anim Notify Event Structure)
   n5 = Break Anim Notify Event(AnimNotifyEvent=@n4) -> (TriggerWeightThreshold=~, NotifyTriggerChance=~, bCanBeFilteredViaRequest=~, bTriggerOnDedicatedServer=~, bTriggerOnFollower=~)
   > n6 = Cast To AnimNotify_SoundLibrary(Object=@n5.Notify)
   > n7 = Set SoundLibraryNotify(SoundLibraryNotify=@n6)
   > n8 = GetEditorProperty(Object=@n7, PropertyName=SoundContext)
   > Set SettingsObject(SettingsObject=@n8)
   > n9 = Set bIsStateNotify(bIsStateNotify=false)
   n10 = Get SettingsObject()
   n11 = Cast To SoundLibraryAnimContext(execute=[@n18.then, @n9.then], Object=@n10)
   > Is Valid(InputObject=@n10)
   n12 = Break Animation Notify Settings(SoundLibraryAnimContextSettings=Get AnimSettings(self=@n11)) -> (bOnThisActor=~, bAttachToActor=~, bStopWhenAttachedDestroyed=~, Offset=~, FadeTime=~, NotifyTriggerChance=~, VolumeMultiplier=~)
   > Branch(Condition="Equal (GameplayTag)"(A=@n12.EventName, B=Get OldEventTag() -> (OldEventTag:Gameplay Tag Structure)))
   > Set AnimSettings(
    AnimSettings=Make Animation Notify Settings(
     AnimRateThreshold=@n12
     EventName=Get NewEventTag() -> (NewEventTag:Gameplay Tag Structure)
     bOnThisActor=False
     bAttachToActor=@n12.bAttachToActor
     bStopWhenAttachedDestroyed=False
     Offset="0,0,0"
     AttachName=@n12.AttachName
     FadeTime=@n12.FadeTime
     NotifyTriggerChance=@n12.NotifyTriggerChance
     VolumeMultiplier=1
    )
    self=@n11
   )
   n13 = Get ModifiedAssets()
   > n14 = Branch(Condition=Contains Item(TargetArray=@n13, ItemToFind=@n3))
   Add(execute=@n14.else, TargetArray=@n13, NewItem=@n3)
   n15 = Cast To AnimNotifyState_SoundLibrary(execute=@n6.CastFailed, Object=@n5.NotifyStateClass)
   > n16 = Set SoundLibraryNotifyState(SoundLibraryNotifyState=@n15)
   > n17 = GetEditorProperty(Object=@n16, PropertyName=SoundContext)
   > Set bIsStateNotify(bIsStateNotify=true)
   > n18 = Set SettingsObject(SettingsObject=@n17)
  ]}
]
variables [
 bTrackContainsOutsideNotify {type=boolean}
 ModifiedAssets {type="array<Object>", default []}
 Notify Name {type="array<string>", default []}
 ReloadStateTagMap {default {
   ReloadCartridge {tagName=SoundLibrary.Weapon.Reload.Cartridge}
   ReloadEnd {tagName=SoundLibrary.Weapon.Reload.End}
   ReloadStart {tagName=SoundLibrary.Weapon.Reload.Start}
  }}
]
