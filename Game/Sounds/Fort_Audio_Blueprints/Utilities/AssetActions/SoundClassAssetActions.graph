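# Editor utility of asset actions for selected SoundClasses: dB volume offsets, reparenting, submix assignment, reference reporting and SoundMix adjuster cleanup.
# RebalanceVolume recursively folds class volume into referencing sounds' volumes, resetting the class to unity.
# Reparenting sets volume from the ratio of old and new parent-chain gains.
# CreateNewClassTree builds hierarchies from comma-separated, dash-indented text, creating assets through embedded Python.
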EditorUtilityBlueprint /Game/Sounds/Fort_Audio_Blueprints/Utilities/AssetActions/SoundClassAssetActions
graphs [
 CalcSoundClassGainStage {nodes [
   n1 = CalcSoundClassGainStage!() -> (then, Current SoundClass:Sound Class Object Reference, Gain Stage:double=1) {is=title K2Node_FunctionEntry}
   > Set CurrentSoundClass(CurrentSoundClass:Sound Class Object Reference=@n1) -> (Output_Get:Sound Class Object Reference)
   > n2 = Set GainStage(GainStage:double=@n1.Gain Stage) -> (Output_Get:double)
   n3 = Get CurrentSoundClass() -> (CurrentSoundClass:Sound Class Object Reference)
   PrintString(execute=[@n2.then, @n8.then], InString=GetDisplayName(Object=@n3), bPrintToScreen=false)
   > n4 = Branch(Condition=IsValid(Object=@n3))
   > n5 = Sequence()
   n6 = Get CurrentSoundClass()
   > n7 = Set GainStage(GainStage=float * float(A=GetSoundClassVolume!(SoundClass:Sound Class Object Reference=@n6) -> (Volume:double), B=Get GainStage() -> (GainStage:double)))
   > PrintString(InString="To String (Float)"(InDouble=@n7), bPrintToScreen=false)
   n8 = Set CurrentSoundClass(
    execute=@n5.then_1
    CurrentSoundClass=Get ParentClass!(self:Sound Class Object Reference=Get CurrentSoundClass()) -> (ParentClass:Sound Class Object Reference)
   )
   n9 = Get GainStage()
   PrintString(execute=@n4.else, InString=Append(A="Final: ", B="To String (Float)"(InDouble=@n9)), bPrintToScreen=false)
   > Return Node!(execute, OutGainStage:double=Get GainStage())
  ]}
 CheckIfUnreferencedBySounds {nodes [
   CheckIfUnreferencedBySounds() {is=title K2Node_FunctionEntry}
   > n1 = Iterate Over Selected Assets()
   PrintString(execute=@n1.Completed, InString=Get FinalOutputString() -> (FinalOutputString:String))
   Clear(execute=@n1.Iteration, TargetArray:Array of Soft Object Path Structures=Get Paths() -> (Paths:Array of Soft Object Path Structures))
   > n2 = Cast To SoundClass!(execute, Object=@n1) -> (then, CastFailed, AsSound Class:Sound Class Object Reference) {is=title K2Node_DynamicCast}
   > n3 = Get Asset Referencers Soft!(execute:Exec, Object:Object Reference=@n2) -> (Iteration:Exec, Referencer:Asset Data Structure, Completed:Exec) {is=title K2Node_MacroInstance}
   LogString(execute=@n3.Completed, InString=GetDisplayName(Object=@n2))
   > n4 = For Each Loop(Array:Array of Asset Data Structures=Get AssetDatas() -> (AssetDatas:Array of Asset Data Structures)) -> (Array Element:Asset Data Structure)
   n5 = Break Asset Data(AssetData=@n4)
   > LogString(InString=Append(A="     ", B="To String (Name)"(InName=@n5.AssetName)))
   Branch(
    execute=@n3.Iteration
    Condition=ClassIsChildOf(
     TestClass=GetClass!(InAssetData:"Asset Data Structure (by ref)"=@n3) -> (ReturnValue:Object Class Reference)
     ParentClass=/Script/Engine.SoundBase
    )
   )
   > Add(TargetArray:Array of Asset Data Structures=Get AssetDatas(), NewItem:"Asset Data Structure (by ref)"=@n3) {is=title K2Node_CallArrayFunction}
   Get FinalOutputString()
  ]}
 CountCharactersInString {nodes [
   n1 = CountCharactersInString!() -> (then, String:String, Character:String) {is=title K2Node_FunctionEntry}
   > n2 = For Each Loop(Array:Array of Strings=GetCharacterArrayFromString!(SourceString:String=@n1) -> (ReturnValue:Array of Strings)) -> (Array Element:String)
   Return Node!(execute=@n2.Completed, Total Count:int=Get Count() -> (Count:int=0))
   > Branch(Condition="Equal Exactly (String)"(A=@n2, B=@n1.Character))
   > Increment Int(Value=Get Count() -> (Count=0))
  ]}
 CreateNewClassTree {nodes [
   n1 = CreateNewClassTree!() -> (then, Tree Syntax:String) {is=title K2Node_FunctionEntry}
   > n2 = Iterate Over Selected Assets()
   > n3 = Cast To SoundClass(Object=@n2)
   > n4 = Set CurrentRootNode(CurrentRootNode:Sound Class Object Reference=@n3) -> (Output_Get:Sound Class Object Reference)
   > Set StartingRootNode(StartingRootNode:Sound Class Object Reference=@n4) -> (Output_Get:Sound Class Object Reference)
   > n5 = For Each Loop(Array=ParseIntoArray(SourceString=@n1, Delimiter=","))
   > PrintString(InString=@n5)
   > Set DepthPrev(DepthPrev:int=Get DepthCurr() -> (DepthCurr:int=0)) -> (Output_Get:int=0)
   n6 = CountCharactersInString!(String:String=@n5, Character:String=-) -> (Total Count:int=0) {is=title K2Node_CallFunction}
   > PrintString(InString="To String (Integer)"(InInt=@n6))
   > Set DepthCurr(DepthCurr:int=@n6) -> (Output_Get:int=0)
   n7 = Get StartingRootNode() -> (StartingRootNode:Sound Class Object Reference)
   n8 = Get Object Path String(Object=@n7)
   n9 = GetDisplayName(Object=@n7)
   n10 = Split(SourceString=@n5, InStr=-, SearchDir=FromEnd)
   > n11 = Execute Python Script!(
    execute
    PythonScript="import unreal\r\n\r\nfactory = unreal.SoundClassFactory()\r\n\r\nsoundclass = unreal.AssetToolsHelpers.get_asset_tools().create_asset(asset_name, destination_path, unreal.SoundClass, factory)"
    destination_path:String=LeftChop(SourceString=Split(SourceString=@n8, InStr=@n9, SearchDir=FromStart), Count=1)
    asset_name:String=SelectString(A=@n10.RightS, B=@n5, bPickA=@n10.ReturnValue)
   ) -> (then, ReturnValue, soundclass:Object Reference) {inputs [destination_path, asset_name], outputs [soundclass]}
   > n12 = Cast To SoundClass(Object=@n11.soundclass)
   > Set NewlyCreatedNode(NewlyCreatedNode:Sound Class Object Reference=@n12) -> (Output_Get:Sound Class Object Reference)
   > n13 = Sequence()
   > Branch(Condition="Not Equal (Integer)"(A=Get DepthCurr() -> (DepthCurr=0), B=Get DepthPrev() -> (DepthPrev:int=0)))
   > PrintString(InString=Depth Change)
   > n14 = Branch(Condition="integer > integer"(A=Get DepthCurr() -> (DepthCurr=0), B=Get DepthPrev() -> (DepthPrev=0)))
   n15 = Get LastCreatedNode() -> (LastCreatedNode:Sound Class Object Reference)
   > Set CurrentRootNode(CurrentRootNode=Select(
     Option 0:Sound Class Object Reference=Get CurrentRootNode() -> (CurrentRootNode:Sound Class Object Reference)
     Option 1:Sound Class Object Reference=@n15
     Index=IsValid(Object=@n15)
    ) -> (ReturnValue:Sound Class Object Reference))
   > PrintString(InString=Descending)
   n16 = Get CurrentRootNode()
   n17 = Get ParentClass(self=@n16)
   Set CurrentRootNode(execute=@n14.else, CurrentRootNode=Select(Option 0=Get StartingRootNode(), Option 1=@n17, Index=IsValid(Object=@n17)))
   > PrintString(InString=Append!(A="Ascending to: ", B=GetDisplayName(Object=@n16), C:String=" who has a parent of ", D:String=GetDisplayName(Object=@n17)) -> (ReturnValue))
   n18 = Get CurrentRootNode()
   n19 = Get NewlyCreatedNode() -> (NewlyCreatedNode:Sound Class Object Reference)
   Add Unique(
    execute=@n13.then_1
    TargetArray:Array of Sound Class Object References=Get ChildClasses!(self:Sound Class Object Reference=@n18) -> (ChildClasses:Array of Sound Class Object References)
    NewItem:"Sound Class Object Reference (by ref)"=@n19
   )
   > SaveLoadedAsset(AssetToSave=@n18, bOnlyIfIsDirty=false)
   > Set LastCreatedNode(LastCreatedNode:Sound Class Object Reference=@n19) -> (Output_Get:Sound Class Object Reference)
   > SaveLoadedAsset(AssetToSave=@n19, bOnlyIfIsDirty=false)
  ]}
 EventGraph {nodes [
   n1 = Update By Decibels!() -> (OutputDelegate, then, Decibel Amount:double=~, Disable Rounding:bool=~) {is=title K2Node_CustomEvent}
   > n2 = GetSelectedAssets()
   > n3 = For Each Loop(Array:Array of Object References=@n2) -> (Array Element:Object Reference)
   > n4 = Cast To SoundClass(Object=@n3)
   > n5 = GetEditorProperty!(execute:Exec, Object:Object Reference=@n4, PropertyName:Name=Properties) -> (then:Exec, PropertyValue:"Sound Class Properties Structure (by ref)", ReturnValue:bool)
   n6 = Break Sound Class Properties!(SoundClassProperties:"Sound Class Properties Structure (by ref)"=@n5) -> (Volume:float=~) {is=title K2Node_BreakStruct}
   n7 = Db to Linear!(Db:double=float + float(A=Linear to Db!(Linear:double=@n6) -> (Db:double=~) {is=title K2Node_MacroInstance}, B=@n1.Decibel Amount) {is=title K2Node_CommutativeAssociativeBinaryOperator}) -> (Linear:double=~) {is=title K2Node_MacroInstance}
   n8 = "To Text (Float)"(Value=@n7, bUseGrouping=false, MinimumIntegralDigits=1, MaximumIntegralDigits=324, MinimumFractionalDigits=0, MaximumFractionalDigits=3)
   n9 = SelectFloat(A=@n7, B=String To Float(InString="To String (Text)"(InText=@n8)), bPickA=@n1.Disable Rounding)
   > PrintString(InString="To String (Float)"(InDouble=@n9))
   > n10 = Set members in Sound Class Properties!(execute:Exec, StructRef:"Sound Class Properties Structure (by ref)"=@n5, Volume:float=@n9) -> (then:Exec, StructOut:"Sound Class Properties Structure (by ref)") {is=title K2Node_SetFieldsInStruct}
   > SetEditorProperty(Object=@n4, PropertyName=Properties, PropertyValue:"Sound Class Properties Structure (by ref)"=@n10)
  ]}
 GetAllAdjustersFromSoundClass {nodes [
   n1 = GetAllAdjustersFromSoundClass!() -> (then, Sound Class:Sound Class Object Reference) {is=title K2Node_FunctionEntry}
   > n2 = Get Asset Referencers!(execute:Exec, Object:Object Reference=@n1) -> (Iteration:Exec, Referencer:Object Reference, Completed:Exec) {is=title K2Node_MacroInstance}
   Return Node!(
    execute=@n2.Completed
    Out Adjusters:Array of Sound Class Adjuster Structures=Get Adjusters() -> (Adjusters:Array of Sound Class Adjuster Structures)
   )
   > n3 = Cast To SoundMix!(execute, Object=@n2) -> (then, CastFailed, AsSound Mix:Sound Mix Object Reference) {is=title K2Node_DynamicCast}
   > Append Array!(
    execute
    TargetArray:Array of Sound Class Adjuster Structures=Get Adjusters()
    SourceArray:Array of Sound Class Adjuster Structures=Get SoundClassEffects!(self:Sound Mix Object Reference=@n3) -> (SoundClassEffects:Array of Sound Class Adjuster Structures)
   ) -> (then)
  ]}
 GetSoundBaseVolume {nodes [
   n1 = GetSoundBaseVolume!() -> (then, SoundBase:Sound Base Object Reference) {is=title K2Node_FunctionEntry}
   > n2 = GetEditorProperty(
    Object=@n1
    PropertyName=Select(
     Option 0:Name=VolumeMultiplier
     Option 1:Name=Volume
     Index=Cast To SoundWave!(Object=@n1) -> (AsSound Wave:Sound Wave Object Reference, bSuccess:bool=~) {is=title K2Node_DynamicCast}.bSuccess
    ) -> (ReturnValue:Name)
   ) -> (PropertyValue:double)
   > Return Node!(execute, Volume:double=@n2)
  ]}
 GetSoundClassVolume {nodes [
   n1 = GetSoundClassVolume!() -> (then, SoundClass:Sound Class Object Reference) {is=title K2Node_FunctionEntry}
   > Return Node(Volume=Break Sound Class Properties(SoundClassProperties=Get Properties!(self:Sound Class Object Reference=@n1) -> (Properties:Sound Class Properties Structure)) -> (Volume=~))
  ]}
 MakeModulatorSettings {nodes [
   n1 = MakeModulatorSettings!() -> (then, New Modulator:Sound Modulator Base Object Reference, Modulators:Set of Sound Modulator Base Object References) {is=title K2Node_FunctionEntry}
   > Set DestinationSettingsCopy(DestinationSettingsCopy:Sound Modulation Destination Settings Structure=Get DefaultDestinationSettings() -> (DefaultDestinationSettings:Sound Modulation Destination Settings Structure)) -> (Output_Get:Sound Modulation Destination Settings Structure)
   > n2 = Set ModulatorsLocal(ModulatorsLocal:Set of Sound Modulator Base Object References=@n1.Modulators) -> (Output_Get:Set of Sound Modulator Base Object References)
   > Add!(execute, TargetSet:Set of Sound Modulator Base Object References=@n2, NewItem:"Sound Modulator Base Object Reference (by ref)"=@n1) -> (then) {is=title K2Node_CallFunction}
   > n3 = Set members in Sound Modulation Destination Settings!(
    execute:Exec
    StructRef:"Sound Modulation Destination Settings Structure (by ref)"=Get DestinationSettingsCopy() -> (DestinationSettingsCopy:Sound Modulation Destination Settings Structure)
    Modulators:Set of Sound Modulator Base Object References=Get ModulatorsLocal() -> (ModulatorsLocal:Set of Sound Modulator Base Object References)
   ) -> (then:Exec, StructOut:"Sound Modulation Destination Settings Structure (by ref)") {is=title K2Node_SetFieldsInStruct}
   > Return Node!(execute, DefaultDestinationSettings:Sound Modulation Destination Settings Structure=@n3)
  ]}
 MoveSoundClassToNewParent {nodes [
   n1 = MoveSoundClassToNewParent!() -> (then, New Parent:Sound Class Object Reference) {is=title K2Node_FunctionEntry}
   > Set NewParentSoundClass(NewParentSoundClass:Sound Class Object Reference=@n1) -> (Output_Get:Sound Class Object Reference)
   > n2 = Iterate Over Selected Assets()
   > n3 = Cast To SoundClass(Object=@n2)
   > Set TargetSoundClass(TargetSoundClass:Sound Class Object Reference=@n3) -> (Output_Get:Sound Class Object Reference)
   n4 = Get TargetSoundClass() -> (TargetSoundClass:Sound Class Object Reference)
   > TransactObject(Object=@n4)
   > TransactObject(Object=Get NewParentSoundClass() -> (NewParentSoundClass:Sound Class Object Reference))
   > TransactObject(Object=Get ParentClass(self=@n4))
   n5 = Get TargetSoundClass()
   > n6 = GetEditorProperty(Object=@n5, PropertyName=Properties) -> (PropertyValue:"Sound Class Properties Structure (by ref)")
   n7 = Get NewParentSoundClass()
   n8 = Get ParentClass(self=Get TargetSoundClass())
   > n9 = Set members in Sound Class Properties(
    StructRef=@n6
    Volume=float / float(
     A=CalcSoundClassGainStage!(Current SoundClass:Sound Class Object Reference=@n8, Gain Stage:double=1) -> (OutGainStage:double) {is=title K2Node_CallFunction}
     B=CalcSoundClassGainStage(Current SoundClass=@n7, Gain Stage=1)
    )
   )
   > SetEditorProperty(Object=@n5, PropertyName=Properties, PropertyValue=@n9)
   n10 = Get ParentClass(self=Get TargetSoundClass())
   > n11 = GetEditorProperty(Object=@n10, PropertyName=ChildClasses) -> (PropertyValue:Array of Sound Class Object References)
   > n12 = Set LocalChildClasses(LocalChildClasses:Array of Sound Class Object References=@n11) -> (Output_Get:Array of Sound Class Object References)
   > Remove Item!(execute, TargetArray:Array of Sound Class Object References=@n12, Item:"Sound Class Object Reference (by ref)"=Get TargetSoundClass()) -> (then, ReturnValue:bool)
   > SetEditorProperty(
    Object=@n10
    PropertyName=ChildClasses
    PropertyValue:Array of Sound Class Object References=Get LocalChildClasses() -> (LocalChildClasses:Array of Sound Class Object References)
   )
   n13 = Get NewParentSoundClass()
   > n14 = GetEditorProperty(Object=@n13, PropertyName=ChildClasses)
   > n15 = Set LocalChildClasses(LocalChildClasses=@n14)
   > Remove Item(TargetArray=@n15, Item=Get TargetSoundClass())
   > SetEditorProperty(Object=@n13, PropertyName=ChildClasses, PropertyValue=Get LocalChildClasses())
   > SetEditorProperty(Object=Get TargetSoundClass(), PropertyName=ParentClass, PropertyValue:Sound Class Object Reference=Get NewParentSoundClass())
  ]}
 OnClassSetProperties {nodes [OnClassSetProperties!() -> (then, Properties:Sound Class Properties Structure) {is=title K2Node_FunctionEntry}]}
 RebalanceInternal {nodes [
   n1 = RebalanceInternal!() -> (then, Class:Sound Class Object Reference, Gain Adjustment:double=~) {is=title K2Node_FunctionEntry}
   > n2 = Get Asset Referencers(Object=@n1)
   n3 = For Each Loop(Exec=@n2.Completed, Array:Array of Sound Class Object References=Get ChildClasses(self=@n1)) -> (Array Element:Sound Class Object Reference)
   > RebalanceInternal!(execute, self, Class:Sound Class Object Reference=@n3, Gain Adjustment:double=Get Gain Adjustment() -> (Gain Adjustment:double)) -> (then) {is=title K2Node_CallFunction}
   n4 = Cast To SoundBase!(execute=@n2.Iteration, Object=@n2) -> (then, CastFailed, AsSound Base:Sound Base Object Reference) {is=title K2Node_DynamicCast}
   > n5 = Branch(Condition=Contains Item!(
     TargetSet:Set of Sound Base Object References=Get RebalancedClasses() -> (RebalancedClasses:Set of Sound Base Object References)
     ItemToFind:"Sound Base Object Reference (by ref)"
    ) -> (ReturnValue:bool) {is=title K2Node_CallFunction})
   n6 = GetDisplayName(Object=@n4)
   n7 = GetSoundBaseVolume!(SoundBase:Sound Base Object Reference=@n4) -> (Volume:double) {is=title K2Node_CallFunction}
   n8 = float / float(A=@n7, B=Get Gain Adjustment())
   PrintString(
    execute=@n5.else
    InString="To String (Text)"(InText=Format Text!(
      Format="LOCTEXT(\"\\r\\n{name}\\r\\n- Current Volume: {vol}\\r\\n- Final Volume: {newvol}\")"
      name:String=@n6
      vol:double=@n7
      newvol:double=@n8
     ) -> (Result))
   )
   > SetSoundBaseVolume!(execute, self, SoundBase:Sound Base Object Reference=@n4, Volume:double=@n8) -> (then)
   > Add(TargetSet:Set of Sound Base Object References=Get RebalancedClasses(), NewItem:"Sound Base Object Reference (by ref)"=@n4)
  ]}
 RebalanceVolume {nodes [
   RebalanceVolume() {is=title K2Node_FunctionEntry}
   > Clear!(execute, TargetSet:Set of Sound Base Object References=Get RebalancedClasses()) -> (then) {is=title K2Node_CallFunction}
   > n1 = Iterate Over Selected Assets(AssetClass=/Script/Engine.SoundClass)
   > n2 = Cast To SoundClass(Object=@n1)
   n3 = GetSoundClassVolume(SoundClass=@n2) {is=title K2Node_CallFunction}
   n4 = float / float(A=1, B=@n3)
   > RebalanceInternal(Class=@n2, Gain Adjustment=@n4)
   > TransactObject(Object=@n2)
   > n5 = GetEditorProperty(Object=@n2, PropertyName=Properties) -> (PropertyValue:"Sound Class Properties Structure (by ref)")
   > n6 = Set members in Sound Class Properties(StructRef=@n5, Volume=float * float(A=@n3, B=@n4) {is=title K2Node_PromotableOperator})
   > SetEditorProperty(Object=@n2, PropertyName=Properties, PropertyValue:"Sound Class Properties Structure (by ref)"=@n6)
  ]}
 RemoveFromSoundMixes {nodes [
   RemoveFromSoundMixes() {is=title K2Node_FunctionEntry}
   > n1 = Iterate Over Selected Assets()
   > n2 = Cast To SoundClass(Object=@n1)
   > n3 = Get Asset Referencers(Object=@n2)
   > n4 = Cast To SoundMix(Object=@n3)
   > Set ShouldDirty(ShouldDirty:bool=false) -> (Output_Get:bool)
   > Clear(TargetArray:Array of Sound Class Adjuster Structures=Get NewArray() -> (NewArray:Array of Sound Class Adjuster Structures)) {is=title K2Node_CallArrayFunction}
   > n5 = For Each Loop(Array:Array of Sound Class Adjuster Structures=Get SoundClassEffects(self=@n4)) -> (Array Element:Sound Class Adjuster Structure)
   Branch(execute=@n5.Completed, Condition=Get ShouldDirty() -> (ShouldDirty:bool))
   > SetEditorProperty(Object=@n4, PropertyName=SoundClassEffects, PropertyValue:Array of Sound Class Adjuster Structures=Get NewArray())
   n6 = Branch(
    execute=@n5.LoopBody
    Condition="Equal (Object)"(
     A=Break Sound Class Adjuster!(SoundClassAdjuster:"Sound Class Adjuster Structure (by ref)"=@n5) -> (
      SoundClassObject:Sound Class Object Reference
      VolumeAdjuster:float=~
      PitchAdjuster:float=~
      LowPassFilterFrequency:float=~
      bApplyToChildren:bool=~
      VoiceCenterChannelVolumeAdjuster:float=~
     ) {is=title K2Node_BreakStruct}
     B=@n2
    )
   )
   > Set ShouldDirty(ShouldDirty=true)
   Add(execute=@n6.else, TargetArray:Array of Sound Class Adjuster Structures=Get NewArray(), NewItem:"Sound Class Adjuster Structure (by ref)"=@n5) {is=title K2Node_CallArrayFunction}
  ]}
 ReplaceSoundClassInReferencers {nodes [
   n1 = ReplaceSoundClassInReferencers!() -> (then, Sound Class:Sound Class Object Reference) {is=title K2Node_FunctionEntry}
   > Set SoundClassToReplaceWith(SoundClassToReplaceWith:Sound Class Object Reference=@n1) -> (Output_Get:Sound Class Object Reference)
   > n2 = Iterate Over Selected Assets()
   CheckoutLoadedAssets(execute=@n2.Completed, AssetsToCheckout=Get AssetsToCheckOut() -> (AssetsToCheckOut:Array of Object References))
   n3 = Get Asset Referencers(execute=@n2.Iteration, Object=@n2)
   > n4 = Cast To SoundMix(Object=@n3)
   n5 = Get SoundClassEffects(self=@n4)
   > n6 = For Each Loop(Array=@n5)
   > Branch(Condition="Equal (Object)"(A=Break Sound Class Adjuster!(SoundClassAdjuster=@n6) -> (SoundClassObject), B=@n2))
   > Set members in Sound Class Adjuster!(
    execute:Exec
    StructRef:"Sound Class Adjuster Structure (by ref)"="Get (a ref)"(Array:Array of Sound Class Adjuster Structures=@n5, Dimension 1=@n6.Array Index) -> (Output:"Sound Class Adjuster Structure (by ref)")
    SoundClassObject:Sound Class Object Reference=Get SoundClassToReplaceWith() -> (SoundClassToReplaceWith:Sound Class Object Reference)
   ) -> (then:Exec, StructOut:"Sound Class Adjuster Structure (by ref)") {is=title K2Node_SetFieldsInStruct}
   > Add Unique(TargetArray:Array of Object References=Get AssetsToCheckOut(), NewItem:"Object Reference (by ref)"=@n4)
  ]}
 SetClassDefaultSubmix {nodes [
   n1 = SetClassDefaultSubmix!() -> (then, Class:Sound Class Object Reference, SubMix:Sound Submix Object Reference) {is=title K2Node_FunctionEntry}
   > n2 = Sequence!(execute) -> (then_0, then_1, then_2:Exec)
   > TransactObject(Object=@n1)
   > n3 = GetEditorProperty(Object=@n1, PropertyName=Properties)
   > n4 = Set members in Sound Class Properties!(execute, StructRef=@n3, bApplyEffects:bool=False) -> (then, StructOut)
   > n5 = Set members in Sound Class Properties!(execute, StructRef=@n4, DefaultSubmix:Sound Submix Object Reference=@n1.SubMix) -> (then, StructOut)
   > SetEditorProperty(Object=@n1, PropertyName=Properties, PropertyValue:"Sound Class Properties Structure (by ref)"=@n5)
   n6 = GetEditorProperty(execute=@n2.then_1, Object=@n1, PropertyName=ChildClasses) -> (PropertyValue:Array of Sound Class Object References)
   > n7 = For Each Loop(Array:Array of Sound Class Object References=@n6) -> (Array Element:Sound Class Object Reference)
   > SetClassDefaultSubmix!(execute, self, Class:Sound Class Object Reference=@n7, SubMix:Sound Submix Object Reference=@n1.SubMix) -> (then) {is=title K2Node_CallFunction}
  ]}
 SetParentToNone {nodes [
   SetParentToNone() {is=title K2Node_FunctionEntry}
   > n1 = Iterate Over Selected Assets()
   > n2 = Cast To SoundClass(Object=@n1)
   > TransactObject(Object=@n2)
   > SetEditorProperty(Object=@n2, PropertyName=ParentClass, PropertyValue:Sound Class Object Reference=Get Null() -> (Null:Sound Class Object Reference))
  ]}
 SetSoundBaseVolume {nodes [
   n1 = SetSoundBaseVolume!() -> (then, SoundBase:Sound Base Object Reference, Volume:double=~) {is=title K2Node_FunctionEntry}
   > SetEditorProperty(
    Object=@n1
    PropertyName=Select(Option 0=VolumeMultiplier, Option 1=Volume, Index=Cast To SoundWave(Object=@n1) -> (bSuccess=~).bSuccess)
    PropertyValue:double=@n1.Volume
   )
  ]}
 Traverse {nodes [
   n1 = Traverse!() -> (then, SoundClass:Object Reference, Indentation:String) {is=title K2Node_FunctionEntry}
   > n2 = FindPackageReferencersForAsset!(execute, AssetPath:String=Get Object Path String(Object=@n1), bLoadAssetsToConfirm:bool=true) -> (then, ReturnValue:Array of Strings)
   > n3 = For Each Loop(Array:Array of Strings=@n2) -> (Array Element:String)
   n4 = GetEditorProperty(execute=@n3.Completed, Object=@n1, PropertyName=ChildClasses)
   > n5 = For Each Loop(Array:Array of Sound Class Object References=@n4) -> (Array Element:Sound Class Object Reference)
   > PrintString(InString=Append!(A=@n1.Indentation, B="SoundClass: ", C=GetDisplayName(Object=@n5)) -> (ReturnValue))
   > Traverse!(execute, self, SoundClass:Object Reference=@n5, Indentation:String=Append!(A=@n1.Indentation, B="     ") -> (ReturnValue)) -> (then) {is=title K2Node_CallFunction}
   n6 = Split(SourceString=@n3, InStr=/, SearchDir=FromEnd)
   n7 = Append!(A=SoundBase', B=@n6, C:String=/, D:String=@n6.RightS, E:String=., F:String=@n6.RightS, G:String=') -> (ReturnValue)
   n8 = Get Asset By Object Path!(
    execute=@n3.LoopBody
    self:Asset Registry Interface=GetAssetRegistry()
    ObjectPath:"Soft Object Path Structure (by ref)"=MakeSoftClassPath(PathString=Append!(A=@n1.Indentation, B=@n7) -> (ReturnValue))
    bIncludeOnlyOnDiskAssets:bool=false
    bSkipARFilteredAssets:bool=true
   ) -> (then, ReturnValue:Asset Data Structure) {is=title K2Node_CallFunction}
   > n9 = Cast To SoundBase(Object=GetAsset(InAssetData=@n8))
   > PrintString(InString=Append!(A=@n1.Indentation, B="SoundBase: ", C:String=GetDisplayName(Object=@n9)) -> (ReturnValue))
  ]}
]
variables [
 DefaultDestinationSettings {type=SoundModulationDestinationSettings, default {bEnableModulation=true, modulator=null, value=1, modulators []}}
 PropertiesCopy {
  type=SoundClassProperties
  default {
   attenuationDistanceScale=1
   bAlwaysPlay=false
   bApplyAmbientVolumes=false
   bApplyEffects=false
   bCenterChannelOnly=false
   bIsMusic=false
   bIsUISound=false
   bReverb=true
   default2DReverbSendAmount=0
   defaultSubmix=null
   lFEBleed=0.5
   loadingBehavior=Inherited
   lowPassFilterFrequency=20000
   outputTarget=Speaker
   pitch=1
   radioFilterVolume=0
   radioFilterVolumeThreshold=0
   voiceCenterChannelVolume=0
   volume=1
   modulationSettings {
    highpassModulationDestination {bEnableModulation=false, modulator=null, value=20, modulators []}
    lowpassModulationDestination {bEnableModulation=false, modulator=null, value=20000, modulators []}
    pitchModulationDestination {bEnableModulation=false, modulator=null, value=0, modulators []}
    volumeModulationDestination {bEnableModulation=false, modulator=null, value=0, modulators []}
   }
   sizeOfFirstAudioChunkInSeconds {default=0, perPlatform {}}
  }
 }
 RebalancedClasses {type="array<SoundBase>", default []}
]
