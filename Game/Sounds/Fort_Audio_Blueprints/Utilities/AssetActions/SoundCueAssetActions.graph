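# Editor utility Blueprint of batch asset actions that audio developers run on selected Sound Cues.
# It gathers empty or unreferenced cues into local collections, sets concurrency, duplicates cues and sound sets, and renames branch parameters on first-level child nodes.
# Python snippets create attenuation assets from overrides and Fort Sound Sequence data assets from selections.
# Recursive helpers exclude nodes from branch culling and print cue trees, while another multiplies sound-class volume up the parent chain.
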EditorUtilityBlueprint /Game/Sounds/Fort_Audio_Blueprints/Utilities/AssetActions/SoundCueAssetActions
graphs [
 CheckForEmptyCues {nodes [
   CheckForEmptyCues() {is=title K2Node_FunctionEntry}
   > n1 = Iterate Over Selected Assets()
   n2 = Show Message Dialog(
    execute=@n1.Completed
    Title=MakeLiteralText(Value="LOCTEXT(\"Create Collection\")")
    Message=Format Text!(
     Format="LOCTEXT(\"You\\'re about to create a collection called {name} containing all empty cues found. Do you want to continue?\")"
     name:Name=Get FoundCuesCollection() -> (FoundCuesCollection:Name=None)
    ) -> (Result)
    MessageType=YesNoCancel
    DefaultValue=No
    MessageCategory=Warning
   )
   > n3 = Switch on EAppReturnType(Selection=@n2)
   n4 = Get CollectionManagerScriptingSubsystem()
   n5 = CreateOrEmptyCollection!(
    execute=@n3.Yes
    self:Collection Manager Subsystem Object Reference=@n4
    Container:Collection Container Source Structure
    Collection:Name=Get FoundCuesCollection() -> (FoundCuesCollection=None)
    ShareType:ECollectionScriptingShareType Enum=Local
   ) -> (then, OutNewOrEmptyCollection:Collection Structure, ReturnValue:bool)
   > AddAssetsToCollection!(
    execute
    self:Collection Manager Subsystem Object Reference=@n4
    Collection:"Collection Structure (by ref)"=@n5
    AssetPaths:Array of Soft Object Path Structures=Get FoundCuesSoftObjPaths() -> (FoundCuesSoftObjPaths:Array of Soft Object Path Structures)
   ) -> (then, ReturnValue:bool)
   n6 = Cast To SoundCue!(execute=@n1.Iteration, Object=@n1) -> (then, CastFailed, AsSound Cue:Sound Cue Object Reference) {is=title K2Node_DynamicCast}
   > n7 = Is Valid(InputObject=Get FirstNode!(self:Sound Cue Object Reference=@n6) -> (FirstNode:Sound Node Object Reference))
   Add(
    execute=@n7.Is Not Valid
    TargetArray:Array of Soft Object Path Structures=Get FoundCuesSoftObjPaths()
    NewItem:"Soft Object Path Structure (by ref)"=GetSoftObjectPath(Object=@n6)
   ) {is=title K2Node_CallArrayFunction}
  ]}
 CheckForUnreferenced {nodes [
   CheckForUnreferenced() {is=title K2Node_FunctionEntry}
   > n1 = Iterate Over Selected Assets()
   n2 = Show Message Dialog(
    execute=@n1.Completed
    Title=MakeLiteralText(Value="LOCTEXT(\"Create Collection\")")
    Message=Format Text(
     Format="LOCTEXT(\"You\\'re about to create a collection called {name} containing all empty cues found. Do you want to continue?\")"
     name=Get FoundCuesCollection() -> (FoundCuesCollection=None)
    )
    MessageType=YesNoCancel
    DefaultValue=No
    MessageCategory=Warning
   )
   > n3 = Switch on EAppReturnType(Selection=@n2)
   n4 = Get CollectionManagerScriptingSubsystem()
   n5 = CreateOrEmptyCollection(execute=@n3.Yes, self=@n4, Collection=Get FoundCuesCollection() -> (FoundCuesCollection=None), ShareType=Local)
   > AddAssetsToCollection(self=@n4, Collection=@n5, AssetPaths=Get FoundCuesSoftObjPaths())
   n6 = Cast To SoundCue(execute=@n1.Iteration, Object=@n1)
   > Set bFoundReferencer(bFoundReferencer:bool=false) -> (Output_Get:bool)
   > n7 = Get Asset Referencers!(execute:Exec, Object:Object Reference=@n6) -> (Iteration:Exec, Referencer:Object Reference, Completed:Exec) {is=title K2Node_MacroInstance}
   n8 = Branch(execute=@n7.Completed, Condition=Get bFoundReferencer() -> (bFoundReferencer:bool))
   Add(execute=@n8.else, TargetArray=Get FoundCuesSoftObjPaths(), NewItem=GetSoftObjectPath(Object=@n6))
   Set bFoundReferencer(execute=@n7.Iteration, bFoundReferencer=true)
  ]}
 ConvertAttenuationOverrideToAsset {nodes [
   ConvertAttenuationOverrideToAsset() {is=title K2Node_FunctionEntry}
   > n1 = Iterate Over Selected Assets()
   K2Node_Composite_0 = Get Destination and Asset Name!(Object:Object Reference=@n1) -> (Destination:String, Name:String) {is=title K2Node_Composite}
   > n2 = Execute Python Script!(
    execute
    PythonScript="import unreal\r\n\r\nfactory = unreal.SoundAttenuationFactory()\r\n\r\nattenuation = unreal.AssetToolsHelpers.get_asset_tools().create_asset(attenuation_name, destination_path, unreal.SoundAttenuation, factory)"
    destination_path:String=@K2Node_Composite_0
    attenuation_name:String=@K2Node_Composite_0.Name
   ) -> (then, ReturnValue, attenuation:Object Reference) {inputs [destination_path, attenuation_name], outputs [attenuation]}
   > n3 = Cast To SoundAttenuation!(execute, Object=@n2.attenuation) -> (then, CastFailed, AsSound Attenuation:Sound Attenuation Object Reference) {is=title K2Node_DynamicCast}
   > n4 = GetEditorProperty!(execute:Exec, Object:Object Reference=@n1, PropertyName:Name=AttenuationOverrides) -> (then:Exec, PropertyValue:Sound Attenuation Settings Structure, ReturnValue:bool)
   > SetEditorProperty(Object=@n3, PropertyName=Attenuation, PropertyValue:Sound Attenuation Settings Structure=@n4)
   > SetEditorProperty(Object=@n1, PropertyName=bOverrideAttenuation, PropertyValue:bool=MakeLiteralBool(Value=false))
   > SetEditorProperty(Object=@n1, PropertyName=AttenuationSettings, PropertyValue:Sound Attenuation Object Reference=@n3)
  ]}
 Get Destination and Asset Name {
  path=ConvertAttenuationOverrideToAsset.K2Node_Composite_0.Get Destination and Asset Name
  nodes [
   n1 = Inputs!() -> (Object:Object Reference)
   n2 = GetDisplayName(Object=@n1)
   n3 = Split(SourceString=@n2, InStr=_, SearchDir=FromEnd)
   Outputs!(
    Destination:String=Split(SourceString=Get Object Path String(Object=@n1), InStr=@n2, SearchDir=FromStart)
    Name:String=Append(A=SelectString(A=@n3, B=@n2, bPickA=@n3.ReturnValue), B=_Att)
   )
  ]
 }
 CreateSoundSequence {nodes [
   n1 = CreateSoundSequence!() -> (then, Name:String) {is=title K2Node_FunctionEntry}
   > n2 = Iterate Over Selected Assets()
   K2Node_Composite_0 = GetDestinationAndName!(Object:Object Reference=@n2) -> (Destination:String) {is=title K2Node_Composite}
   n3 = Execute Python Script!(
    execute=@n2.Completed
    PythonScript="import unreal\r\n\r\nfactory = unreal.DataAssetFactory()\r\n\r\ndata_asset = unreal.AssetToolsHelpers.get_asset_tools().create_asset(data_asset_name, destination_path, data_asset_class, factory)"
    destination_path=@K2Node_Composite_0
    data_asset_name:String=@n1
    data_asset_class:Fort Sound Sequence Class Reference=Get DataAssetClass() -> (DataAssetClass:Fort Sound Sequence Class Reference)
   ) -> (then, ReturnValue, data_asset:Object Reference) {inputs [destination_path, data_asset_name, data_asset_class], outputs [data_asset]}
   > SetEditorProperty(
    Object=@n3.data_asset
    PropertyName=SoundSequenceData
    PropertyValue:Array of Sound Sequence Data Structures=Get SequenceArray() -> (SequenceArray:Array of Sound Sequence Data Structures)
   )
   n4 = Cast To SoundCue(execute=@n2.Iteration, Object=@n2)
   > Add!(
    execute
    TargetArray:Array of Sound Sequence Data Structures=Get SequenceArray()
    NewItem_Sound:Sound Base Object Reference=@n4
    NewItem_Delay:float=0
   ) -> (then, ReturnValue)
  ]}
 GetDestinationAndName {
  path=CreateSoundSequence.K2Node_Composite_0.GetDestinationAndName
  nodes [
   n1 = Inputs()
   Outputs!(Destination=Split(SourceString=Get Object Path String(Object=@n1), InStr=GetDisplayName(Object=@n1), SearchDir=FromStart))
  ]
 }
 CreateSoundsFromTemplateCue {nodes [
   n1 = CreateSoundsFromTemplateCue!() -> (then, Sound Cue Name List:Array of Strings) {is=title K2Node_FunctionEntry}
   > n2 = GetSelectedAssets()
   > n3 = For Each Loop(Array=@n1)
   > Branch(Condition="Not Equal (Integer)"!(A:int=Len(S=@n3), B:int=0) -> (ReturnValue:bool) {is=title K2Node_CallFunction})
   n4 = Get Object Path String(Object="Get (a copy)"(Array:Array of Object References=@n2) -> (Output:Object Reference))
   n5 = Split!(InPath:String=@n4) -> (PathPart:String, FilenamePart:String, ExtensionPart:String)
   n6 = Split!(SourceString:String=@n5.FilenamePart, InStr:String=_, SearchCase:ESearchCase Enum, SearchDir:ESearchDir Enum=FromEnd) -> (LeftS:String, RightS:String, ReturnValue:bool)
   n7 = Append!(A=@n6, B=_, C:String=@n3, D:String=_, E:String=@n6.RightS) -> (ReturnValue)
   > DuplicateAsset!(execute, SourceAssetPath:String=@n4, DestinationAssetPath:String=Append(A=@n5, B=/, C=@n7, D=., E=@n7)) -> (then, ReturnValue:Object Reference)
  ]}
 DuplicateSoundSet {nodes [
   n1 = DuplicateSoundSet!() -> (then, NewSetName:String) {is=title K2Node_FunctionEntry}
   > Set NewSetName(NewSetName:String=@n1) -> (Output_Get:String)
   > n2 = GetSelectedAssets()
   > n3 = For Each Loop(Array:Array of Object References=@n2) -> (Array Element:Object Reference)
   > n4 = Sequence()
   n5 = Get Object Path String(Object=@n3)
   n6 = Append!(
    A=Split!(InPath:String=@n5) -> (PathPart:String, FilenamePart:String, ExtensionPart:String)
    B=/
    C=Get NewSetName() -> (NewSetName:String)
    D=/
   ) -> (ReturnValue)
   > n7 = MakeDirectory!(execute, DirectoryPath:String=@n6) -> (then, ReturnValue:bool)
   > n8 = Branch(Condition=@n7)
   Set NewSetPath(execute=[@n8.else, @n8.then], NewSetPath:String=@n6) -> (Output_Get:String)
   n9 = Get NewSetPath() -> (NewSetPath:String)
   Branch(execute=@n4.then_1, Condition="Not Equal (Integer)"(A=Len(S=@n9), B=0))
   n10 = Split!(
    SourceString:String=Split!(InPath:String=@n5) -> (PathPart:String, FilenamePart:String, ExtensionPart:String).FilenamePart
    InStr:String=_
    SearchCase:ESearchCase Enum
    SearchDir:ESearchDir Enum=FromEnd
   ) -> (LeftS:String, RightS:String, ReturnValue:bool)
   n11 = Append!(A=@n10, B=_, C=Get NewSetName(), D=_, E:String=@n10.RightS) -> (ReturnValue)
   > DuplicateAsset(SourceAssetPath=@n5, DestinationAssetPath=Append(A=Get NewSetPath(), B=/, C=@n11, D=., E=@n11))
  ]}
 EventGraph {links [], nodes []}
 ExcludeFromBranchCulling {nodes [
   ExcludeFromBranchCulling() {is=title K2Node_FunctionEntry}
   > n1 = Iterate Over Selected Assets()
   > n2 = Cast To SoundCue(Object=@n1)
   > ExcludeFromBranchCulling_Rec!(execute, self, SoundCue:Sound Cue Object Reference=@n2, Node:Sound Node Object Reference=Get FirstNode(self=@n2)) -> (then)
  ]}
 ExcludeFromBranchCulling_Rec {nodes [
   n1 = ExcludeFromBranchCulling_Rec!() -> (then, SoundCue:Sound Cue Object Reference, Node:Sound Node Object Reference) {is=title K2Node_FunctionEntry}
   > Is Valid(InputObject=@n1.Node)
   > n2 = Sequence()
   > SetEditorProperty(
    Object=Get Node() -> (Node:Sound Node Object Reference)
    PropertyName=bShouldExcludeFromBranchCulling
    PropertyValue:bool=MakeLiteralBool(Value=true)
   )
   n3 = For Each Loop(
    Exec=@n2.then_1
    Array:Array of Sound Node Object References=Get ChildNodes!(self:Sound Node Object Reference=Get Node()) -> (ChildNodes:Array of Sound Node Object References)
   ) -> (Array Element:Sound Node Object Reference)
   > ExcludeFromBranchCulling_Rec(SoundCue=Get SoundCue() -> (SoundCue:Sound Cue Object Reference), Node=@n3) {is=title K2Node_CallFunction}
  ]}
 GetObjectWithName {nodes [
   n1 = GetObjectWithName!() -> (then, Name:String, ObjectList:Array of Object References) {is=title K2Node_FunctionEntry}
   > n2 = For Each Loop(Array:Array of Object References=@n1.ObjectList) -> (Array Element:Object Reference)
   Return Node!(execute=@n2.Completed, Object:Object Reference)
   > Branch(Condition="Equal Exactly (String)"(A=@n1, B=GetDisplayName(Object=@n2)))
   > Return Node(Object=@n2)
  ]}
 GetSoundClassGain {nodes [
   n1 = GetSoundClassGain!() -> (then, SoundClass:Sound Class Object Reference) {is=title K2Node_FunctionEntry}
   > n2 = Set CurrentClass(CurrentClass:Sound Class Object Reference=@n1) -> (Output_Get:Sound Class Object Reference)
   > n3 = Get Sound Class Volume!(execute:Exec, Object:Object Reference=@n2) -> (then:Exec, Volume:double=~) {is=title K2Node_MacroInstance}
   > n4 = Set GainStage(GainStage:double=@n3) -> (Output_Get:double)
   n5 = GetEditorProperty(
    execute=[@n4.then, @n7.then_1]
    Object=Get CurrentClass() -> (CurrentClass:Sound Class Object Reference)
    PropertyName=MakeLiteralName(Value=ParentClass)
   ) -> (PropertyValue:Sound Class Object Reference)
   > Set ParentClass(ParentClass:Sound Class Object Reference=@n5) -> (Output_Get:Sound Class Object Reference)
   > n6 = Branch(Condition=IsValid(Object=Get ParentClass() -> (ParentClass:Sound Class Object Reference)))
   > n7 = Sequence()
   > n8 = Set CurrentClass(CurrentClass=Get ParentClass())
   > n9 = Get Sound Class Volume(Object=@n8) -> (Volume=~)
   > Set GainStage(GainStage=float * float(A=@n9, B=Get GainStage() -> (GainStage:double)))
   Return Node!(execute=@n6.else, OutGainStage:double=Get GainStage())
  ]}
 GetSoundClassVolume {nodes [
   n1 = Inputs!() -> (execute:Exec, Object)
   > n2 = GetEditorProperty(Object=@n1, PropertyName=Properties) -> (PropertyValue:"Sound Class Properties Structure (by ref)")
   > Outputs!(
    then:Exec
    Volume:double=Break Sound Class Properties!(SoundClassProperties:"Sound Class Properties Structure (by ref)"=@n2) -> (Volume:float=~, ModulationSettings:Sound Modulation Default Settings Structure) {is=title K2Node_BreakStruct}
   )
  ]}
 SetConcurrency {nodes [
   n1 = SetConcurrency!() -> (then, New Set:Sound Concurrency Object Reference) {is=title K2Node_FunctionEntry}
   > n2 = Iterate Over Selected Assets()
   > n3 = Cast To SoundCue(Object=@n2)
   > TransactObject(Object=@n3)
   > Set bOverrideConcurrency!(execute, bOverrideConcurrency:bool=false, self:Sound Base Object Reference=@n3) -> (then, Output_Get:bool)
   n4 = Get ConcurrencySet!(self:Sound Base Object Reference=@n3) -> (ConcurrencySet:Set of Sound Concurrency Object References)
   > Clear!(execute, TargetSet:Set of Sound Concurrency Object References=@n4) -> (then) {is=title K2Node_CallFunction}
   > Add!(execute, TargetSet:Set of Sound Concurrency Object References=@n4, NewItem:"Sound Concurrency Object Reference (by ref)"=@n1) -> (then) {is=title K2Node_CallFunction}
  ]}
 SwapBranchNodeParamName {nodes [
   SwapBranchNodeParamName!() -> (then, OldName:Name, NewName:Name) {is=title K2Node_FunctionEntry}
   > n1 = Iterate Over Selected Assets()
   > n2 = Cast To SoundCue(Object=@n1)
   > n3 = For Each Loop(Array:Array of Sound Node Object References=Get ChildNodes(self=Get FirstNode(self=@n2))) -> (Array Element:Sound Node Object Reference)
   > n4 = GetEditorProperty(Object=@n3, PropertyName=BoolParameterName) -> (PropertyValue:Name)
   > Branch(Condition=@n4.ReturnValue)
   > Branch(Condition="Equal (Name)"(A=@n4, B=Get OldName() -> (OldName:Name=None)))
   > SetEditorProperty(Object=@n3, PropertyName=BoolParameterName, PropertyValue:Name=Get NewName() -> (NewName:Name=None))
  ]}
 Test {nodes [Test() {is=title K2Node_FunctionEntry}]}
 TraverseGraph {nodes [
   TraverseGraph() {is=title K2Node_FunctionEntry}
   > n1 = Iterate Over Selected Assets()
   n2 = For Each Loop(Exec=@n1.Completed, Array:Array of Sound Cue Object References=Get FoundCues() -> (FoundCues:Array of Sound Cue Object References)) -> (Array Element:Sound Cue Object Reference)
   Return Node!(
    execute=@n2.Completed
    GraphInfo:Map of Sound Cue Object References to Sound Cue Editor Graph Info Structures=Get GraphInfoLocal() -> (GraphInfoLocal:Map of Sound Cue Object References to Sound Cue Editor Graph Info Structures)
   )
   > PrintString(InString=GetDisplayName(Object=@n2))
   n3 = Cast To SoundCue(execute=@n1.Iteration, Object=@n1)
   > TraverseGraph_Rec!(execute, self, SoundCue:Sound Cue Object Reference=@n3, Node:Sound Node Object Reference=Get FirstNode(self=@n3), DepthPrefix:String) -> (then)
  ]}
 TraverseGraph_Rec {nodes [
   n1 = TraverseGraph_Rec!() -> (then, SoundCue:Sound Cue Object Reference, Node:Sound Node Object Reference, DepthPrefix:String) {is=title K2Node_FunctionEntry}
   > Is Valid(InputObject=@n1.Node)
   > n2 = Sequence()
   n3 = Get Node()
   > PrintString(InString=Append!(A=Get DepthPrefix() -> (DepthPrefix:String), B=GetDisplayName(Object=@n3)) -> (ReturnValue))
   n4 = For Each Loop(Exec=@n2.then_1, Array:Array of Sound Node Object References=Get ChildNodes(self=Get Node())) -> (Array Element:Sound Node Object Reference)
   > TraverseGraph_Rec(SoundCue=Get SoundCue(), Node=@n4, DepthPrefix=Append(A="     ", B=Get DepthPrefix())) {is=title K2Node_CallFunction}
  ]}
]
variables [
 CurrentAttenuation {
  type=SoundAttenuationSettings
  default {
   absorptionMethod=Linear
   attenuationShape=Sphere
   audioLinkSettingsOverride=null
   bApplyNormalizationToStereoSounds=false
   bAttenuate=true
   bAttenuateWithLPF=false
   bEnableFocusInterpolation=false
   bEnableListenerFocus=false
   bEnableLogFrequencyScaling=false
   bEnableOcclusion=false
   bEnablePriorityAttenuation=false
   bEnableReverbSend=true
   bEnableSendToAudioLink=true
   bEnableSourceDataOverride=false
   bEnableSubmixSends=false
   bSpatialize=true
   bUseComplexCollisionForOcclusion=false
   binauralRadius=0
   coneOffset=0
   coneSphereFalloffDistance=0
   coneSphereRadius=0
   dBAttenuationAtMax=-60
   distanceAlgorithm=Linear
   falloffDistance=3600
   falloffMode=Continues
   focusAttackInterpSpeed=1
   focusAzimuth=30
   focusDistanceScale=1
   focusPriorityScale=1
   focusReleaseInterpSpeed=1
   focusVolumeAttenuation=1
   hPFFrequencyAtMax=0
   hPFFrequencyAtMin=0
   lPFFrequencyAtMax=20000
   lPFFrequencyAtMin=20000
   lPFRadiusMax=6000
   lPFRadiusMin=3000
   manualPriorityAttenuation=1
   manualReverbSendLevel=0
   nonFocusAzimuth=60
   nonFocusDistanceScale=1
   nonFocusPriorityScale=1
   nonFocusVolumeAttenuation=1
   nonSpatializedRadiusEnd=0
   nonSpatializedRadiusMode=OmniDirectional
   nonSpatializedRadiusStart=0
   occlusionInterpolationTime=0.1
   occlusionLowPassFilterFrequency=20000
   occlusionTraceChannel=ECC_Visibility
   occlusionVolumeAttenuation=1
   priorityAttenuationDistanceMax=4000
   priorityAttenuationDistanceMin=400
   priorityAttenuationMax=1
   priorityAttenuationMethod=Linear
   priorityAttenuationMin=1
   reverbDistanceMax=4000
   reverbDistanceMin=400
   reverbSendMethod=Linear
   reverbWetLevelMax=0.95
   reverbWetLevelMin=0.3
   spatializationAlgorithm=SPATIALIZATION_Default
   stereoSpread=200
   attenuationShapeExtents=<400 0 0>
   customAttenuationCurve {
    externalCurve=null
    editorCurveData {defaultValue=3.4028234663852886e+38, postInfinityExtrap=RCCE_Constant, preInfinityExtrap=RCCE_Constant, keys []}
   }
   customHighpassAirAbsorptionCurve {
    externalCurve=null
    editorCurveData {defaultValue=3.4028234663852886e+38, postInfinityExtrap=RCCE_Constant, preInfinityExtrap=RCCE_Constant, keys []}
   }
   customLowpassAirAbsorptionCurve {
    externalCurve=null
    editorCurveData {defaultValue=3.4028234663852886e+38, postInfinityExtrap=RCCE_Constant, preInfinityExtrap=RCCE_Constant, keys []}
   }
   customPriorityAttenuationCurve {
    externalCurve=null
    editorCurveData {defaultValue=3.4028234663852886e+38, postInfinityExtrap=RCCE_Constant, preInfinityExtrap=RCCE_Constant, keys []}
   }
   customReverbSendCurve {
    externalCurve=null
    editorCurveData {defaultValue=3.4028234663852886e+38, postInfinityExtrap=RCCE_Constant, preInfinityExtrap=RCCE_Constant, keys []}
   }
   pluginSettings {
    occlusionPluginSettingsArray []
    reverbPluginSettingsArray []
    sourceDataOverridePluginSettingsArray []
    spatializationPluginSettingsArray []
   }
   submixSendSettings []
  }
 }
 Destination {type=string, default=""}
 NewSetName {type=string, default=""}
 NewSetPath {type=string, default=""}
 SoundClassName {type=string, default=""}
 Test {default=()}
]
