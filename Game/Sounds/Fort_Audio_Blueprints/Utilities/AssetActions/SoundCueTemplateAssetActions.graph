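# Editor utility Blueprint that adds a scripted asset action for Fortnite sound cue templates.
# Its ReloadDefaultSettings function iterates over the selected assets and casts each to FortSoundCueTemplateBase.
# Each matching template is registered with TransactObject for the editor transaction system, then has its own ReloadDefaultSettings called.
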EditorUtilityBlueprint /Game/Sounds/Fort_Audio_Blueprints/Utilities/AssetActions/SoundCueTemplateAssetActions
graphs [
 EventGraph {links [], nodes []}
 ReloadDefaultSettings {nodes [
   ReloadDefaultSettings() {is=title K2Node_FunctionEntry}
   > n1 = Iterate Over Selected Assets()
   > n2 = Cast To FortSoundCueTemplateBase!(execute, Object=@n1) -> (then, CastFailed, AsFort Sound Cue Template Base:Fort Sound Cue Template Base Object Reference) {is=title K2Node_DynamicCast}
   > TransactObject(Object=@n2)
   > ReloadDefaultSettings(self:Fort Sound Cue Template Base Object Reference=@n2) {is=title K2Node_CallFunction}
  ]}
]
variables []
