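# Editor utility asset actions for batch-editing and auditing audio in Level Sequences.
# Events replace sounds by asset or name substring, set looping, and extend clip ends by percentage, clamped before the following section.
# Edits act on selected assets or selected sections, using shared track/section loop macros and editor-property writes.
# Sound-class scans recurse through sub-sequence tracks via AnalyzeSequence, printing per-sequence counts of exact class matches.
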EditorUtilityBlueprint /Game/Sounds/Fort_Audio_Blueprints/Utilities/AssetActions/SoundLevelSequenceAssetActions
graphs [
 AnalyzeSequence {nodes [
   n1 = AnalyzeSequence!() -> (then, Sequence:Movie Scene Sequence Object Reference) {is=title K2Node_FunctionEntry}
   > PrintSoundsInSequence!(execute, self, LevelSequence:Movie Scene Sequence Object Reference=@n1) -> (then)
   > n2 = GetTracks!(execute, Sequence:Movie Scene Sequence Object Reference=@n1) -> (then, ReturnValue:Array of Movie Scene Track Object References)
   > n3 = For Each Loop(Array:Array of Movie Scene Track Object References=@n2) -> (Array Element:Movie Scene Track Object Reference)
   > n4 = GetLevelSequencesInSubSequenceTrack!(execute, self, InSubSequenceTrack:Movie Scene Track Object Reference=@n3) -> (then, ReturnLevelSequences:Array of Movie Scene Sequence Object References)
   > n5 = For Each Loop(Array:Array of Movie Scene Sequence Object References=@n4) -> (Array Element:Movie Scene Sequence Object Reference)
   > AnalyzeSequence!(execute, self, Sequence:Movie Scene Sequence Object Reference=@n5) -> (then) {is=title K2Node_CallFunction}
  ]}
 EventGraph {nodes [
   n1 = GetSelectedAssets()
   > n2 = For Each Loop(Array:Array of Object References=@n1) -> (Array Element:Object Reference)
   > n3 = Cast To LevelSequence!(execute, Object=@n2) -> (then, CastFailed, AsLevel Sequence:Level Sequence Object Reference) {is=title K2Node_DynamicCast}
   n4 = AddTrack!(
    execute=@n10.then
    Sequence:Movie Scene Sequence Object Reference=@n3
    TrackType:Movie Scene Track Class Reference=/Script/MovieSceneTracks.MovieSceneAudioTrack
   ) -> (then, ReturnValue:Movie Scene Audio Track Object Reference)
   n5 = AddSection!(execute, Track:Movie Scene Track Object Reference=@n4) -> (then, ReturnValue:Movie Scene Section Object Reference)
   n6 = FindTracksByExactType!(
    execute=@n3.then
    Sequence:Movie Scene Sequence Object Reference=@n3
    TrackType:Movie Scene Track Class Reference=&MovieSceneAudioTrack
   ) -> (then, ReturnValue:Array of Movie Scene Audio Track Object References)
   > n7 = GetSections!(
    execute
    Track:Movie Scene Track Object Reference="Get (a copy)"(Array:Array of Movie Scene Audio Track Object References=@n6, Dimension 1=1) -> (Output:Movie Scene Audio Track Object Reference)
   ) -> (then, ReturnValue:Array of Movie Scene Section Object References)
   n8 = "Get (a copy)"(Array:Array of Movie Scene Section Object References=@n7) -> (Output:Movie Scene Section Object Reference)
   > n9 = GetStartFrameSeconds!(execute, Section:Movie Scene Section Object Reference=@n8) -> (then, ReturnValue:float)
   > n10 = GetEndFrameSeconds!(execute, Section:Movie Scene Section Object Reference=@n8) -> (then, ReturnValue:float)
   SetRangeSeconds!(execute=@n5.then, Section:Movie Scene Section Object Reference=@n5, StartTime:float=@n9, EndTime:float=@n10) -> (then)
   n11 = "Get (a ref)"(Array:Array of Names=Get Property Name() -> (Property Name:Array of Names), Dimension 1=0) -> (Output:"Name (by ref)")
   > n12 = GetEditorProperty!(execute:Exec, Object:Object Reference=@n8, PropertyName:Name=@n11) -> (then:Exec, PropertyValue:Sound Base Object Reference, ReturnValue:bool)
   > SetEditorProperty(Object=@n5, PropertyName=@n11, PropertyValue:Sound Base Object Reference=@n12)
   n13 = ReplaceAudioSectionsByAsset!() -> (OutputDelegate, then, SoundsToReplace:Array of Sound Base Object References, Replacement:Sound Base Object Reference) {is=title K2Node_CustomEvent}
   > n14 = For Each Audio Tracks in Selected Sequences!(Exec:Exec) -> (LoopBody:Exec, AudioTracks:Array of Movie Scene Audio Track Object References, Completed:Exec) {is=title K2Node_MacroInstance}
   > n15 = For Each Audio Section!(Exec:Exec, MovieSceneAudioTracks:Array of Movie Scene Audio Track Object References=@n14) -> (
    LoopBody:Exec
    Movie Scene Audio Section:Movie Scene Audio Section Object Reference
    CurrentAudioTrack:Movie Scene Audio Track Object Reference
    Completed:Exec
   ) {is=title K2Node_MacroInstance}
   n16 = Get AssetIndex() -> (AssetIndex:int=0)
   n17 = Get CurrentSequence() -> (CurrentSequence:Level Sequence Object Reference)
   PrintString(
    execute=[@n15.Completed, @n47.Completed]
    InString=Append!(A="Replaced ", B="To String (Integer)"(InInt=@n16), C:String=" audio assets in ", D:String=GetDisplayName(Object=@n17)) -> (ReturnValue)
    Duration=5
   )
   > Set AssetIndex(AssetIndex:int=0) -> (Output_Get:int=0)
   Branch(
    execute=@n15.LoopBody
    Condition=Contains Item(
     TargetArray:Array of Sound Base Object References=@n13.SoundsToReplace
     ItemToFind:"Sound Base Object Reference (by ref)"=GetSound!(self:Movie Scene Audio Section Object Reference=@n15) -> (ReturnValue:Sound Base Object Reference)
    )
   )
   > SetSound!(execute, self:Movie Scene Audio Section Object Reference=@n15, InSound:Sound Base Object Reference=@n13.Replacement) -> (then)
   > n18 = Increment Int(Value=Get AssetIndex() -> (AssetIndex=0))
   > Set AssetIndex(AssetIndex=@n18) -> (Output_Get=0)
   n19 = ExtendAllAudioClips!() -> (OutputDelegate, then, ExtendFramesByPercentFloat:double=~) {is=title K2Node_CustomEvent}
   > Set ExtendFramesByPercentFloat(ExtendFramesByPercentFloat:double=@n19.ExtendFramesByPercentFloat) -> (Output_Get:double)
   > n20 = GetAudioTracks!(execute, self) -> (then, MovieSceneAudioTracks:Array of Movie Scene Audio Track Object References, bIsValid:bool)
   > Branch(Condition=@n20.bIsValid)
   > n21 = For Each Loop(Array:Array of Movie Scene Audio Track Object References=@n20) -> (Array Element:Movie Scene Audio Track Object Reference)
   > n22 = GetSections(Track=@n21)
   > n23 = For Each Loop(Array:Array of Movie Scene Section Object References=@n22) -> (Array Element:Movie Scene Section Object Reference)
   n24 = int + int(A=@n23.Array Index, B=1) -> (ReturnValue=0) {is=title K2Node_CommutativeAssociativeBinaryOperator}
   n25 = Is Valid Index!(TargetArray:Array of Movie Scene Section Object References=@n22, IndexToTest:int=@n24) -> (ReturnValue:bool)
   > n26 = Branch(Condition=@n25)
   > n27 = GetStartFrame!(execute, Section:Movie Scene Section Object Reference="Get (a copy)"(Array=@n22, Dimension 1=@n24)) -> (then, ReturnValue:int=0)
   n28 = GetEndFrame!(execute=[@n27.then, @n26.else], Section:Movie Scene Section Object Reference=@n23) -> (then, ReturnValue:int=0)
   n29 = "Absolute (Float)"(A=float / float(A=Get ExtendFramesByPercentFloat() -> (ExtendFramesByPercentFloat:double), B=100) {is=title K2Node_CallFunction})
   > Branch(Condition="float >= float"(A=float * float(A=@n29, B=100), B=5) {is=title K2Node_CallFunction})
   n30 = Truncate(A=float * float(A="To Float (Integer)"(InInt=@n28), B=@n29))
   n31 = Select(
    Option 0:int=int - int(A=@n28, B=@n30) {is=title K2Node_CallFunction}
    Option 1:int=int + int(A=@n28, B=@n30) -> (ReturnValue=0)
    Index="float > float"(A=Get ExtendFramesByPercentFloat(), B=0) {is=title K2Node_CallFunction}
   ) -> (ReturnValue:int)
   n32 = int - int(A=@n27, B=1)
   > SetEndFrame!(
    execute
    Section:Movie Scene Section Object Reference=@n23
    EndFrame:int="Clamp (Integer)"(Value=@n31, Min=0, Max=Select(Option 0=@n31, Option 1=@n32, Index=@n25)) -> (ReturnValue=0)
   ) -> (then)
   > PrintString(InString=Hello)
   n33 = SetLoopingOnAllAudioClips!() -> (OutputDelegate, then, IsLooping:bool=~) {is=title K2Node_CustomEvent}
   > Set IsLooping(IsLooping:bool=@n33.IsLooping) -> (Output_Get:bool)
   > n34 = GetAudioTracks()
   > Branch(Condition=@n34.bIsValid)
   > n35 = For Each Loop(Array:Array of Movie Scene Audio Track Object References=@n34) -> (Array Element:Movie Scene Audio Track Object Reference)
   > n36 = GetSections(Track=@n35)
   > n37 = For Each Loop(Array:Array of Movie Scene Section Object References=@n36) -> (Array Element:Movie Scene Section Object Reference)
   > n38 = SetEditorProperty(Object=@n37, PropertyName=bLooping, PropertyValue:bool=Get IsLooping() -> (IsLooping:bool))
   > PrintString(InString="To String (Boolean)"(InBool=@n38))
   n39 = ExtendSelectedAudioClips!() -> (OutputDelegate, then, ExtendFramesByPercentFloat:double=~) {is=title K2Node_CustomEvent}
   > Set ExtendFramesByPercentFloat(ExtendFramesByPercentFloat=@n39.ExtendFramesByPercentFloat)
   > Branch(Condition=Get Selected Sections!() -> (IsSelectedSectionValid:bool) {is=title K2Node_CallFunction})
   n40 = GetSelectedSections!() -> (ReturnValue:Array of Movie Scene Section Object References)
   > n41 = For Each Loop(Array=@n40)
   > ExtendFramesByPercent!(
    execute
    self
    CurrentSection:Movie Scene Section Object Reference=@n41
    CurrentIndex:int=@n41.Array Index
    SectionArray:Array of Movie Scene Section Object References=@n40
   ) -> (then)
   n42 = SetLoopingOnSelectedAudioClips!() -> (OutputDelegate, then, IsLooping:bool=~) {is=title K2Node_CustomEvent}
   > Set IsLooping(IsLooping=@n42.IsLooping)
   > Branch(Condition=Get Selected Sections())
   > n43 = For Each Loop(Array=GetSelectedSections())
   > n44 = SetEditorProperty(Object=@n43, PropertyName=bLooping, PropertyValue=Get IsLooping())
   > PrintString(InString="To String (Boolean)"(InBool=@n44))
   n45 = ReplaceAudioSectionsContainingString!() -> (OutputDelegate, then, ContainingString:String, Replacement:Sound Base Object Reference) {is=title K2Node_CustomEvent}
   > n46 = For Each Audio Tracks in Selected Sequences()
   > n47 = For Each Audio Section(MovieSceneAudioTracks=@n46)
   n48 = GetSound(self=@n47)
   > Branch(Condition=Contains(SearchIn=GetObjectName(Object=@n48), Substring=@n45.ContainingString))
   > SetSound(self=@n47, InSound=@n45.Replacement)
   > n49 = Increment Int(Value=Get AssetIndex() -> (AssetIndex=0))
   > Set AssetIndex(AssetIndex=@n49) -> (Output_Get=0)
   n50 = ScanForSoundBaseClass!() -> (OutputDelegate, then, SoundBaseType:Sound Base Class Reference) {is=title K2Node_CustomEvent}
   > n51 = Set SoundBaseClass(SoundBaseClass:Sound Base Class Reference=@n50.SoundBaseType) -> (Output_Get:Sound Base Class Reference)
   Set ScannedSequenceIndex(execute=[@n60.then, @n51.then], ScannedSequenceIndex:int=0) -> (Output_Get:int=0)
   > n52 = GetSelectedAssets()
   > n53 = For Each Loop(Array:Array of Object References=@n52) -> (Array Element:Object Reference)
   n54 = Get ScannedSequenceIndex() -> (ScannedSequenceIndex:int=0)
   n55 = Get Class Display Name(Class=Get SoundBaseClass() -> (SoundBaseClass:Sound Base Class Reference))
   PrintString(
    execute=@n53.Completed
    InString=Append!(
     A="[SCANNED "
     B="To String (Integer)"(InInt=@n54)
     C=" SEQUENCES FOR "
     D=ToUpper!(SourceString:String=@n55) -> (ReturnValue:String)
     E:String=" ASSETS]"
    ) -> (ReturnValue)
    Duration=5
   )
   n56 = GetDisplayName(Object=@n53)
   > PrintString(InString=Append!(A="[FINISHING SCAN OF ", B=ToUpper(SourceString=@n56), C="]") -> (ReturnValue), Duration=5)
   n57 = Get Class Display Name(Class=Get SoundBaseClass())
   n58 = GetDisplayName(Object=@n53)
   PrintString(
    execute=@n53.LoopBody
    InString=Append!(A="[SCANNING ", B=ToUpper(SourceString=@n58), C=" AND ALL SUB-SEQUENCES FOR ", D:String=ToUpper(SourceString=@n57), E:String=" ASSETS]") -> (ReturnValue)
    Duration=5
   )
   > n59 = Cast To MovieSceneSequence!(execute, Object=@n53) -> (then, CastFailed, AsMovie Scene Sequence:Movie Scene Sequence Object Reference) {is=title K2Node_DynamicCast}
   > AnalyzeSequence(Sequence=@n59)
   ScanForSoundWaves() {is=title K2Node_CustomEvent}
   > n60 = Set SoundBaseClass(SoundBaseClass=/Script/Engine.SoundWave)
  ]}
 ExtendFramesByPercent {nodes [
   n1 = ExtendFramesByPercent!() -> (then, CurrentSection:Movie Scene Section Object Reference, CurrentIndex:int, SectionArray:Array of Movie Scene Section Object References) {is=title K2Node_FunctionEntry}
   n2 = int + int(A=@n1.CurrentIndex, B=1) -> (ReturnValue=0)
   n3 = Is Valid Index(TargetArray=@n1.SectionArray, IndexToTest=@n2)
   > n4 = Branch(Condition=@n3)
   > n5 = GetStartFrame(Section="Get (a copy)"(Array=@n1.SectionArray, Dimension 1=@n2)) -> (ReturnValue=0)
   n6 = GetEndFrame(execute=[@n5.then, @n4.else], Section=@n1) -> (ReturnValue=0)
   n7 = Get ExtendFramesByPercentFloat()
   > Branch(Condition="Not Equal (Float)"!(A:double=@n7, B:double=0) -> (ReturnValue:bool) {is=title K2Node_CallFunction})
   n8 = Get ExtendFramesByPercentFloat()
   n9 = Truncate(A=float * float(A="To Float (Integer)"(InInt=@n6), B="Absolute (Float)"(A=@n8)))
   n10 = Select(Option 0=int - int(A=@n6, B=@n9), Option 1=int + int(A=@n6, B=@n9) -> (ReturnValue=0), Index="float > float"(A=@n7, B=0))
   n11 = "Clamp (Integer)"(Value=@n10, Min=@n6, Max=int - int(A=@n5, B=1)) -> (ReturnValue=0)
   > SetEndFrame(Section=@n1, EndFrame=Select(Option 0=@n10, Option 1=@n11, Index=@n3))
   > PrintString(InString="To String (Integer)"(InInt=@n11))
  ]}
 ForEachAudioSection {nodes [
   n1 = Inputs!() -> (Exec:Exec, MovieSceneAudioTracks:Array of Movie Scene Audio Track Object References)
   > n2 = For Each Loop(Array:Array of Movie Scene Audio Track Object References=@n1) -> (Array Element:Movie Scene Audio Track Object Reference)
   > n3 = GetSections(Track=@n2)
   > n4 = For Each Loop(Array:Array of Movie Scene Section Object References=@n3) -> (Array Element:Movie Scene Section Object Reference)
   > n5 = Cast To MovieSceneAudioSection!(execute, Object=@n4) -> (then, CastFailed, AsMovie Scene Audio Section:Movie Scene Audio Section Object Reference) {is=title K2Node_DynamicCast}
   > Outputs!(
    LoopBody:Exec
    Movie Scene Audio Section:Movie Scene Audio Section Object Reference=@n5
    CurrentAudioTrack:Movie Scene Audio Track Object Reference=@n2
    Completed:Exec=@n2.Completed
   )
  ]}
 ForEachAudioTracksInSelectedSequences {nodes [
   Inputs!() -> (Exec)
   > n1 = GetSelectedAssets()
   > n2 = For Each Loop(Array:Array of Object References=@n1) -> (Array Element:Object Reference)
   > n3 = Cast To LevelSequence(Object=@n2)
   > n4 = Set CurrentSequence(CurrentSequence:Level Sequence Object Reference=@n3) -> (Output_Get:Level Sequence Object Reference)
   > n5 = FindTracksByType!(
    execute
    Sequence:Movie Scene Sequence Object Reference=@n4
    TrackType:Movie Scene Track Class Reference=&MovieSceneAudioTrack
   ) -> (then, ReturnValue:Array of Movie Scene Audio Track Object References)
   > Outputs!(LoopBody, AudioTracks:Array of Movie Scene Audio Track Object References=@n5, Completed=@n2.Completed)
  ]}
 Get Selected Sections {nodes [
   Get Selected Sections() {is=title K2Node_FunctionEntry}
   > Return Node!(execute, IsSelectedSectionValid:bool=Is Valid Index(TargetArray=GetSelectedSections()))
  ]}
 GetAudioTracks {nodes [
   GetAudioTracks() {is=title K2Node_FunctionEntry}
   > n1 = GetSelectedAssets()
   > n2 = For Each Loop(Array=@n1)
   > n3 = Cast To LevelSequence(Object=@n2)
   > n4 = FindTracksByType(Sequence=@n3, TrackType=&MovieSceneAudioTrack)
   n5 = IsValid(Object=@n3)
   > Branch(Condition=@n5)
   > Return Node!(execute, MovieSceneAudioTracks:Array of Movie Scene Audio Track Object References=@n4, bIsValid:bool=@n5)
  ]}
 GetLevelSequencesInSubSequenceTrack {nodes [
   n1 = GetLevelSequencesInSubSequenceTrack!() -> (then, InSubSequenceTrack:Movie Scene Track Object Reference) {is=title K2Node_FunctionEntry}
   > n2 = GetSections(Track=@n1)
   > n3 = For Each Loop(Array:Array of Movie Scene Section Object References=@n2) -> (Array Element:Movie Scene Section Object Reference)
   Return Node!(
    execute=@n3.Completed
    ReturnLevelSequences:Array of Movie Scene Sequence Object References=Get LevelSequences() -> (LevelSequences:Array of Movie Scene Sequence Object References)
   )
   > n4 = Cast To MovieSceneSubSection!(execute, Object=@n3) -> (then, CastFailed, AsMovie Scene Sub Section:Movie Scene Sub Section Object Reference) {is=title K2Node_DynamicCast}
   > Add(
    TargetArray:Array of Movie Scene Sequence Object References=Get LevelSequences()
    NewItem:"Movie Scene Sequence Object Reference (by ref)"=GetSequence!(self:Movie Scene Sub Section Object Reference=@n4) -> (ReturnValue:Movie Scene Sequence Object Reference)
   ) {is=title K2Node_CallArrayFunction}
  ]}
 PrintSoundsInSequence {nodes [
   n1 = PrintSoundsInSequence!() -> (then, LevelSequence:Movie Scene Sequence Object Reference) {is=title K2Node_FunctionEntry}
   > n2 = FindTracksByType(Sequence=@n1, TrackType=&MovieSceneAudioTrack)
   > n3 = For Each Audio Section(MovieSceneAudioTracks=@n2)
   Increment Int(" "=@n3.Completed, Value=Get ScannedSequenceIndex() -> (ScannedSequenceIndex=0))
   > Branch(Condition="integer > integer"(A=Get AssetIndex() -> (AssetIndex=0), B=0) {is=title K2Node_CallFunction})
   n4 = Get Class Display Name(Class=Get SoundBaseClass())
   n5 = GetDisplayName(Object=@n1)
   n6 = "To String (Integer)"(InInt=Get AssetIndex() -> (AssetIndex=0))
   > PrintString(
    InString=Append!(A="     ", B=Append!(A=@n5, B=" contains ", C:String=@n6, D:String=" ", E:String=@n4, F:String=(s)) -> (ReturnValue)) -> (ReturnValue)
    TextColor=(R=1,G=0.1066,B=0,A=1)
    Duration=5
   )
   > Set AssetIndex(AssetIndex=0) -> (Output_Get=0)
   n7 = GetSound(self=@n3)
   Branch(
    execute=@n3.LoopBody
    Condition="Equal (Class)"!(A:Object Class Reference=Get SoundBaseClass(), B:Object Class Reference=Get Class(Object=@n7) -> (ReturnValue:Sound Base Class Reference)) -> (ReturnValue:bool) {is=title K2Node_CallFunction}
   )
   > Increment Int(Value=Get AssetIndex() -> (AssetIndex=0))
  ]}
]
variables [
 AssetIndex {type=integer}
 AudioTracksInSelectedSequences {type="array<MovieSceneAudioTrack>", default []}
 CurrentSequence {type=LevelSequence, default=null}
 ExtendFramesByPercentFloat {type=number, default=0.02}
 IsLooping {type=boolean}
 Property Name {type="array<string>", default [Sound,bLooping]}
 ScannedSequenceIndex {type=integer}
 SoundBaseClass {type=Class, default=null}
]
