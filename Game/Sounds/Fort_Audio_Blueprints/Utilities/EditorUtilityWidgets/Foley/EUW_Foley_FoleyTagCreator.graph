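# Editor utility widget that bulk-adds foley tags to a chosen Sound Library by appending each surface type, such as Dirt or Metal, to a base tag.
# Editing the base tag previews the resulting names.
# The button maps each found tag to the Sound variable, null by default and not exposed in the UI.
# Optionally, missing tags are written to a selected default or plugin tag INI, then the SetGameplayTag event retries the lookup.
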EditorUtilityWidgetBlueprint /Game/Sounds/Fort_Audio_Blueprints/Utilities/EditorUtilityWidgets/Foley/EUW_Foley_FoleyTagCreator
graphs [EventGraph {nodes [
   "On Pressed (AddToLibButton)"() {is=title K2Node_ComponentBoundEvent}
   > n1 = For Each Loop(Array=Get SurfaceTypes() -> (SurfaceTypes:Array of Strings))
   Branch(execute=@n1.Completed, Condition=Get CreateNewTag() -> (CreateNewTag:bool))
   n2 = Get BaseTag() -> (BaseTag:Sound Library Tag Structure)
   n3 = Append!(A="To String (Name)"(InName=GetTagName!(GameplayTag:"Gameplay Tag Structure (by ref)"=@n2) -> (ReturnValue:Name=None)), B=., C:String=@n1) -> (ReturnValue)
   n4 = RequestGameplayTag!(execute=[@n11.then, @n1.LoopBody], TagName:String=@n3, bErrorIfNotFound:bool=true) -> (then, OutTag:Gameplay Tag Structure, ReturnValue:bool)
   > n5 = Branch(Condition=@n4.ReturnValue)
   > Add!(
    execute
    TargetMap:Map of Gameplay Tag Structures to Sound Base Soft Object References=Get Sounds!(self:Sound Library Object Reference=Get SoundLibrary() -> (SoundLibrary:Sound Library Object Reference)) -> (Sounds:Map of Gameplay Tag Structures to Sound Base Soft Object References)
    Key:"Gameplay Tag Structure (by ref)"=@n4
    Value:"Sound Base Soft Object Reference (by ref)"=Get Sound() -> (Sound:Sound Base Soft Object Reference)
   ) -> (then) {is=title K2Node_CallFunction}
   n6 = Branch(execute=@n5.else, Condition=Get CreateNewTag())
   n7 = Get ComboBox() -> (ComboBox:Editor Utility Combo Box String Object Reference)
   > n8 = AddNewGameplayTagToINI!(
    execute
    NewTag:String=@n3
    Comment:String
    TagSourceName:Name=String To Name(InString=GetSelectedOption!(self:"Combo Box (String) Object Reference"=@n7) -> (ReturnValue:String))
    bIsRestrictedTag:bool=false
    bAllowNonRestrictedChildren:bool=true
   ) -> (then, ReturnValue:bool)
   > PrintString(InString=Append!(A=SelectString(A="Creation Success: ", B="Creation Failed: ", bPickA=@n8), B=@n3) -> (ReturnValue))
   > Branch(Condition=@n8)
   > SetGameplayTag()
   PrintString(execute=@n6.else, InString=Tag either already exists or is not created in database)
   "On Property Changed (TagView)"!() -> (OutputDelegate, then, PropertyName:Name=None) {is=title K2Node_ComponentBoundEvent}
   > Set Text(self:"Editable Text (Multi- Line) Object Reference"=Get FoleyListText() -> (FoleyListText:Editor Utility Multi Line Editable Text Object Reference))
   > Set FutureTagNames(FutureTagNames:String) -> (Output_Get:String)
   > n9 = For Each Loop(Array=Get SurfaceTypes())
   Set Text(execute=@n9.Completed, Text="To Text (String)"(InString=Get FutureTagNames() -> (FutureTagNames:String)), self=Get FoleyListText())
   n10 = "To String (Name)"(InName=GetTagName(GameplayTag=Get BaseTag()) -> (ReturnValue=None))
   Set FutureTagNames(execute=@n9.LoopBody, FutureTagNames=Append!(A=Get FutureTagNames(), B="\r\n", C:String=Append(A=@n10, B=., C=@n9)) -> (ReturnValue))
   "On Property Changed (CreateNewTagView)"!() -> (OutputDelegate, then, PropertyName:Name=None) {is=title K2Node_ComponentBoundEvent}
   > SetVisibility(
    self=Get TagsSourceData() -> (TagsSourceData:Vertical Box Object Reference)
    InVisibility=Select(Option 0=Collapsed, Option 1=Visible, Index=Get CreateNewTag())
   )
   n11 = SetGameplayTag() {is=title K2Node_CustomEvent}
   Event PreConstruct()
   > SetObject(
    self=[
     Get SoundsView() -> (SoundsView:Single Property View Object Reference)
     Get CreateNewTagView() -> (CreateNewTagView:Single Property View Object Reference)
     Get TagView() -> (TagView:Single Property View Object Reference)
    ]
    NewObject=Self-Reference()
   )
   Event Construct()
   > n12 = GetEnabledPluginNames!(execute) -> (then, ReturnValue:Array of Strings)
   > SortStringArray!(execute, TargetArray:Array of Strings=@n12, bStableSort:bool=false, SortOrder:EArraySortOrder Enum=Ascending) -> (then)
   > n13 = For Each Loop(Array=@n12)
   > AddOption!(execute, self:"Combo Box (String) Object Reference"=Get ComboBox(), Option:String=Append!(A=@n13, B=Tags.ini) -> (ReturnValue)) -> (then)
  ]}]
variables [
 BaseTag {type=SoundLibraryTag, default {tagName=None}}
 CreateNewTag {type=boolean}
 FutureTagNames {type=string, default=""}
 Sound {type=SoundBase, default=null}
 SoundLibrary {type=SoundLibrary, default=null}
 SurfaceTypes {type="array<string>", default [Dirt,Grass,Metal,Sand,Snow,Ice,Stone,Water,Wood,Mud,WetConcrete,TreeCanopy,Carpet]}
]
widgets {
 info {rootWidgetClass=&Overlay, widgetCount=15}
 widgets [
  Overlay_141 {widgetClassPath=&Overlay, widget=Overlay_141}
  Border_1 {
   widgetClassPath=&Border
   parent=Overlay_141
   widget=Border_1
   properties {
    horizontalAlignment=HAlign_Left
    verticalAlignment=VAlign_Center
    background {
     bIsDynamicallyLoaded=false
     drawAs=Image
     imageType=NoImage
     mirroring=NoMirror
     resourceName=None
     resourceObject=null
     tiling=NoTile
     imageSize=<32 32>
     margin=<ltrb 0 0 0 0>
     outlineSettings {
      bUseBrushTransparency=false
      roundingType=HalfHeightRadius
      width=0
      color {colorUseRule=UseColor_Specified, specifiedColor=<rgba 0 0 0 0>}
      cornerRadii=<0 0 0 0>
     }
     tintColor {colorUseRule=UseColor_Specified, specifiedColor=<rgba 0.0208 0.0208 0.0208 1>}
     uVRegion {bIsValid=false, max=<0 0>, min=<0 0>}
    }
    padding=<ltrb 0 0 0 0>
   }
   slot {class=OverlaySlot}
  }
  Border_0 {
   widgetClassPath=&Border
   parent=Border_1
   widget=Border_0
   properties {background {
     bIsDynamicallyLoaded=false
     drawAs=Border
     imageType=NoImage
     mirroring=NoMirror
     resourceName=None
     resourceObject=null
     tiling=NoTile
     imageSize=<32 32>
     margin=<ltrb 1 1 1 1>
     outlineSettings {
      bUseBrushTransparency=false
      roundingType=HalfHeightRadius
      width=0
      color {colorUseRule=UseColor_Specified, specifiedColor=<rgba 0 0 0 0>}
      cornerRadii=<0 0 0 0>
     }
     tintColor {colorUseRule=UseColor_Specified, specifiedColor=<rgba 0 0 0 1>}
     uVRegion {bIsValid=false, max=<0 0>, min=<0 0>}
    }}
   slot {class=BorderSlot, properties {horizontalAlignment=HAlign_Left, verticalAlignment=VAlign_Center, padding=<ltrb 0 0 0 0>}}
  }
  HorizontalBox_128 {widgetClassPath=&HorizontalBox, parent=Border_0, widget=HorizontalBox_128, slot {class=BorderSlot}}
  SizeBox_1 {
   widgetClassPath=&SizeBox
   parent=HorizontalBox_128
   widget=SizeBox_1
   properties {bOverride_WidthOverride=true, widthOverride=350}
   slot {class=HorizontalBoxSlot, properties {padding=<ltrb 10 10 10 10>}}
  }
  VerticalBox_0 {widgetClassPath=&VerticalBox, parent=SizeBox_1, widget=VerticalBox_0, slot {class=SizeBoxSlot}}
  SoundsView {
   widgetClassPath=&SinglePropertyView
   parent=VerticalBox_0
   widget=SoundsView
   properties {propertyName=SoundLibrary}
   slot {class=VerticalBoxSlot, properties {padding=<ltrb 5 5 5 5>}}
  }
  TagView {
   widgetClassPath=&SinglePropertyView
   parent=VerticalBox_0
   widget=TagView
   properties {nameOverride=Library Tag, propertyName=BaseTag}
   slot {class=VerticalBoxSlot, properties {padding=<ltrb 5 5 5 5>}}
  }
  CreateNewTagView {
   widgetClassPath=&SinglePropertyView
   parent=VerticalBox_0
   widget=CreateNewTagView
   properties {
    nameOverride=Should create new tag if not found?
    propertyName=CreateNewTag
    toolTipText=This is destructive and will create tags if they do not exist. Make sure you are adding them to the right .ini file if this is checked.
   }
   slot {class=VerticalBoxSlot, properties {padding=<ltrb 5 5 5 5>}}
  }
  TagsSourceData {
   widgetClassPath=&VerticalBox
   bIsVariable=true
   parent=VerticalBox_0
   widget=TagsSourceData
   properties {visibility=Collapsed}
   slot {class=VerticalBoxSlot, properties {horizontalAlignment=HAlign_Left, verticalAlignment=VAlign_Center, padding=<ltrb 5 5 5 5>}}
  }
  EditorUtilityEditableText_161 {
   widgetClassPath=/Script/Blutility.EditorUtilityEditableText
   bIsVariable=true
   parent=TagsSourceData
   widget=EditorUtilityEditableText_161
   properties {text="Source Tag Name (Add to Plug-in .ini file):"}
   slot {class=VerticalBoxSlot}
  }
  ComboBox {
   widgetClassPath=/Script/Blutility.EditorUtilityComboBoxString
   bIsVariable=true
   parent=TagsSourceData
   widget=ComboBox
   properties {
    selectedOption=DefaultGameplayTags.ini
    contentPadding=<ltrb 5 5 5 5>
    defaultOptions [DefaultGameplayTags.ini]
    foregroundColor {colorUseRule=UseColor_Specified, specifiedColor=<rgba 0.5271 0.5271 0.5271 1>}
   }
   slot {class=VerticalBoxSlot}
  }
  AddToLibButton {
   widgetClassPath=&EditorUtilityButton
   parent=VerticalBox_0
   widget=AddToLibButton
   slot {class=VerticalBoxSlot, properties {horizontalAlignment=HAlign_Left, padding=<ltrb 5 5 5 5>}}
  }
  UEFN_TextBlock_C_1 {
   widgetClassPath=&UEFN_TextBlock
   parent=AddToLibButton
   widget=UEFN_TextBlock_C_1
   properties {
    justification=Center
    text=Add Foley Tag Data
    font {
     bForceMonospaced=false
     bMaterialIsStencil=false
     fontMaterial=null
     fontObject=/LauncherSocial/BrutalType
     letterSpacing=0
     monospacedWidth=1
     size=9
     skewAmount=0
     typefaceFontName=Regular
     outlineSettings {
      bApplyOutlineToDropShadows=false
      bMiteredCorners=false
      bSeparateFillAlpha=false
      outlineMaterial=null
      outlineSize=0
      outlineColor=<rgba 0 0 0 1>
     }
    }
    shadowColorAndOpacity=<rgba 0 0 0 1>
   }
   slot {class=ButtonSlot}
  }
  FoleyListText {
   widgetClassPath=/Script/Blutility.EditorUtilityMultiLineEditableText
   bIsVariable=true
   parent=VerticalBox_0
   widget=FoleyListText
   slot {class=VerticalBoxSlot, properties {padding=<ltrb 5 5 5 5>}}
  }
 ]
}
