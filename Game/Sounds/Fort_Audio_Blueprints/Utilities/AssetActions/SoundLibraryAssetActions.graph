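# Editor asset-action utility for batch editing selected SoundLibrary assets and their gameplay-tag-to-sound maps.
# Tag actions remove, exactly swap or substring-replace keys by rebuilding the Sounds map, then transacting and writing it back.
# SplitAssetByChildTags creates a library per child tag at a chosen depth through inline Python, copying settings and matching entries.
# ClearParent resets Parent and reparents to SoundLibrary; AddSurfaceFoleyTags opens the foley tag creator widget.
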EditorUtilityBlueprint /Game/Sounds/Fort_Audio_Blueprints/Utilities/AssetActions/SoundLibraryAssetActions
graphs [
 AddSurfaceFoleyTags {nodes [
   AddSurfaceFoleyTags() {is=title K2Node_FunctionEntry}
   > SpawnAndRegisterTab(
    self=Get EditorUtilitySubsystem()
    InBlueprint=/Game/Sounds/Fort_Audio_Blueprints/Utilities/EditorUtilityWidgets/Foley/EUW_Foley_FoleyTagCreator
   )
  ]}
 ClearParent {nodes [
   n1 = Iterate Over Selected Assets(Trigger=@n3.then)
   n2 = Cast To SoundLibrary!(execute, Object=@n1) -> (then, CastFailed, AsSound Library:Sound Library Object Reference) {is=title K2Node_DynamicCast}
   > SetEditorProperty(
    Object=@n2
    PropertyName=Parent
    PropertyValue:Sound Library Object Reference=Get NoneLibrary() -> (NoneLibrary:Sound Library Object Reference)
   )
   n3 = ClearParent() {is=title K2Node_FunctionEntry}
   n4 = GetBlueprintAsset(execute=@n1.Iteration, Object=@n1)
   > ReparentBlueprint(Blueprint=@n4, NewParentClass=/Script/SoundLibrary)
  ]}
 EventGraph {links [], nodes []}
 GetChildrenWithDepth {nodes [
   n1 = GetChildrenWithDepth!() -> (then, ParentTag:"Gameplay Tag Structure (by ref)", Depth:int) {is=title K2Node_FunctionEntry}
   > n2 = RequestGameplayTagChildren!(execute, Parent:"Gameplay Tag Structure (by ref)"=@n1) -> (then, ReturnValue_GameplayTags:Array of Gameplay Tag Structures)
   > n3 = For Each Loop(Array:Array of Gameplay Tag Structures=@n2) -> (Array Element:Gameplay Tag Structure)
   n4 = For Each Loop(Exec=@n3.Completed, Array:Array of Strings=Get FilteredTagStrings() -> (FilteredTagStrings:Array of Strings)) -> (Array Element:String)
   Return Node!(
    execute=@n4.Completed
    FilteredChildTags:Array of Gameplay Tag Structures=Get FilteredTags() -> (FilteredTags:Array of Gameplay Tag Structures)
   )
   > n5 = RequestGameplayTag!(execute, TagName:String=@n4, bErrorIfNotFound:bool=true) -> (then, OutTag:Gameplay Tag Structure, ReturnValue:bool)
   > Branch(Condition=@n5.ReturnValue)
   > Add(TargetArray:Array of Gameplay Tag Structures=Get FilteredTags(), NewItem:"Gameplay Tag Structure (by ref)"=@n5) {is=title K2Node_CallArrayFunction}
   n6 = GetDebugStringFromGameplayTag!(GameplayTag:Gameplay Tag Structure=@n3) -> (ReturnValue:String)
   n7 = GetDebugStringFromGameplayTag(GameplayTag=@n1)
   n8 = ParseIntoArray(
    SourceString=Split(SourceString=@n6, InStr=Split(SourceString=@n7, InStr=., SearchDir=FromEnd).RightS, SearchDir=FromEnd).RightS
    Delimiter=.
    CullEmptyStrings=false
   )
   n9 = Get Depth() -> (Depth:int=0)
   Branch(execute=@n3.LoopBody, Condition="Equal (Integer)"(A=Length(TargetArray:Array of Strings=@n8), B=int + int(A=@n9)))
   > Add(TargetArray:Array of Strings=Get FilteredTagStrings(), NewItem:"String (by ref)"=@n6)
  ]}
 RemoveTags {nodes [
   RemoveTags!() -> (then, TagsToRemove:Gameplay Tag Container Structure, ExactMatch:bool=~) {is=title K2Node_FunctionEntry}
   > n1 = Iterate Over Selected Assets()
   > Clear!(
    execute
    TargetMap:Map of Gameplay Tag Structures to Sound Base Soft Object References=Get LocalSounds() -> (LocalSounds:Map of Gameplay Tag Structures to Sound Base Soft Object References)
   ) -> (then) {is=title K2Node_CallFunction}
   > Set bShouldDirty(bShouldDirty:bool=false) -> (Output_Get:bool)
   > n2 = Cast To SoundLibrary(Object=@n1)
   n3 = Get Sounds!(self:Sound Library Object Reference=@n2) -> (Sounds:Map of Gameplay Tag Structures to Sound Base Soft Object References)
   > n4 = Keys!(execute, TargetMap:Map of Gameplay Tag Structures to Sound Base Soft Object References=@n3) -> (then, Keys:Array of Gameplay Tag Structures)
   > n5 = Values!(execute, TargetMap:Map of Gameplay Tag Structures to Sound Base Soft Object References=@n3) -> (then, Values:Array of Sound Base Soft Object References)
   > n6 = For Each Loop(Array:Array of Gameplay Tag Structures=@n4) -> (Array Element:Gameplay Tag Structure)
   Branch(execute=@n6.Completed, Condition=Get bShouldDirty() -> (bShouldDirty:bool))
   > TransactObject(Object=@n2)
   > SetEditorProperty(Object=@n2, PropertyName=Sounds, PropertyValue:Map of Gameplay Tag Structures to Sound Base Soft Object References=Get LocalSounds())
   n7 = Branch(
    execute=@n6.LoopBody
    Condition=MatchesAnyTags!(
     TagOne:Gameplay Tag Structure=@n6
     OtherContainer:"Gameplay Tag Container Structure (by ref)"=Get TagsToRemove() -> (TagsToRemove:Gameplay Tag Container Structure)
     bExactMatch:bool=Get ExactMatch() -> (ExactMatch:bool)
    ) -> (ReturnValue:bool)
   )
   > Set bShouldDirty(bShouldDirty=true)
   Add!(
    execute=@n7.else
    TargetMap:Map of Gameplay Tag Structures to Sound Base Soft Object References=Get LocalSounds()
    Key:"Gameplay Tag Structure (by ref)"=@n6
    Value:"Sound Base Soft Object Reference (by ref)"="Get (a copy)"(Array:Array of Sound Base Soft Object References=@n5, Dimension 1=@n6.Array Index) -> (Output:Sound Base Soft Object Reference)
   ) -> (then) {is=title K2Node_CallFunction}
  ]}
 SplitAssetByChildTags {nodes [
   n1 = SplitAssetByChildTags!() -> (then, Tag:Gameplay Tag Structure, Depth:int=1) {is=title K2Node_FunctionEntry}
   > Set TagsToCompare(TagsToCompare:Array of Gameplay Tag Structures=GetChildrenWithDepth!(ParentTag:"Gameplay Tag Structure (by ref)"=@n1, Depth:int=@n1.Depth) -> (FilteredChildTags:Array of Gameplay Tag Structures) {is=title K2Node_CallFunction}) -> (Output_Get:Array of Gameplay Tag Structures)
   > n2 = Iterate Over Selected Assets()
   > n3 = Cast To SoundLibrary(Object=@n2)
   > Set CurrentSoundLibrary(CurrentSoundLibrary:Sound Library Object Reference=@n3) -> (Output_Get:Sound Library Object Reference)
   > n4 = For Each Loop(Array=Get TagsToCompare() -> (TagsToCompare:Array of Gameplay Tag Structures))
   > Clear(TargetArray:Array of Gameplay Tag Structures=Get CurrentSplitTags() -> (CurrentSplitTags:Array of Gameplay Tag Structures)) {is=title K2Node_CallArrayFunction}
   > Clear(TargetArray:Array of Sound Base Soft Object References=Get CurrentSplitSounds() -> (CurrentSplitSounds:Array of Sound Base Soft Object References))
   n5 = Get Sounds(self=Get CurrentSoundLibrary() -> (CurrentSoundLibrary:Sound Library Object Reference))
   > n6 = Keys(TargetMap=@n5)
   > n7 = Values(TargetMap=@n5)
   > n8 = For Each Loop(Array=@n6)
   n9 = For Each Loop(Exec=@n8.Completed, Array=Get CurrentSplitTags())
   n10 = Get CurrentSplitTags()
   Branch(execute=@n9.Completed, Condition=Is Not Empty!(TargetArray:Array of Gameplay Tag Structures=@n10) -> (ReturnValue:bool))
   n11 = Length(TargetArray:Array of Gameplay Tag Structures=@n10)
   > PrintString(
    InString=Append!(A="Creating split SoundLibrary with ", B="To String (Integer)"(InInt=@n11), C:String=" elements in it") -> (ReturnValue)
    bPrintToScreen=false
   )
   n12 = Split!(InPath:String=Get Object Path String(Object=Get CurrentSoundLibrary())) -> (PathPart:String, FilenamePart:String, ExtensionPart:String)
   n13 = Split!(SourceString:String=@n12.FilenamePart, InStr:String=_Lib, SearchCase:ESearchCase Enum, SearchDir:ESearchDir Enum=FromStart) -> (LeftS:String, RightS:String, ReturnValue:bool)
   n14 = Append!(
    A=SelectString(A=@n13, B=@n12.FilenamePart, bPickA=@n13.ReturnValue)
    B=_
    C=Split(SourceString=GetDebugStringFromGameplayTag(GameplayTag=@n4), InStr=., SearchDir=FromEnd).RightS
    D:String=_Lib
   ) -> (ReturnValue)
   > n15 = Execute Python Script!(
    execute
    PythonScript="import unreal\r\n\r\nfactory = unreal.SoundLibraryFactory()\r\n\r\nlibrary = unreal.AssetToolsHelpers.get_asset_tools().create_asset(name, path, unreal.SoundLibrary, factory)"
    path:String=@n12
    name:String=@n14
   ) -> (then, ReturnValue, library:Sound Library Object Reference) {inputs [path,name], outputs [library]}
   n16 = Get CurrentSoundLibrary()
   > Set LibraryRequirements!(
    execute
    LibraryRequirements:Audio Gameplay Requirements Structure=Get LibraryRequirements!(self:Sound Library Object Reference=@n16) -> (LibraryRequirements:Audio Gameplay Requirements Structure)
    self:Sound Library Object Reference=@n15.library
   ) -> (then, Output_Get:Audio Gameplay Requirements Structure)
   > SetEditorProperty(
    Object=@n15.library
    PropertyName=Parent
    PropertyValue:Sound Library Object Reference=Get Parent!(self:Sound Library Object Reference=@n16) -> (Parent:Sound Library Object Reference)
   )
   > SetEditorProperty(
    Object=@n15.library
    PropertyName=bCanActAsParent
    PropertyValue:bool=Get bCanActAsParent!(self:Sound Library Object Reference=@n16) -> (bCanActAsParent:bool)
   )
   n17 = Get Sounds(self=@n15.library)
   > Clear(TargetMap=@n17) {is=title K2Node_CallFunction}
   > n18 = For Each Loop(Array=Get CurrentSplitTags())
   > Add(TargetMap=@n17, Key=@n18, Value="Get (a copy)"(Array=Get CurrentSplitSounds(), Dimension 1=@n18.Array Index))
   PrintString(execute=@n9.LoopBody, InString=GetDebugStringFromGameplayTag(GameplayTag=@n9))
   Branch(
    execute=@n8.LoopBody
    Condition=MatchesTag!(TagOne:Gameplay Tag Structure=@n8, TagTwo:Gameplay Tag Structure=@n4, bExactMatch:bool=false) -> (ReturnValue:bool)
   )
   > Add(TargetArray:Array of Gameplay Tag Structures=Get CurrentSplitTags(), NewItem:"Gameplay Tag Structure (by ref)"=@n8) {is=title K2Node_CallArrayFunction}
   > Add(
    TargetArray:Array of Sound Base Soft Object References=Get CurrentSplitSounds()
    NewItem:"Sound Base Soft Object Reference (by ref)"="Get (a copy)"(Array=@n7, Dimension 1=@n8.Array Index)
   )
   Append!(A=@n12, B=/, C=@n14) -> (ReturnValue)
  ]}
 SwapTagExact {nodes [
   SwapTagExact!() -> (then, SwapTag:Gameplay Tag Structure, NewTag:Gameplay Tag Structure) {is=title K2Node_FunctionEntry}
   > n1 = Iterate Over Selected Assets()
   > n2 = Cast To SoundLibrary(Object=@n1)
   n3 = Get Sounds(self=@n2)
   > n4 = Keys(TargetMap=@n3)
   > n5 = Values(TargetMap=@n3)
   > n6 = For Each Loop(Array=@n4)
   TransactObject(execute=@n6.Completed, Object=@n2)
   > SetEditorProperty(Object=@n2, PropertyName=Sounds, PropertyValue:Map of Gameplay Tag Structures to Sound Base Soft Object References=Get LocalSounds())
   n7 = Get SwapTag() -> (SwapTag:Gameplay Tag Structure)
   Add(
    execute=@n6.LoopBody
    TargetMap=Get LocalSounds()
    Key=Select(
     Option 0:"Gameplay Tag Structure (by ref)"=@n6
     Option 1:"Gameplay Tag Structure (by ref)"=Get NewTag() -> (NewTag:Gameplay Tag Structure)
     Index="Equal (GameplayTag)"(A=@n6, B=@n7)
    ) -> (ReturnValue:"Gameplay Tag Structure (by ref)")
    Value="Get (a copy)"(Array=@n5, Dimension 1=@n6.Array Index)
   ) {is=title K2Node_CallFunction}
  ]}
 SwapTagsBySubstring {nodes [
   SwapTagsBySubstring!() -> (then, StringToReplace:String, StringToInsert:String) {is=title K2Node_FunctionEntry}
   > n1 = Iterate Over Selected Assets()
   > n2 = Cast To SoundLibrary(Object=@n1)
   n3 = Get Sounds(self=@n2)
   > n4 = Keys(TargetMap=@n3)
   > n5 = Values(TargetMap=@n3)
   > n6 = For Each Loop(Array=@n4)
   TransactObject(execute=@n6.Completed, Object=@n2)
   > SetEditorProperty(Object=@n2, PropertyName=Sounds, PropertyValue=Get LocalSounds())
   n7 = GetDebugStringFromGameplayTag(GameplayTag=@n6)
   n8 = Branch(execute=@n6.LoopBody, Condition=Contains(SearchIn=@n7, Substring=Get StringToReplace() -> (StringToReplace:String)))
   > n9 = RequestGameplayTag(TagName=Replace(SourceString=@n7, From=Get StringToReplace(), To=Get StringToInsert() -> (StringToInsert:String)), bErrorIfNotFound=true)
   > n10 = Branch(Condition=@n9.ReturnValue)
   n11 = "Get (a copy)"(Array=@n5, Dimension 1=@n6.Array Index)
   > Add(TargetMap=Get LocalSounds(), Key=@n9, Value=@n11)
   Add(execute=[@n8.else, @n10.else], TargetMap=Get LocalSounds(), Key=@n6, Value=@n11)
  ]}
]
variables []
