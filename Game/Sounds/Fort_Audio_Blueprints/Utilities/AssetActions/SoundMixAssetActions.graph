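# Editor utility with asset actions for batch-editing Sound Mixes and baking their volumes into Sound Classes.
# Perform SoundMix Operation matches each selected mix's adjusters to a reference mix by class, then multiplies or divides volumes, adds missing entries or removes duplicates.
# Other functions compare two mixes for duplicates and strip invalid or neutral adjusters.
# Recursive functions walk Sound Class children, writing multiplied or inverse-divided volumes through editor properties.
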EditorUtilityBlueprint /Game/Sounds/Fort_Audio_Blueprints/Utilities/AssetActions/SoundMixAssetActions
graphs [
 ApplyInvAdjusterToSCGraph {nodes [
   n1 = ApplyInvAdjusterToSCGraph!() -> (then, Class:Sound Class Object Reference, Adjuster:"Sound Class Adjuster Structure (by ref)") {is=title K2Node_FunctionEntry}
   > Set LocalClass(LocalClass:Sound Class Object Reference=@n1) -> (Output_Get:Sound Class Object Reference)
   > n2 = Sequence()
   n3 = Get LocalClass() -> (LocalClass:Sound Class Object Reference)
   > n4 = GetEditorProperty!(execute:Exec, Object:Object Reference=@n3, PropertyName:Name=Properties) -> (then:Exec, PropertyValue:"Sound Class Properties Structure (by ref)", ReturnValue:bool)
   > n5 = Set members in Sound Class Properties!(
    execute:Exec
    StructRef:"Sound Class Properties Structure (by ref)"=@n4
    Volume:float=float / float(
     A=Break Sound Class Properties!(SoundClassProperties:"Sound Class Properties Structure (by ref)"=@n4) -> (Volume:float=~, ModulationSettings:Sound Modulation Default Settings Structure) {is=title K2Node_BreakStruct}
     B=Break Sound Class Adjuster!(SoundClassAdjuster:"Sound Class Adjuster Structure (by ref)"=@n1.Adjuster) -> (VolumeAdjuster:float=~) {is=title K2Node_BreakStruct}
    ) {is=title K2Node_CallFunction}
   ) -> (then:Exec, StructOut:"Sound Class Properties Structure (by ref)") {is=title K2Node_SetFieldsInStruct}
   > SetEditorProperty(Object=@n3, PropertyName=Properties, PropertyValue:"Sound Class Properties Structure (by ref)"=@n5)
   n6 = GetEditorProperty(execute=@n2.then_1, Object=Get LocalClass(), PropertyName=ChildClasses) -> (PropertyValue:Array of Sound Class Object References)
   > n7 = For Each Loop(Array:Array of Sound Class Object References=@n6) -> (Array Element:Sound Class Object Reference)
   > ApplyInvAdjusterToSCGraph!(execute, self, Class:Sound Class Object Reference=@n7, Adjuster:"Sound Class Adjuster Structure (by ref)"=@n1.Adjuster) -> (then) {is=title K2Node_CallFunction}
  ]}
 ApplySoundMixToSCGraph {nodes [
   n1 = ApplySoundMixToSCGraph!() -> (then, Mix:Sound Mix Object Reference, Class:Sound Class Object Reference) {is=title K2Node_FunctionEntry}
   > Set LocalMix(LocalMix:Sound Mix Object Reference=@n1) -> (Output_Get:Sound Mix Object Reference)
   > Set LocalClass(LocalClass=@n1.Class)
   > n2 = Sequence()
   > n3 = GetEditorProperty(Object=Get LocalMix() -> (LocalMix:Sound Mix Object Reference), PropertyName=SoundClassEffects) -> (PropertyValue:Array of Sound Class Adjuster Structures)
   > n4 = For Each Loop(Array:Array of Sound Class Adjuster Structures=@n3) -> (Array Element:Sound Class Adjuster Structure)
   n5 = Break Sound Class Adjuster!(SoundClassAdjuster=@n4) -> (SoundClassObject:Sound Class Object Reference, VolumeAdjuster=~, bApplyToChildren:bool=~)
   n6 = Get LocalClass()
   > Branch(Condition="Equal (Object)"(A=@n5, B=@n6) {is=title K2Node_CallFunction})
   > TransactObject(Object=@n6)
   > n7 = GetEditorProperty(Object=@n6, PropertyName=Properties) -> (PropertyValue:"Sound Class Properties Structure (by ref)")
   > n8 = Set members in Sound Class Properties(StructRef=@n7, Volume=float * float(A=Break Sound Class Properties(SoundClassProperties=@n7) -> (Volume=~), B=@n5.VolumeAdjuster))
   > SetEditorProperty(Object=@n6, PropertyName=Properties, PropertyValue=@n8)
   n9 = GetEditorProperty(execute=@n2.then_1, Object=Get LocalClass(), PropertyName=ChildClasses) -> (PropertyValue:Array of Sound Class Object References)
   > n10 = For Each Loop(Array:Array of Sound Class Object References=@n9) -> (Array Element:Sound Class Object Reference)
   > ApplySoundMixToSCGraph!(execute, self, Mix:Sound Mix Object Reference=Get LocalMix(), Class:Sound Class Object Reference=@n10) -> (then) {is=title K2Node_CallFunction}
   n11 = Branch()
   n12 = GetEditorProperty(execute=@n11.else, Object=Get LocalClass(), PropertyName=ChildClasses)
   > n13 = For Each Loop(Array=@n12)
   > ApplyInvAdjusterToSCGraph(Class=@n13, Adjuster=@n4)
  ]}
 CompareAdjusters {nodes [
   n1 = CompareAdjusters!() -> (then, AdjusterA:Sound Class Adjuster Structure, AdjusterB:Sound Class Adjuster Structure) {is=title K2Node_FunctionEntry}
   n2 = Break Sound Class Adjuster!(SoundClassAdjuster=@n1) -> (
    SoundClassObject
    VolumeAdjuster=~
    PitchAdjuster:float=~
    LowPassFilterFrequency:float=~
    bApplyToChildren=~
    VoiceCenterChannelVolumeAdjuster:float=~
   )
   n3 = Break Sound Class Adjuster(SoundClassAdjuster=@n1.AdjusterB) -> (VolumeAdjuster=~, PitchAdjuster=~, LowPassFilterFrequency=~, bApplyToChildren=~, VoiceCenterChannelVolumeAdjuster=~)
   > n4 = Branch(Condition="Equal (Object)"(A=@n2, B=@n3))
   > Return Node!(
    execute
    MixFound:bool=true
    IsDupe:bool="Equal (Float)"!(A:double=@n2.VolumeAdjuster, B:double=@n3.VolumeAdjuster) -> (ReturnValue:bool) {is=title K2Node_CallFunction}
    Adjuster A:Sound Class Adjuster Structure=@n1
    Adjuster B:Sound Class Adjuster Structure=@n1.AdjusterB
   )
   Return Node(execute=@n4.else, MixFound=false, IsDupe=false)
   float - float(A=@n3.VolumeAdjuster, B=@n2.VolumeAdjuster) {is=title K2Node_CallFunction}
   n5 = "To String (Float)"(InDouble=@n2.VolumeAdjuster)
   SelectString(
    A=Append!(
     A="CMP DIFF: "
     B=GetDisplayName(Object=@n2)
     C:String=" A: "
     D:String=@n5
     E:String=" B: "
     F:String="To String (Float)"(InDouble=@n3.VolumeAdjuster)
    ) -> (ReturnValue)
    B=Append!(A="CMP DUPE: ", B=GetDisplayName(Object=@n2), C=" at ", D=@n5) -> (ReturnValue)
    bPickA="Equal (Object)"(A=@n2, B=@n3)
   )
  ]}
 CompareMixes {nodes [
   n1 = CompareMixes!() -> (then, MixA:Sound Mix Object Reference, MixB:Sound Mix Object Reference) {is=title K2Node_FunctionEntry}
   > n2 = GetEditorProperty(Object=@n1, PropertyName=SoundClassEffects) -> (PropertyValue:Array of Sound Class Adjuster Structures)
   > Clear(TargetArray:Array of Sound Class Adjuster Structures=Get Dupes() -> (Dupes:Array of Sound Class Adjuster Structures))
   > n3 = For Each Loop(Array:Array of Sound Class Adjuster Structures=@n2) -> (Array Element:Sound Class Adjuster Structure)
   Return Node!(execute=@n3.Completed)
   > n4 = GetEditorProperty(Object=@n1.MixB, PropertyName=SoundClassEffects)
   > n5 = Set Effects(Effects:Array of Sound Class Adjuster Structures=@n4) -> (Output_Get:Array of Sound Class Adjuster Structures)
   > n6 = For Each Loop(Array=@n5)
   Is Valid(exec=@n6.Completed, InputObject=Reroute Node(InputPin:Object Reference) -> (OutputPin:Object Reference))
   > SetEditorProperty(
    Object=@n1.MixB
    PropertyName=SoundClassEffects
    PropertyValue:Array of Sound Class Adjuster Structures=Get Effects() -> (Effects:Array of Sound Class Adjuster Structures)
   )
   > SetEditorProperty(Object=@n1, PropertyName=SoundClassEffects, PropertyValue=Get Effects())
   n7 = CompareAdjusters!(AdjusterA:Sound Class Adjuster Structure=@n3, AdjusterB:Sound Class Adjuster Structure=@n6) -> (MixFound:bool, IsDupe:bool, Adjuster A:Sound Class Adjuster Structure, Adjuster B:Sound Class Adjuster Structure) {is=title K2Node_CallFunction}
   Branch(execute=@n6.LoopBody, Condition=@n7)
   > Branch(Condition=@n7.IsDupe)
   > Remove Item!(execute, TargetArray:Array of Sound Class Adjuster Structures=@n5, Item:"Sound Class Adjuster Structure (by ref)"=@n7.Adjuster B) -> (then, ReturnValue:bool)
   > Add(TargetArray:Array of Sound Class Adjuster Structures=Get Dupes(), NewItem:"Sound Class Adjuster Structure (by ref)"=@n7.Adjuster B) {is=title K2Node_CallArrayFunction}
   Get Dupes()
  ]}
 EventGraph {nodes [
   n1 = Perform SoundMix Operation!() -> (OutputDelegate, then, Reference Mix:Sound Mix Object Reference, Operation Type:SoundMixOperationType Enum=NewEnumerator1) {is=title K2Node_CustomEvent}
   > n2 = GetSelectedAssets()
   > n3 = For Each Loop(Array:Array of Object References=@n2) -> (Array Element:Object Reference)
   > n4 = Cast To SoundMix!(execute, Object=@n3) -> (then, CastFailed, AsSound Mix:Sound Mix Object Reference) {is=title K2Node_DynamicCast}
   > PerformSoundMixOperation_Internal!(
    execute
    self
    Operation Mix:Sound Mix Object Reference=@n4
    Reference Mix:Sound Mix Object Reference=@n1.Reference Mix
    Operation Type:SoundMixOperationType Enum=@n1.Operation Type
   ) -> (then)
  ]}
 ModifyAdjuster {nodes [
   n1 = ModifyAdjuster!() -> (then, Adjuster:"Sound Class Adjuster Structure (by ref)", Operation:SoundMixOperationType Enum=NewEnumerator0, Term:double=~) {is=title K2Node_FunctionEntry}
   n2 = Break Sound Class Adjuster(SoundClassAdjuster=@n1) -> (VolumeAdjuster=~, PitchAdjuster=~, LowPassFilterFrequency=~, bApplyToChildren=~, VoiceCenterChannelVolumeAdjuster=~)
   n3 = "To String (Text)"(InText="To Text (Float)"(
     Value=float / float(A=@n2.VolumeAdjuster, B=@n1.Term)
     MinimumIntegralDigits=1
     MaximumIntegralDigits=324
     MinimumFractionalDigits=0
     MaximumFractionalDigits=3
    ))
   > Set members in Sound Class Adjuster!(
    execute:Exec
    StructRef:"Sound Class Adjuster Structure (by ref)"=@n1
    VolumeAdjuster:float=Select!(
     NewEnumerator1:double=float * float(A=@n1.Term, B=@n2.VolumeAdjuster)
     NewEnumerator2:double=String To Float(InString=@n3)
     NewEnumerator5:double=@n2.VolumeAdjuster
     NewEnumerator6:double=@n2.VolumeAdjuster
     Index:SoundMixOperationType Enum=@n1.Operation
    ) -> (ReturnValue:double)
   ) -> (then:Exec, StructOut:"Sound Class Adjuster Structure (by ref)") {is=title K2Node_SetFieldsInStruct}
  ]}
 PerformSoundMixOperation_Internal {nodes [
   n1 = PerformSoundMixOperation_Internal!() -> (
    then
    Operation Mix:Sound Mix Object Reference
    Reference Mix:Sound Mix Object Reference
    Operation Type:SoundMixOperationType Enum=NewEnumerator1
   ) {is=title K2Node_FunctionEntry}
   > Set OperationMixLocal(OperationMixLocal:Sound Mix Object Reference=@n1) -> (Output_Get:Sound Mix Object Reference)
   > Set ReferenceMixLocal(ReferenceMixLocal:Sound Mix Object Reference=@n1.Reference Mix) -> (Output_Get:Sound Mix Object Reference)
   > Set LocalOperation(LocalOperation:SoundMixOperationType Enum=@n1.Operation Type) -> (Output_Get:SoundMixOperationType Enum=NewEnumerator1)
   > n2 = GetEditorProperty(Object=Get ReferenceMixLocal() -> (ReferenceMixLocal:Sound Mix Object Reference), PropertyName=SoundClassEffects)
   > Set ReferenceMixEffects(ReferenceMixEffects:Array of Sound Class Adjuster Structures=@n2) -> (Output_Get:Array of Sound Class Adjuster Structures)
   > n3 = For Each Loop(Array:Array of Sound Class Adjuster Structures=@n2) -> (Array Element:Sound Class Adjuster Structure)
   > n4 = GetEditorProperty(Object=Get OperationMixLocal() -> (OperationMixLocal:Sound Mix Object Reference), PropertyName=SoundClassEffects)
   > Set OperationMixEffects(OperationMixEffects:Array of Sound Class Adjuster Structures=@n4) -> (Output_Get:Array of Sound Class Adjuster Structures)
   > Set MixFound(MixFound:bool=false) -> (Output_Get:bool)
   > n5 = For Each Loop(Array=@n4)
   n6 = Sequence(execute=@n5.Completed)
   n7 = Get OperationMixEffects() -> (OperationMixEffects:Array of Sound Class Adjuster Structures)
   n8 = Contains Item(TargetArray:Array of Sound Class Adjuster Structures=@n7, ItemToFind:"Sound Class Adjuster Structure (by ref)"=@n3)
   n9 = Get MixFound() -> (MixFound:bool)
   n10 = Get LocalOperation() -> (LocalOperation:SoundMixOperationType Enum=NewEnumerator1)
   > Branch(Condition=AND Boolean!(
     A=NOT Boolean(A=@n9)
     B=NOT Boolean(A=@n8)
     C:bool="Equal (Enum)"(A:SoundMixOperationType Enum=@n10, B:SoundMixOperationType Enum=NewEnumerator5)
    ) -> (ReturnValue))
   > Add(TargetArray=@n7, NewItem=@n3)
   SetEditorProperty(execute=@n6.then_1, Object=Get OperationMixLocal(), PropertyName=SoundClassEffects, PropertyValue=Get OperationMixEffects())
   n11 = CompareAdjusters(AdjusterA=@n5, AdjusterB=@n3)
   n12 = Get MixFound()
   Branch(execute=@n5.LoopBody, Condition=AND Boolean!(A=@n11, B=NOT Boolean(A=@n12)) -> (ReturnValue))
   > Set MixFound(MixFound=true)
   > n13 = Branch(Condition=@n11.IsDupe)
   > Branch(Condition="Equal (Enum)"(A=Get LocalOperation() -> (LocalOperation=NewEnumerator1), B=NewEnumerator6))
   > Remove Item(TargetArray=Get OperationMixEffects(), Item=@n11.Adjuster B)
   n14 = Get OperationMixEffects()
   ModifyAdjuster!(
    execute=@n13.else
    self
    Adjuster:"Sound Class Adjuster Structure (by ref)"="Get (a ref)"(
     Array:Array of Sound Class Adjuster Structures=@n14
     Dimension 1=Find Item!(TargetArray:Array of Sound Class Adjuster Structures=@n14, ItemToFind:"Sound Class Adjuster Structure (by ref)"=@n11.Adjuster A) -> (ReturnValue)
    ) -> (Output:"Sound Class Adjuster Structure (by ref)")
    Operation:SoundMixOperationType Enum=Get LocalOperation() -> (LocalOperation=NewEnumerator1)
    Term:double=Break Sound Class Adjuster(SoundClassAdjuster=@n11.Adjuster B) -> (VolumeAdjuster=~, PitchAdjuster=~, LowPassFilterFrequency=~, bApplyToChildren=~, VoiceCenterChannelVolumeAdjuster=~).VolumeAdjuster
   ) -> (then) {is=title K2Node_CallFunction}
   Remove Item(TargetArray=Get ReferenceMixEffects() -> (ReferenceMixEffects:Array of Sound Class Adjuster Structures), Item=@n3)
   SetEditorProperty(Object=Get ReferenceMixLocal(), PropertyName=SoundClassEffects, PropertyValue=Get ReferenceMixEffects())
  ]}
 RemoveEmptyMixModifiers {nodes [
   RemoveEmptyMixModifiers() {is=title K2Node_FunctionEntry}
   > n1 = Iterate Over Selected Assets()
   > n2 = Cast To SoundMix(Object=@n1)
   n3 = Get SoundClassEffects!(self:Sound Mix Object Reference=@n2) -> (SoundClassEffects:Array of Sound Class Adjuster Structures)
   > n4 = For Each Loop(Array=@n3)
   n5 = Sequence(execute=@n4.Completed)
   > Branch(Condition=Is Not Empty!(TargetArray:Array of Integers=Get IndicesToRemove() -> (IndicesToRemove:Array of Integers)) -> (ReturnValue:bool))
   > TransactObject(Object=@n2)
   n6 = Reverse for Each Loop!(Exec:Exec=@n5.then_1, Array:Array of Integers=Get IndicesToRemove()) -> (LoopBody:Exec, ArrayIndex:int, ArrayElement:int, Completed:Exec) {is=title K2Node_MacroInstance}
   Clear(execute=@n6.Completed, TargetArray:Array of Integers=Get IndicesToRemove())
   Remove Index!(execute=@n6.LoopBody, TargetArray:Array of Sound Class Adjuster Structures=@n3, IndexToRemove:int=@n6.ArrayElement) -> (then)
   n7 = Break Sound Class Adjuster(SoundClassAdjuster=@n4) -> (VolumeAdjuster=~, PitchAdjuster=~, LowPassFilterFrequency=~, bApplyToChildren=~, VoiceCenterChannelVolumeAdjuster=~)
   n8 = IsValid(Object=@n7)
   n9 = "Equal (Float)"(A=@n7.VolumeAdjuster, B=1) {is=title K2Node_PromotableOperator}
   n10 = "Equal (Float)"(A=@n7.LowPassFilterFrequency, B=20000)
   n11 = "Equal (Float)"(A=@n7.PitchAdjuster, B=1)
   Branch(execute=@n4.LoopBody, Condition=OR Boolean(A=NOT Boolean(A=@n8), B=AND Boolean!(A=@n9, B=@n11, C:bool=@n10) -> (ReturnValue)))
   > Add(TargetArray:Array of Integers=Get IndicesToRemove(), NewItem:"Integer (by ref)"=@n4.Array Index)
  ]}
]
variables []
