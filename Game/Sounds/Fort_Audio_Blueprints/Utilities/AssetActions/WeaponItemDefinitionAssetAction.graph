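# Editor asset action utility whose SetWhipAudioData function assigns a chosen bullet whip tracker audio data asset to selected Fortnite weapon item definitions.
# For each selected weapon, it checks for whip tracker item data and, when none is found, sets it and records the weapon in AssetsToSave.
# A remove-and-replace step exists but has no incoming execution link.
# After iteration, it saves the recorded assets together if any exist.
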EditorUtilityBlueprint /Game/Sounds/Fort_Audio_Blueprints/Utilities/AssetActions/WeaponItemDefinitionAssetAction
graphs [
 EventGraph {links [], nodes []}
 SetWhipAudioData {nodes [
   n1 = Iterate Over Selected Assets(Trigger=@n4.then, AssetClass=/Script/FortniteGame.FortWeaponItemDefinition)
   > n2 = Cast To FortWeaponItemDefinition!(execute, Object=@n1) -> (then, CastFailed, AsFort Weapon Item Definition:Fort Weapon Item Definition Object Reference) {is=title K2Node_DynamicCast}
   n3 = RemoveItemData!(
    execute
    self:Item Definition Base Object Reference=@n2
    StructType:Script Struct Object Reference=/Script/FortniteGame.FortItemComponentData_BulletWhipTrackerData
   ) -> (Found:Exec, NotFound:Exec)
   SetItemData!(
    execute=[@n3.Found, @n5.NotFound]
    self:Item Definition Base Object Reference=@n2
    ItemData:Fortnite Bullet Whip Tracker Data Item Data Structure=Make Fortnite Bullet Whip Tracker Data Item Data!(BulletWhipTrackerData:Bullet Whip Tracker Audio Data Object Reference=Get WhipData() -> (WhipData:Bullet Whip Tracker Audio Data Object Reference)) -> (FortItemComponentData_BulletWhipTrackerData:Fortnite Bullet Whip Tracker Data Item Data Structure) {is=title K2Node_MakeStruct}
   ) -> (then)
   > Add Unique(
    TargetArray:Array of Fort Weapon Item Definition Object References=Get AssetsToSave() -> (AssetsToSave:Array of Fort Weapon Item Definition Object References)
    NewItem:"Fort Weapon Item Definition Object Reference (by ref)"=@n2
   )
   n4 = SetWhipAudioData!() -> (then, WhipData:Bullet Whip Tracker Audio Data Object Reference) {is=title K2Node_FunctionEntry}
   n5 = GetItemData!(execute=@n2.then, self:Item Definition Base Object Reference=@n2) -> (Found:Exec, NotFound:Exec, ItemData:"Fortnite Bullet Whip Tracker Data Item Data Structure (by ref)")
   Break Fortnite Bullet Whip Tracker Data Item Data!(FortItemComponentData_BulletWhipTrackerData:"Fortnite Bullet Whip Tracker Data Item Data Structure (by ref)"=@n5) -> (BulletWhipTrackerData:Bullet Whip Tracker Audio Data Object Reference) {is=title K2Node_BreakStruct}
   n6 = Get AssetsToSave()
   SaveLoadedAssets(execute=@n8.then, AssetsToSave=@n6, bOnlyIfIsDirty=false)
   n7 = Is Not Empty!(TargetArray:Array of Fort Weapon Item Definition Object References=@n6) -> (ReturnValue:bool)
   n8 = Branch(execute=@n1.Completed, Condition=@n7)
  ]}
]
variables []
